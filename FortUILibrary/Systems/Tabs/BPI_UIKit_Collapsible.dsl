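; BPI_UIKit_Collapsible is a Blueprint interface in the UIKit tab system for UI elements that can collapse and expand.
; SetIsCollapsed takes the target collapsed state and a flag to skip the animation.
; GetRemainingAnimTime returns a boolean and two floats, which by its name report on the collapse animation still left to play.
; It declares only these signatures, leaving the behavior to implementers.
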(fn GetRemainingAnimTime
  (return false 0.0 0.0))

(fn SetIsCollapsed
  (params (IsCollapsed :type Boolean) (SkipAnimation :type Boolean)))
