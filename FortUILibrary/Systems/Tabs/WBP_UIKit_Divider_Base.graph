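# UI Kit base widget that draws a tab-system divider bar as a tinted Image inside a Grid panel.
# PreConstruct applies color, brush size and padding from variables, then overrides padding on mobile, which also selects the compact width.
# Separate functions set orientation and spacing through the bar's grid slot alignment and margins.
# A ValidateGridSlot macro guards slot edits and prints a warning when the bar has no grid slot.
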WidgetBlueprint /FortUILibrary/Systems/Tabs/WBP_UIKit_Divider_Base
graphs [
 EventGraph {nodes [
   Event PreConstruct()
   > SetDividerColorAndOpacity!(execute, self, InColorAndOpacity:Linear Color Structure=Get Div Color and Opacity() -> (Div Color and Opacity:Linear Color Structure)) -> (then)
   > SetDividerBrushSize!(
    execute
    self
    DivWidth:double=Get Div Width() -> (Div Width:double)
    DivWidthCompactUI:double=Get Div Width CompactUI() -> (Div Width CompactUI:double)
   ) -> (then)
   > SetDividerPadding!(execute, self, Padding:Margin Structure=Get Div Padding() -> (Div Padding:Margin Structure)) -> (then)
   > SetCompactUIStyle() {is=title K2Node_CallFunction}
  ]}
 SetCompactUIStyle {nodes [
   SetCompactUIStyle() {is=title K2Node_FunctionEntry}
   > Branch(Condition=IsMobileGame())
   > SetPadding!(
    execute
    self:Grid Slot Object Reference=SlotAsGridSlot(Widget=Get DividerBar() -> (DividerBar:Image Object Reference))
    InPadding_Left:float=0
    InPadding_Top:float=24
    InPadding_Right:float=0
    InPadding_Bottom:float=24
   ) -> (then)
  ]}
 SetDividerBrushSize {nodes [
   n1 = SetDividerBrushSize!() -> (then, DivWidth:double=~, DivWidthCompactUI:double=~) {is=title K2Node_FunctionEntry}
   n2 = MakeVector2D(X=@n1, Y=@n1)
   n3 = IsMobileGame()
   n4 = MakeVector2D(X=@n1.DivWidthCompactUI, Y=@n1.DivWidthCompactUI)
   > SetBrush(
    self=Get DividerBar()
    InBrush=Make Slate Brush!(
     TintColor:Slate Color Structure=(SpecifiedColor=(R=1,G=1,B=1,A=1),ColorUseRule=UseColor_Specified)
     DrawAs:ESlateBrushDrawType Enum=Image
     Tiling:ESlateBrushTileType Enum=NoTile
     Mirroring:ESlateBrushMirrorType Enum=NoMirror
     ImageSize:Vector 2D Structure=Select(Option 0:Vector 2D Structure=@n2, Option 1:Vector 2D Structure=@n4, Index=@n3) -> (ReturnValue:Vector 2D Structure)
     Margin:Margin Structure=(Left=0,Top=0,Right=0,Bottom=0)
     ResourceObject:Object Reference
     OutlineSettings:Slate Brush Outline Settings Structure=(CornerRadii=(X=0,Y=0,Z=0,W=0),Color=(SpecifiedColor=(R=0,G=0,B=0,A=0),ColorUseRule=UseColor_Specified),Width=0,RoundingType=HalfHeightRadius,bUseBrushTransparency=False)
    ) -> (SlateBrush:Slate Brush Structure) {is=title K2Node_MakeStruct}
   )
  ]}
 SetDividerColorAndOpacity {nodes [
   n1 = SetDividerColorAndOpacity!() -> (then, InColorAndOpacity:Linear Color Structure) {is=title K2Node_FunctionEntry}
   > SetColorAndOpacity(self:Image Object Reference=Get DividerBar(), InColorAndOpacity:Linear Color Structure=@n1)
  ]}
 SetDividerOrientation {nodes [
   n1 = SetDividerOrientation!() -> (then, IsVertical:bool=~) {is=title K2Node_FunctionEntry}
   > n2 = Validate Grid Slot!(execute:Exec) -> (Valid:Exec, Not Valid:Exec, Grid Slot:Grid Slot Object Reference) {is=title K2Node_MacroInstance}
   > SetHorizontalAlignment(
    self:Grid Slot Object Reference=@n2
    InHorizontalAlignment=Select(Option 0:EHorizontalAlignment Enum=HAlign_Center, Option 1:EHorizontalAlignment Enum=HAlign_Fill, Index=@n1) -> (ReturnValue:EHorizontalAlignment Enum)
   )
   > SetVerticalAlignment(
    self:Grid Slot Object Reference=@n2
    InVerticalAlignment=Select(Option 0:EVerticalAlignment Enum=VAlign_Fill, Option 1:EVerticalAlignment Enum=VAlign_Center, Index=@n1) -> (ReturnValue:EVerticalAlignment Enum)
   )
  ]}
 SetDividerPadding {nodes [
   n1 = SetDividerPadding!() -> (then, Padding:Margin Structure) {is=title K2Node_FunctionEntry}
   > n2 = Validate Grid Slot()
   > Set Padding!(execute, Padding:Margin Structure=@n1, self:Grid Slot Object Reference=@n2) -> (then, Output_Get:Margin Structure)
  ]}
 SetDividerSpacing {nodes [
   n1 = SetDividerSpacing!() -> (then, Spacing:double=~, IsVertical:bool=~) {is=title K2Node_FunctionEntry}
   > n2 = Validate Grid Slot()
   > Set Padding(
    Padding=Select(
     Option 0:Margin Structure=Make Margin(Left=@n1, Top=0, Right=@n1, Bottom=0)
     Option 1:Margin Structure=Make Margin(Left=0, Top=@n1, Right=0, Bottom=@n1)
     Index=@n1.IsVertical
    ) -> (ReturnValue:Margin Structure)
    self=@n2
   )
  ]}
 ValidateGridSlot {nodes [
   Inputs()
   n1 = SlotAsGridSlot(Widget=Get DividerBar())
   > n2 = Branch(Condition=IsValid(Object=@n1))
   > Outputs!(Valid:Exec, Not Valid:Exec=@n4.then, Grid Slot:Grid Slot Object Reference=@n1)
   n3 = Self-Reference()
   n4 = PrintString!(
    execute=@n2.else
    InString=Append!(A="WARNING: ", B=GetDisplayName(Object=@n3), C:String=" - needs to be placed inside Grid panel.") -> (ReturnValue)
    bPrintToScreen
    bPrintToLog
    TextColor=(R=1,G=0,B=0.5683,A=1)
    Duration=8
    Key
   ) -> (then)
  ]}
]
variables [
 Div Color and Opacity {type=LinearColor, default=<rgba 1 1 1 1>}
 Div Padding {type=Margin, default=<ltrb 0 0 0 0>}
 Div Width {type=number, default=6}
 Div Width CompactUI {type=number, default=2}
]
widgets {
 info {inheritedWidgetCount=0, parentClass=&UserWidget, rootWidgetClass=&GridPanel, widgetCount=2}
 widgets [
  Root {widgetClassPath=&GridPanel, widget=Root, properties {columnFill [1], rowFill [1]}}
  DividerBar {
   widgetClassPath=&Image
   bIsVariable=true
   parent=Root
   widget=DividerBar
   properties {
    renderOpacity=0.8
    brush {
     bIsDynamicallyLoaded=false
     drawAs=Image
     imageType=NoImage
     mirroring=NoMirror
     resourceName=None
     resourceObject=null
     tiling=NoTile
     imageSize=<6 6>
     margin=<ltrb 0 0 0 0>
     outlineSettings {
      bUseBrushTransparency=false
      roundingType=HalfHeightRadius
      width=0
      color {colorUseRule=UseColor_Specified, specifiedColor=<rgba 0 0 0 0>}
      cornerRadii=<0 0 0 0>
     }
     tintColor {colorUseRule=UseColor_Specified, specifiedColor=<rgba 1 1 1 1>}
     uVRegion {bIsValid=false, max=<0 0>, min=<0 0>}
    }
   }
   slot {class=GridSlot, properties {horizontalAlignment=HAlign_Center}}
  }
 ]
}
