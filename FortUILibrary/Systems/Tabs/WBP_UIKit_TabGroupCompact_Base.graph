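# Base compact tab group widget on UIKitTabGroup that lays out tab buttons vertically or horizontally in a GridPanel.
# Tabs are built from initial or supplied data, with cached viewmodels injected into each button's MVVM view.
# Tab creation styles buttons and interleaves spacers and optional dividers using doubled grid indices.
# Divider placements are cached by tab ID or index, surviving rebuilds, and selection is restored instantly at runtime.
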WidgetBlueprint /FortUILibrary/Systems/Tabs/WBP_UIKit_TabGroupCompact_Base
graphs [
 AddDivider {nodes [
   n1 = AddDivider!() -> (then, TabID:Name, AddAfter:bool=true) {is=title K2Node_FunctionEntry}
   > n2 = Set CurrentTabIndex(CurrentTabIndex:int=GetTabIndexByID!(self:UIKit Tab Group Object Reference, TabNameID:Name=@n1) -> (ReturnValue:int=0)) -> (Output_Get:int=0)
   > n3 = Branch(Condition="integer >= integer"(A=@n2))
   > HandleAddDivider!(execute, self, TabID:"Name (by ref)"=@n1, TabIndex:int=@n2, AddAfter:bool=@n1.AddAfter) -> (then)
   Add!(
    execute=@n3.else
    TargetMap:Map of Names to S UI Divider Data Structures=Get CachedDividers() -> (CachedDividers:Map of Names to S UI Divider Data Structures)
    Key:"Name (by ref)"=@n1
    Value:"S UI Divider Data Structure (by ref)"=Make S UI Divider Data!(
     Widget_6_#1:WBP UIKit Divider Base Object Reference
     IsAddedAfter_3_#2:bool=@n1.AddAfter
    ) -> (S_UI_DividerData:S UI Divider Data Structure) {is=title K2Node_MakeStruct}
   ) -> (then) {is=title K2Node_CallFunction}
  ]}
 AddDividerByIndex {nodes [
   n1 = AddDividerByIndex!() -> (then, TabIndex:int, AddAfter:bool=true) {is=title K2Node_FunctionEntry}
   > n2 = Set CurrentTabID(CurrentTabID:Name=GetTabIDbyIndex!(self:UIKit Tab Group Object Reference, Index:int=@n1) -> (ReturnValue:Name=None)) -> (Output_Get:Name=None)
   > n3 = Branch(Condition="Not Equal (Name)"(A=@n2) -> (ReturnValue=~))
   > HandleAddDivider(TabID=@n2, TabIndex=@n1, AddAfter=@n1.AddAfter)
   Add(
    execute=@n3.else
    TargetMap=Get CachedDividers()
    Key=String To Name(InString="To String (Integer)"(InInt=@n1))
    Value=Make S UI Divider Data(IsAddedAfter_3_#2=@n1.AddAfter)
   )
  ]}
 AddDividerToGrid {nodes [
   n1 = AddDividerToGrid!() -> (then, Divider:WBP UIKit Divider Base Object Reference, Index:int, AddAfter:bool=~) {is=title K2Node_FunctionEntry}
   > n2 = Set CurrentGridCell(CurrentGridCell:int=int + int(A=int * int(A=@n1.Index, B=2), B=SelectInt(A=2, B=0, bPickA=@n1.AddAfter))) -> (Output_Get:int=0)
   n3 = Get IsVertical() -> (IsVertical:bool)
   > AddChildToGrid!(
    execute
    self:Grid Panel Object Reference=Get TabsContainer() -> (TabsContainer:Grid Panel Object Reference)
    Content:Widget Object Reference=@n1
    InRow:int=SelectInt(A=@n2, B=0, bPickA=@n3)
    InColumn:int=SelectInt(A=0, B=@n2, bPickA=@n3)
   ) -> (then, ReturnValue:Grid Slot Object Reference)
  ]}
 AddSpacerToGrid {nodes [
   n1 = AddSpacerToGrid!() -> (then, TabID:"Name (by ref)", Index:int) {is=title K2Node_FunctionEntry}
   > Branch(Condition="integer > integer"(A=@n1.Index))
   > n2 = Construct Object from Class(Class=&Spacer) -> (ReturnValue:Spacer Object Reference)
   > n3 = Set CurrentSpacer(CurrentSpacer:Spacer Object Reference=@n2) -> (Output_Get:Spacer Object Reference)
   n4 = Get TabSpacing() -> (TabSpacing:double)
   n5 = Get IsVertical()
   > SetSize!(
    execute
    self:Spacer Object Reference=@n3
    InSize_X:double=Select(Option 0:double=@n4, Option 1:double=1, Index=@n5) -> (ReturnValue:double)
    InSize_Y:double=Select(Option 0=1, Option 1=@n4, Index=@n5)
   ) -> (then)
   n6 = int * int(A=@n1.Index, B=2)
   n7 = Get IsVertical()
   > AddChildToGrid(
    self=Get TabsContainer()
    Content=@n3
    InRow=Select(Option 0:int=0, Option 1:int=@n6, Index=@n7) -> (ReturnValue:int)
    InColumn=Select(Option 0=@n6, Option 1=0, Index=@n7)
   )
   > Add(
    TargetMap:Map of Names to Spacer Object References=Get CachedSpacers() -> (CachedSpacers:Map of Names to Spacer Object References)
    Key=Get TabID() -> (TabID:"Name (by ref)"=None)
    Value:"Spacer Object Reference (by ref)"=@n3
   )
  ]}
 AddTab {nodes [
   n1 = AddTab!() -> (then, Tab VM:MVVM Base Viewmodel Object Reference, Tab ID:Name, Index:int=-1) {is=title K2Node_FunctionEntry}
   > n2 = Set CurrentTabIndex(CurrentTabIndex=SelectInt(
     A=GetTabCount!(self:Common Tab List Widget Base Object Reference) -> (ReturnValue:int=0)
     B=@n1.Index
     bPickA="integer < integer"(A=@n1.Index)
    )) -> (Output_Get=0)
   n3 = "To String (Integer)"(InInt=@n2)
   > n4 = Set CurrentTabID(CurrentTabID=Select(Option 0:Name=@n1.Tab ID, Option 1:Name=String To Name(InString=@n3), Index="Equal (Name)"(A=@n1.Tab ID)) -> (ReturnValue:Name)) -> (Output_Get=None)
   > Add(
    TargetMap:Map of Names to MVVM Base Viewmodel Object References=Get CachedViewModels() -> (CachedViewModels:Map of Names to MVVM Base Viewmodel Object References)
    Key=@n4
    Value:"MVVM Base Viewmodel Object Reference (by ref)"=@n1
   )
   > n5 = RegisterTab!(
    execute
    self:Common Tab List Widget Base Object Reference
    TabNameID:Name=@n4
    ButtonWidgetType:Common Button Base Class Reference=Get ButtonClass() -> (ButtonClass:Common Button Base Class Reference)
    ContentWidget:Widget Object Reference
    TabIndex:int=@n2
   ) -> (then, ReturnValue:bool)
   > n6 = Branch(Condition=@n5)
   > Branch(Condition=IsValid(Object=@n1))
   > SetTabViewModel!(execute, self, TabID:Name=@n4, VM:Notify Field Value Changed Interface=@n1) -> (then)
   Remove!(execute=@n6.else, TargetMap:Map of Names to MVVM Base Viewmodel Object References=Get CachedViewModels(), Key:"Name (by ref)"=@n4) -> (then, ReturnValue:bool)
  ]}
 AddTabToGrid {nodes [
   n1 = AddTabToGrid!() -> (then, Widget:Widget Object Reference, Index:int) {is=title K2Node_FunctionEntry}
   > n2 = Branch(Condition=Get IsVertical())
   n3 = int + int(A=int * int(A=@n1.Index, B=2))
   > n4 = AddChildToGrid(self=Get TabsContainer(), Content=@n1, InRow=@n3, InColumn=0)
   > n5 = SetHorizontalAlignment(
    self:Grid Slot Object Reference=@n4
    InHorizontalAlignment=Get ButtonHorizontalAlignment() -> (ButtonHorizontalAlignment:EHorizontalAlignment Enum=HAlign_Fill)
   )
   SetGridSlotPadding!(
    execute=[@n5.then, @n7.then]
    Widget:Widget Object Reference=Get Widget() -> (Widget:Widget Object Reference)
    Padding:Margin Structure=Get ButtonPadding() -> (ButtonPadding:Margin Structure)
   ) -> (then)
   n6 = AddChildToGrid(execute=@n2.else, self=Get TabsContainer(), Content=@n1, InRow=0, InColumn=@n3)
   > n7 = SetVerticalAlignment(
    self:Grid Slot Object Reference=@n6
    InVerticalAlignment=Get ButtonVerticalAlignment() -> (ButtonVerticalAlignment:EVerticalAlignment Enum=VAlign_Fill)
   )
  ]}
 CacheInitialDividers {nodes [
   CacheInitialDividers() {is=title K2Node_FunctionEntry}
   > Clear!(execute, TargetMap:Map of Names to S UI Divider Data Structures=Get CachedDividers()) -> (then) {is=title K2Node_CallFunction}
   > n1 = For Each Loop(Array:Array of S UI Divider Position Structures=Get Dividers() -> (Dividers:Array of S UI Divider Position Structures)) -> (Array Element:S UI Divider Position Structure)
   n2 = Break S UI Divider Position!(S_UI_DividerPosition:"S UI Divider Position Structure (by ref)"=@n1) -> (
    TabID_8_#1:Name
    TabIndex_12_#3:int
    AddAfterTab_9_#2:bool=~
   ) {is=title K2Node_BreakStruct}
   > n3 = Branch(Condition="Not Equal (Name)"(A=@n2) -> (ReturnValue=~))
   n4 = Make S UI Divider Data(IsAddedAfter_3_#2=@n2.AddAfterTab_9_#2)
   n5 = Get CachedDividers()
   > Add(TargetMap:Map of Names to S UI Divider Data Structures=@n5, Key=@n2, Value:"S UI Divider Data Structure (by ref)"=@n4)
   Branch(execute=@n3.else, Condition="integer >= integer"(A=@n2.TabIndex_12_#3))
   > Add(TargetMap=@n5, Key=String To Name(InString="To String (Integer)"(InInt=@n2.TabIndex_12_#3)), Value=@n4)
  ]}
 CreateEmptyTabs {nodes [
   n1 = CreateEmptyTabs!() -> (then, TabsCount:int) {is=title K2Node_FunctionEntry}
   > RemoveTabGroupWidgets()
   > n2 = For Loop(LastIndex=int - int(A=@n1))
   InitialTabSelection(execute=@n2.Completed)
   n3 = Set TabID(execute=@n2.LoopBody, TabID:Name=String To Name(InString="To String (Integer)"(InInt=@n2))) -> (Output_Get:Name=None)
   > RegisterTab(TabNameID=@n3, ButtonWidgetType=Get ButtonClass(), TabIndex=@n2)
  ]}
 EventGraph {nodes [
   n1 = Event OnTab Button Removal!() -> (OutputDelegate, then, TabId:Name, TabButton:Common Button Base Object Reference) {is=title K2Node_CustomEvent}
   > RemoveDividerWidget!(execute, self, TabID:Name=@n1.TabId) -> (then)
   > RemoveSpacer!(execute, self, TabID:Name=@n1.TabId) -> (then)
   > n2 = Branch(Condition=IsRebuildingTabList!(self:UIKit Tab Group Object Reference) -> (ReturnValue:bool))
   Remove(execute=@n2.else, TargetMap=Get CachedViewModels(), Key=@n1.TabId)
   n3 = Event PreConstruct()
   > n4 = Sequence!(execute) -> (then_0, then_1, then_2:Exec)
   > CacheInitialDividers() {is=title K2Node_CallFunction}
   n5 = Branch(execute=@n4.then_1, Condition=@n3.IsDesignTime)
   PopulateTabsFromData!(
    execute=[@n6.then, @n5.then]
    self
    Data:Array of UIKit Tab Data Structures=Get InitialTabsData() -> (InitialTabsData:Array of UIKit Tab Data Structures)
   ) -> (then)
   > InitialTabSelection()
   n6 = Branch(execute=@n5.else, Condition=Get UseInitialTabs() -> (UseInitialTabs:bool))
   UpdateTabLayout(execute=@n4.then_2)
   n7 = Event HandleTabCreation!() -> (OutputDelegate, then, TabNameID:Name=None, TabButton:Common Button Base Object Reference) {is=title K2Node_Event}
   > HandleCreatedTab!(execute, self, Tab ID:Name=@n7.TabNameID, Button:Common Button Base Object Reference=@n7.TabButton) -> (then)
   Event Construct()
   > Assign On Tab Button Removal!(execute:Exec, self:Common Tab List Widget Base Object Reference, Delegate:"Delegate (by ref)"=@n1) -> (then:Exec) {is=title K2Node_AssignDelegate}
   Event Destruct()
   > Unbind Event from On Tab Button Removal(self:Common Tab List Widget Base Object Reference, Delegate=@n1) {is=title K2Node_RemoveDelegate}
   > RemoveCachedWidgets()
  ]}
 HandleAddDivider {nodes [
   n1 = HandleAddDivider!() -> (then, TabID:"Name (by ref)", TabIndex:int, AddAfter:bool=~) {is=title K2Node_FunctionEntry}
   n2 = Find(TargetMap:Map of Names to S UI Divider Data Structures=Get CachedDividers(), Key:"Name (by ref)"=@n1) -> (Value:S UI Divider Data Structure)
   > n3 = Branch(Condition=@n2.ReturnValue)
   > n4 = Set CurrentDivider(CurrentDivider:WBP UIKit Divider Base Object Reference=Break S UI Divider Data!(S_UI_DividerData:"S UI Divider Data Structure (by ref)"=@n2) -> (Widget_6_#1:WBP UIKit Divider Base Object Reference, IsAddedAfter_3_#2:bool=~) {is=title K2Node_BreakStruct}) -> (Output_Get:WBP UIKit Divider Base Object Reference)
   > n5 = Branch(Condition=IsValid(Object=@n4))
   n6 = Get CurrentDivider() -> (CurrentDivider:WBP UIKit Divider Base Object Reference)
   AddDividerToGrid!(execute=[@n10.then, @n5.then], self, Divider:WBP UIKit Divider Base Object Reference=@n6, Index:int=@n1.TabIndex, AddAfter:bool=@n1.AddAfter) -> (then) {is=title K2Node_CallFunction}
   > SetupDivider!(execute, self, InputPin:WBP UIKit Divider Base Object Reference=@n6) -> (then)
   n7 = Get DividerClass() -> (DividerClass:WBP UIKit Divider Base Class Reference)
   n8 = Construct Object from Class(
    execute=[@n5.else, @n3.else]
    Class=Select(
     Option 0:WBP UIKit Divider Base Class Reference=&/WBP_UIKit_Divider_Base.WBP_UIKit_Divider_Base_C
     Option 1:WBP UIKit Divider Base Class Reference=@n7
     Index=IsValidClass!(Class:Object Class Reference=@n7) -> (ReturnValue:bool)
    ) -> (ReturnValue:WBP UIKit Divider Base Class Reference)
   ) -> (ReturnValue:WBP UIKit Divider Base Object Reference)
   > n9 = Set CurrentDivider(CurrentDivider=@n8)
   > n10 = Add!(
    execute
    TargetMap=Get CachedDividers()
    Key=Get TabID() -> (TabID=None)
    Value_Widget_6_#1:WBP UIKit Divider Base Object Reference=@n9
    Value_IsAddedAfter_3_#2:bool=@n1.AddAfter
   ) -> (then)
  ]}
 HandleCreatedTab {nodes [
   n1 = HandleCreatedTab!() -> (then, Tab ID:Name, Button:Common Button Base Object Reference) {is=title K2Node_FunctionEntry}
   > n2 = Set CurrentTabIndex(CurrentTabIndex=GetTabIndexByID(TabNameID=@n1) -> (ReturnValue=0)) -> (Output_Get=0)
   > SetupTabButton!(execute, self, Tab ID:Name=@n1, Tab Button:Common Button Base Object Reference=@n1.Button) -> (then)
   > AddSpacerToGrid!(execute, self, TabID:"Name (by ref)"=@n1, Index:int=@n2) -> (then) {is=title K2Node_CallFunction}
   > AddTabToGrid!(execute, self, Widget:Widget Object Reference=@n1.Button, Index:int=@n2) -> (then) {is=title K2Node_CallFunction}
   n3 = Find(TargetMap=Get CachedDividers(), Key=@n1)
   > Branch(Condition=@n3.ReturnValue)
   > HandleAddDivider(
    TabID=@n1
    TabIndex=@n2
    AddAfter=Break S UI Divider Data(S_UI_DividerData=@n3) -> (IsAddedAfter_3_#2=~).IsAddedAfter_3_#2
   ) {is=title K2Node_CallFunction}
  ]}
 InitialTabSelection {nodes [
   InitialTabSelection() {is=title K2Node_FunctionEntry}
   n1 = Get InitialSelectedTabID() -> (InitialSelectedTabID:Name=None)
   n2 = GetTabButtonBaseByID!(self:Common Tab List Widget Base Object Reference, TabNameID:Name=@n1) -> (ReturnValue:Common Button Base Object Reference)
   > n3 = Cast To WBP_UIKit_ButtonGeneric_Base!(execute, Object=@n2) -> (then, CastFailed, AsWBP UIKit Button Generic Base:WBP UIKit Button Generic Base Object Reference) {is=title K2Node_DynamicCast}
   > n4 = RequestInstantSelection(self:WBP UIKit Button Generic Base Object Reference=@n3)
   SelectTabByID!(execute=[@n4.then, @n5.then], self:Common Tab List Widget Base Object Reference, TabNameID:Name=@n1, bSuppressClickFeedback:bool=true) -> (then, ReturnValue:bool)
   n5 = Branch(execute=@n3.CastFailed, Condition=IsValid(Object=@n2))
   n6 = GetTabIdAtIndex!(self:Common Tab List Widget Base Object Reference, Index:int=Get InitialSelectedTabIndex() -> (InitialSelectedTabIndex:int=0)) -> (ReturnValue:Name=None)
   n7 = Cast To WBP_UIKit_ButtonGeneric_Base(execute=@n5.else, Object=GetTabButtonBaseByID(TabNameID=@n6))
   > n8 = RequestInstantSelection(self=@n7)
   SelectTabByID(execute=[@n8.then, @n7.CastFailed], TabNameID=@n6, bSuppressClickFeedback=true)
  ]}
 PopulateTabsFromData {nodes [
   n1 = PopulateTabsFromData!() -> (then, Data:Array of UIKit Tab Data Structures) {is=title K2Node_FunctionEntry}
   > Set CurrentSelection(CurrentSelection:Name=GetSelectedTabId!(self:Common Tab List Widget Base Object Reference) -> (ReturnValue:Name=None)) -> (Output_Get:Name=None)
   > RemoveTabGroupWidgets()
   > n2 = For Each Loop(Array:Array of UIKit Tab Data Structures=@n1) -> (Array Element:UIKit Tab Data Structure)
   n3 = Branch(execute=@n2.Completed, Condition=InEditor!() -> (ReturnValue:bool))
   RestoreTabSelection!(execute=@n3.else, self, TabNameID:Name=Get CurrentSelection() -> (CurrentSelection:Name=None)) -> (then)
   n4 = Break UIKit Tab Data!(UIKitTabData:"UIKit Tab Data Structure (by ref)"=@n2) -> (bEnableTabID:bool=~, TabID:Name, TabViewModel:MVVM Base Viewmodel Object Reference) {is=title K2Node_BreakStruct}
   AddTab!(
    execute=@n2.LoopBody
    self
    Tab VM:MVVM Base Viewmodel Object Reference=@n4.TabViewModel
    Tab ID:Name=SelectName!(A:Name=@n4.TabID, B:Name=None, bPickA:bool=@n4) -> (ReturnValue:Name=None)
    Index:int=@n2.Array Index
   ) -> (then) {is=title K2Node_CallFunction}
  ]}
 RemoveCachedWidgets {nodes [
   RemoveCachedWidgets() {is=title K2Node_FunctionEntry}
   > Clear(TargetMap:Map of Names to Spacer Object References=Get CachedSpacers())
   n1 = Get CachedDividers()
   > n2 = Keys!(execute, TargetMap:Map of Names to S UI Divider Data Structures=@n1) -> (then, Keys:Array of Names)
   > n3 = Values!(execute, TargetMap:Map of Names to S UI Divider Data Structures=@n1) -> (then, Values:Array of S UI Divider Data Structures)
   > n4 = For Each Loop(Array:Array of Names=@n2) -> (Array Element:Name)
   > Add(
    TargetMap=@n1
    Key=@n4
    Value_IsAddedAfter_3_#2="Get (a ref)"!(Array:Array of S UI Divider Data Structures=@n3, Dimension 1=@n4.Array Index) -> (
     Output_Widget_6_#1:WBP UIKit Divider Base Object Reference
     Output_IsAddedAfter_3_#2:bool=~
    ).Output_IsAddedAfter_3_#2
   )
  ]}
 RemoveDivider {nodes [
   n1 = RemoveDivider!() -> (then, TabID:Name) {is=title K2Node_FunctionEntry}
   n2 = Find(TargetMap=Get CachedDividers(), Key=@n1)
   > n3 = Branch(Condition=@n2.ReturnValue)
   > RemoveFromParent(self=Break S UI Divider Data(S_UI_DividerData=@n2) -> (IsAddedAfter_3_#2=~))
   > Remove(TargetMap:Map of Names to S UI Divider Data Structures=Get CachedDividers(), Key=@n1)
   > Return Node!(execute, DividerRemoved:bool=true)
   Return Node(execute=@n3.else, DividerRemoved=false)
  ]}
 RemoveDividerByIndex {nodes [
   n1 = RemoveDividerByIndex!() -> (then, TabIndex:int) {is=title K2Node_FunctionEntry}
   > n2 = RemoveDivider!(execute, self, TabID:Name=GetTabIDbyIndex(Index=@n1) -> (ReturnValue=None)) -> (then, DividerRemoved:bool) {is=title K2Node_CallFunction}
   > Return Node(DividerRemoved=@n2)
  ]}
 RemoveDividerWidget {nodes [
   n1 = RemoveDividerWidget!() -> (then, TabID:Name) {is=title K2Node_FunctionEntry}
   n2 = Find(TargetMap=Get CachedDividers(), Key=@n1)
   > Branch(Condition=@n2.ReturnValue)
   n3 = Break S UI Divider Data(S_UI_DividerData=@n2) -> (IsAddedAfter_3_#2=~)
   > RemoveFromParent(self=@n3)
   > Add!(
    execute
    TargetMap=Get CachedDividers()
    Key=@n1
    Value:"S UI Divider Data Structure (by ref)"=Make S UI Divider Data(IsAddedAfter_3_#2=@n3.IsAddedAfter_3_#2)
   ) -> (then)
  ]}
 RemoveSpacer {nodes [
   n1 = RemoveSpacer!() -> (then, TabID:Name) {is=title K2Node_FunctionEntry}
   n2 = Find(TargetMap:Map of Names to Spacer Object References=Get CachedSpacers(), Key=@n1) -> (Value:Spacer Object Reference)
   > Branch(Condition=@n2.ReturnValue)
   > RemoveFromParent(self=@n2)
   > Remove(TargetMap:Map of Names to Spacer Object References=Get CachedSpacers(), Key=@n1)
  ]}
 RemoveTabByIndex {nodes [
   n1 = RemoveTabByIndex!() -> (then, Index:int) {is=title K2Node_FunctionEntry}
   > n2 = RemoveTab!(execute, self:Common Tab List Widget Base Object Reference, TabNameID:Name=GetTabIDbyIndex(Index=@n1) -> (ReturnValue=None)) -> (then, ReturnValue:bool)
   > Return Node!(execute, TabRemoved:bool=@n2)
  ]}
 RemoveTabGroupWidgets {nodes [
   RemoveTabGroupWidgets() {is=title K2Node_FunctionEntry}
   > RemoveAllTabs(self:Common Tab List Widget Base Object Reference)
   > ClearChildren(self=Get TabsContainer())
   > RemoveCachedWidgets() {is=title K2Node_CallFunction}
  ]}
 RemoveTabs {nodes [
   RemoveTabs() {is=title K2Node_FunctionEntry}
   > Clear(TargetMap:Map of Names to MVVM Base Viewmodel Object References=Get CachedViewModels())
   > RemoveTabGroupWidgets() {is=title K2Node_CallFunction}
  ]}
 RestoreTabSelection {nodes [
   n1 = RestoreTabSelection!() -> (then, TabNameID:Name) {is=title K2Node_FunctionEntry}
   > n2 = Cast To WBP_UIKit_ButtonGeneric_Base(Object=GetTabButtonBaseByID(TabNameID=@n1))
   > n3 = RequestInstantSelection(self=@n2)
   SelectTabByID(execute=[@n3.then, @n2.CastFailed], TabNameID=@n1, bSuppressClickFeedback=true)
  ]}
 SetTabViewModel {nodes [
   n1 = SetTabViewModel!() -> (then, TabID:Name, VM:Notify Field Value Changed Interface) {is=title K2Node_FunctionEntry}
   n2 = Get View From User Widget!(
    self:Viewmodel Engine Subsytem Object Reference=Get MVVMSubsystem() -> (ReturnValue:Viewmodel Engine Subsytem Object Reference) {is=title K2Node_GetEngineSubsystem}
    UserWidget:User Widget Object Reference=GetTabButtonBaseByID(TabNameID=@n1)
   ) -> (ReturnValue:MVVM View Object Reference) {is=title K2Node_CallFunction}
   > Branch(Condition=IsValid(Object=@n2))
   > SetViewModelByClass!(execute, self:MVVM View Object Reference=@n2, NewValue:Notify Field Value Changed Interface=@n1.VM) -> (then, ReturnValue:bool)
  ]}
 SetupDivider {nodes [
   n1 = SetupDivider!() -> (then, InputPin:WBP UIKit Divider Base Object Reference) {is=title K2Node_FunctionEntry}
   > SetDividerOrientation!(execute, self:WBP UIKit Divider Base Object Reference=@n1, IsVertical:bool=Get IsVertical()) -> (then)
   > SetDividerSpacing!(
    execute
    self:WBP UIKit Divider Base Object Reference=@n1
    Spacing:double=Get DividerSpacing() -> (DividerSpacing:double)
    IsVertical:bool=Get IsVertical()
   ) -> (then)
   > SetPadding!(execute, self:User Widget Object Reference=@n1, InPadding:Margin Structure=Get DividerPadding() -> (DividerPadding:Margin Structure)) -> (then)
  ]}
 SetupTabButton {nodes [
   n1 = SetupTabButton!() -> (then, Tab ID:Name, Tab Button:Common Button Base Object Reference) {is=title K2Node_FunctionEntry}
   > n2 = Sequence!(execute) -> (then_0, then_1)
   > SetStyle!(
    execute
    self:Common Button Base Object Reference=@n1.Tab Button
    InStyle:Common Button Style Class Reference=/FortUILibrary/UIKit/Buttons/Blocks/ButtonStyle_UIKit_Empty.ButtonStyle_UIKit_Empty_C
   ) -> (then)
   > SetIsFocusable!(
    execute
    self:Common Button Base Object Reference=@n1.Tab Button
    bInIsFocusable:bool=Get IsNaturalNavigation() -> (IsNaturalNavigation:bool)
   ) -> (then)
   n3 = Cast To WBP_UIKit_ButtonGeneric_Base(execute=@n2.then_1, Object=@n1.Tab Button)
   > SetIconSize!(execute, self:WBP UIKit Button Generic Base Object Reference=@n3, Icon:Vector 2D Structure=Get IconSize() -> (IconSize:Vector 2D Structure)) -> (then)
   n4 = Get!() -> (MinButtonSize_X:int, MinButtonSize_Y:int)
   > SetButtonMinSize!(execute, self:WBP UIKit Button Generic Base Object Reference=@n3, Min Width:int=@n4, Min Height:int=@n4.MinButtonSize_Y) -> (then)
   > Branch(Condition=InEditor())
   > SetPreviewData!(
    execute
    self:WBP UIKit Button Generic Base Object Reference=@n3
    VM:MVVM Base Viewmodel Object Reference=GetTabViewModel!(self:UIKit Tab Group Object Reference, TabID:Name=Get Tab ID() -> (Tab ID:Name=None)) -> (ReturnValue:MVVM Base Viewmodel Object Reference)
   ) -> (then)
  ]}
 UpdateTabLayout {nodes [
   UpdateTabLayout() {is=title K2Node_FunctionEntry}
   > n1 = Branch(Condition=Get IsVertical())
   n2 = Get TabsContainer()
   > SetColumnFill!(execute, self:Grid Panel Object Reference=@n2, ColumnIndex:int=0, Coefficient:float=1) -> (then)
   > n3 = SetRowFill!(execute, self:Grid Panel Object Reference=@n2, RowIndex:int=0, Coefficient:float=0) -> (then)
   n4 = Get TabsContainer()
   SetGridSlotAlignment!(
    execute=[@n5.then, @n3.then]
    Widget:Widget Object Reference=@n4
    Horizontal:EHorizontalAlignment Enum=Get TabGroupHorizontalAlignment() -> (TabGroupHorizontalAlignment:EHorizontalAlignment Enum=HAlign_Fill)
    Vertical:EVerticalAlignment Enum=Get TabGroupVerticalAlignment() -> (TabGroupVerticalAlignment:EVerticalAlignment Enum=VAlign_Fill)
   ) -> (then)
   > SetPadding(
    self:Scroll Box Slot Object Reference=SlotAsScrollBoxSlot!(Widget:Widget Object Reference=@n4) -> (ReturnValue:Scroll Box Slot Object Reference)
    InPadding=Get TabGroupPadding() -> (TabGroupPadding:Margin Structure)
   )
   SetRowFill(execute=@n1.else, self=@n2, RowIndex=0, Coefficient=1)
   > n5 = SetColumnFill(self=@n2, ColumnIndex=0, Coefficient=0)
  ]}
]
variables [
 ButtonClass {type=Class, default=/FortUILibrary/UIKit/Buttons/TabButtons/WBP_UIKit_Button_Tab.WBP_UIKit_Button_Tab_C}
 ButtonHorizontalAlignment {type=EHorizontalAlignment, default=HAlign_Fill}
 ButtonPadding {type=Margin, default=<ltrb 0 0 0 0>}
 ButtonVerticalAlignment {type=EVerticalAlignment, default=VAlign_Top}
 CachedDividers {type="map<S_UI_DividerData>", default {}}
 CachedSpacers {type="map<Spacer>", default {}}
 DividerClass {type=Class, default=&WBP_UIKit_Divider_Base_C}
 DividerPadding {type=Margin, default=<ltrb 0 0 0 0>}
 Dividers {type="array<S_UI_DividerPosition>", default []}
 DividerSpacing {type=number, default=24}
 IconSize {type=Vector2D, default=<32 32>}
 InitialSelectedTabID {type=string, default=None}
 InitialSelectedTabIndex {type=integer, default=-1}
 InitialTabsData {type="array<UIKitTabData>", default []}
 IsNaturalNavigation {type=boolean, default=true}
 IsVertical {type=boolean, default=true}
 MinButtonSize {type=IntPoint, default=<0 0>}
 TabGroupHorizontalAlignment {type=EHorizontalAlignment, default=HAlign_Fill}
 TabGroupPadding {type=Margin, default=<ltrb 8 8 8 8>}
 TabGroupVerticalAlignment {type=EVerticalAlignment, default=VAlign_Fill}
 TabSpacing {type=number, default=30}
 UseInitialTabs {type=boolean}
]
widgets {
 info {inheritedWidgetCount=0, parentClass=/Script/FortniteUI.UIKitTabGroup, rootWidgetClass=&GridPanel, widgetCount=1}
 widgets [TabsContainer {widgetClassPath=&GridPanel, bIsVariable=true, widget=TabsContainer, properties {columnFill [1], rowFill [0]}}]
}
