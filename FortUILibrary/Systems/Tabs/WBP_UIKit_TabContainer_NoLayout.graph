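# Layout-free base widget for an expandable UI Kit tab container, pairing a parent tab button with a nested vertical tab group.
# ConstructTabGroup creates the inner group on PreConstruct, and its tab events are re-broadcast tagged with the container's ID.
# Subclasses supply slots, arrow and animations through Override_ functions.
# Expand and collapse state drives animation, style tags, hit-testing and gamepad focus, which briefly locks navigation during animations.
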WidgetBlueprint /FortUILibrary/Systems/Tabs/WBP_UIKit_TabContainer_NoLayout
graphs [
 AddTabToContainer {nodes [
   n1 = AddTabToContainer!() -> (then, TabViewModel:MVVM Base Viewmodel Object Reference, TabID:Name, Index:int=-1) {is=title K2Node_FunctionEntry}
   n2 = Get InnerTabGroup() -> (InnerTabGroup:WBP UIKit Tab Group Compact Base Object Reference)
   > Is Valid(InputObject=@n2)
   > AddTab!(
    execute
    self:WBP UIKit Tab Group Compact Base Object Reference=@n2
    Tab VM:MVVM Base Viewmodel Object Reference=@n1
    Tab ID:Name=@n1.TabID
    Index:int=@n1.Index
   ) -> (then)
  ]}
 ClearInnerTabsSelection {nodes [
   ClearInnerTabsSelection() {is=title K2Node_FunctionEntry}
   > n1 = Get!(execute:Exec) -> (then:Exec, else:Exec, InnerTabGroup:WBP UIKit Tab Group Compact Base Object Reference)
   > n2 = Sequence()
   DeselectAll(execute=@n2.then_1, self:UIKit Tab Group Object Reference=@n1)
   > RefreshTabSlotVisuals()
   n3 = Cast To UIKitModularButton!(
    execute=@n2.then_0
    Object=GetTabButtonBaseByID!(
     self:Common Tab List Widget Base Object Reference=@n1
     TabNameID:Name=GetSelectedTabId!(self:Common Tab List Widget Base Object Reference=@n1) -> (ReturnValue:Name=None)
    ) -> (ReturnValue:Common Button Base Object Reference)
   ) -> (then, CastFailed, AsUIKit Modular Button:UIKit Modular Button Object Reference) {is=title K2Node_DynamicCast}
   > RequestInstantTransitionOnBlocks!(execute, self:UIKit Modular Button Object Reference=@n3, State:EUIKitBlockInstantTransitionState Enum=InstantDeselected) -> (then)
  ]}
 ConstructTabGroup {nodes [
   ConstructTabGroup() {is=title K2Node_FunctionEntry}
   > n1 = Sequence()
   n2 = Get InnerTabGroup()
   > Is Valid(InputObject=@n2)
   > Unbind all Events from On UIKit Tab Clicked!(execute:Exec, self:UIKit Tab Group Object Reference=@n2) -> (then:Exec) {is=title K2Node_ClearDelegate}
   > Unbind all Events from On UIKit Tab Selected!(execute:Exec, self:UIKit Tab Group Object Reference=@n2) -> (then:Exec) {is=title K2Node_ClearDelegate}
   > RemoveFromParent(self=@n2)
   n3 = Create Widget!(
    execute:Exec=@n1.then_1
    Class:User Widget Class Reference=Get ContainerTabgroupClass() -> (ContainerTabgroupClass:WBP UIKit Tab Group Compact Base Class Reference)
    OwningPlayer:Player Controller Object Reference=GetOwningPlayer()
    bAutoListenForInput:bool=False
    bShouldWrapNavigation:bool=True
   ) -> (then:Exec, ReturnValue:WBP UIKit Tab Group Compact Base Object Reference)
   > n4 = Set InnerTabGroup(InnerTabGroup:WBP UIKit Tab Group Compact Base Object Reference=@n3) -> (Output_Get:WBP UIKit Tab Group Compact Base Object Reference)
   > Set IsVertical!(execute, IsVertical:bool=true, self:WBP UIKit Tab Group Compact Base Object Reference=@n4) -> (then, Output_Get:bool)
   > Clear(TargetArray:Array of UIKit Tab Data Structures=Get InitialTabsData!(self:WBP UIKit Tab Group Compact Base Object Reference=@n4) -> (InitialTabsData:Array of UIKit Tab Data Structures))
   > Bind Event to On UIKit Tab Selected(self:UIKit Tab Group Object Reference=@n4, Delegate=Create Event()) {is=title K2Node_AddDelegate}
   > Bind Event to On UIKit Tab Clicked(self:UIKit Tab Group Object Reference=@n4, Delegate=Create Event()) {is=title K2Node_AddDelegate}
   > Bind Event to On Tab Button Creation(self:Common Tab List Widget Base Object Reference=@n4, Delegate=Create Event()) {is=title K2Node_AddDelegate}
   > Call On Container Tab Group Creation!(execute:Exec, self:WBP UIKit Tab Container No Layout Object Reference, TabGroupWidget:WBP UIKit Tab Group Compact Base Object Reference=@n4) -> (then:Exec) {is=title K2Node_CallDelegate}
   n5 = Override_GetContentContainer!(self) -> (Container:Panel Widget Object Reference)
   > ClearChildren(self=@n5)
   > n6 = AddChild(self=@n5, Content=Get InnerTabGroup())
   > n7 = Cast To BorderSlot!(execute, Object=@n6) -> (then, CastFailed, AsBorder Slot:Border Slot Object Reference) {is=title K2Node_DynamicCast}
   > Set HorizontalAlignment!(execute, HorizontalAlignment:EHorizontalAlignment Enum=HAlign_Fill, self:Border Slot Object Reference=@n7) -> (then, Output_Get:EHorizontalAlignment Enum=HAlign_Fill)
   > Set VerticalAlignment!(execute, VerticalAlignment:EVerticalAlignment Enum=VAlign_Fill, self:Border Slot Object Reference=@n7) -> (then, Output_Get:EVerticalAlignment Enum=VAlign_Fill)
   > Set Padding!(
    execute
    Padding:Margin Structure=Get InnerTabGroupPadding() -> (InnerTabGroupPadding:Margin Structure)
    self:Border Slot Object Reference=@n7
   ) -> (then, Output_Get:Margin Structure)
  ]}
 DestroyTabGroupWidget {nodes [
   DestroyTabGroupWidget() {is=title K2Node_FunctionEntry}
   n1 = Get InnerTabGroup()
   > n2 = Is Valid(InputObject=@n1)
   > n3 = RemoveFromParent(self=@n1)
   Set InnerTabGroup(execute=[@n3.then, @n2.Is Not Valid])
  ]}
 EventGraph {nodes [
   n1 = Event OnTabButtonSelected!() -> (OutputDelegate, then, Button:Common Button Base Object Reference, Selected:bool=~) {is=title K2Node_CustomEvent}
   > SetArrowIsSelected!(execute, self, IsSelected:bool=@n1.Selected) -> (then)
   n2 = Event OnInnerTabClicked!() -> (OutputDelegate, then, TabID:Name, TabVM:MVVM Base Viewmodel Object Reference) {is=title K2Node_CustomEvent}
   > Call On Inner Tab Clicked!(
    execute:Exec
    self:WBP UIKit Tab Container No Layout Object Reference
    TabID:Name=Get ContainerTabID() -> (ContainerTabID:Name=None)
    InnerTabID:Name=@n2.TabID
    TabViewModel:MVVM Base Viewmodel Object Reference=@n2.TabVM
   ) -> (then:Exec) {is=title K2Node_CallDelegate}
   n3 = Event OnInnerTabSelected!() -> (OutputDelegate, then, TabID:Name, TabVM:MVVM Base Viewmodel Object Reference) {is=title K2Node_CustomEvent}
   > Call On Inner Tab Selected!(
    execute:Exec
    self:WBP UIKit Tab Container No Layout Object Reference
    TabID:Name=Get ContainerTabID() -> (ContainerTabID=None)
    InnerTabID:Name=@n3.TabID
    TabViewModel:MVVM Base Viewmodel Object Reference=@n3.TabVM
   ) -> (then:Exec) {is=title K2Node_CallDelegate}
   n4 = Event OnInnerTabCreation!() -> (OutputDelegate, then, TabId:Name, TabButton:Common Button Base Object Reference) {is=title K2Node_CustomEvent}
   > Call On Inner Tab Creation!(
    execute:Exec
    self:WBP UIKit Tab Container No Layout Object Reference
    TabID:Name=Get ContainerTabID() -> (ContainerTabID=None)
    InnerTabID:Name=@n4.TabId
    TabButton:Common Button Base Object Reference=@n4.TabButton
   ) -> (then:Exec) {is=title K2Node_CallDelegate}
   TryFocusSelectedInnerTab() {is=title K2Node_CustomEvent}
   > n5 = Switch on ECommonInputType!(execute, Selection:ECommonInputType Enum=GetCurrentInputType(self=Get CommonInputSubsystem())) -> (MouseAndKeyboard:Exec, Gamepad:Exec, Touch:Exec, Count:Exec) {is=title K2Node_SwitchEnum}
   Branch(execute=@n5.Gamepad, Condition=NOT Boolean(A=Get AllowContainerInteractionOnGamepad() -> (AllowContainerInteractionOnGamepad:bool)))
   > n6 = Branch(Condition=IsAnyAnimationPlaying!(self:User Widget Object Reference) -> (ReturnValue:bool))
   > SetAllNavigationRules!(execute, self:Widget Object Reference, Rule:EUINavigationRule Enum=Stop, WidgetToFocus:Name=None) -> (then)
   > Delay(Duration=Get ExpandDuration() -> (ExpandDuration:float))
   > n7 = SetAllNavigationRules(Rule=Escape, WidgetToFocus=None)
   FocusSelectedInnerTab(execute=[@n7.then, @n6.else])
   n8 = Event PreConstruct()
   > Parent: PreConstruct(IsDesignTime=@n8.IsDesignTime)
   > ConstructTabGroup() {is=title K2Node_CallFunction}
   > SetIsExpanded!(execute, self, InIsExpanded:bool=Get IsExpanded() -> (IsExpanded:bool), SkipAnim:bool=true) -> (then)
   n9 = Event OnMouseEnter()
   RefreshTabSlotVisuals(execute=[@n9.then, @n11.then])
   n10 = Event SetIsCollapsed!() -> (OutputDelegate, then, IsCollapsed:bool, SkipAnimation:bool) {is=title K2Node_Event}
   > HandleCollapsedStateChanged!(execute, self, IsCollapsed:bool=@n10.IsCollapsed, SkipAnimation:bool=@n10.SkipAnimation) -> (then)
   Event Destruct()
   > Parent: Destruct()
   > Unbind all Events from On Inner Tab Selected!(execute:Exec, self:WBP UIKit Tab Container No Layout Object Reference) -> (then:Exec) {is=title K2Node_ClearDelegate}
   > Unbind all Events from On Container Tab Group Creation!(execute:Exec, self:WBP UIKit Tab Container No Layout Object Reference) -> (then:Exec) {is=title K2Node_ClearDelegate}
   > DestroyTabGroupWidget() {is=title K2Node_CallFunction}
   n11 = Event OnMouseLeave!() -> (OutputDelegate, then, MouseEvent:"Pointer Event Structure (by ref)") {is=title K2Node_Event}
  ]}
 ExpandToSelectedItem {nodes [
   ExpandToSelectedItem() {is=title K2Node_FunctionEntry}
   > SetIsExpanded(InIsExpanded=true, SkipAnim=Get SkipExpandAnimation() -> (SkipExpandAnimation:bool))
   > TryFocusSelectedInnerTab() {is=title K2Node_CallFunction}
  ]}
 FocusSelectedInnerTab {nodes [
   FocusSelectedInnerTab() {is=title K2Node_FunctionEntry}
   > Is Valid(InputObject=GetTabButton!(self) -> (TabButton:Common Button Base Object Reference))
   > Branch(Condition=HasAnyUserFocusedDescendants!(self:Widget Object Reference=Override_GetTabSlot!(self) -> (Slot:Named Slot Object Reference)) -> (ReturnValue:bool))
   n1 = Get InnerTabGroup()
   > n2 = Set SelectedTab(SelectedTab:Common Button Base Object Reference=GetTabButtonBaseByID(self=@n1, TabNameID=GetSelectedTabId(self=@n1) -> (ReturnValue=None))) -> (Output_Get:Common Button Base Object Reference)
   > Is Valid(InputObject=@n2)
   > SetFocus(self=@n2)
  ]}
 GetInnerTabGroupWidget {nodes [
   GetInnerTabGroupWidget() {is=title K2Node_FunctionEntry}
   > Return Node!(execute, InnerTabGroupWidget:WBP UIKit Tab Group Compact Base Object Reference=Get InnerTabGroup())
  ]}
 GetRemainingAnimTime {nodes [
   GetRemainingAnimTime() {is=title K2Node_FunctionEntry}
   > Return Node!(execute, FunctionIsImplemented:bool=false, RemainingTime:double=0, TotalTime:double=0)
  ]}
 GetTabButton {nodes [
   GetTabButton() {is=title K2Node_FunctionEntry}
   n1 = Override_GetTabSlot()
   > Return Node!(
    execute
    TabButton:Common Button Base Object Reference=Cast To CommonButtonBase!(Object=GetChildAt!(self:Panel Widget Object Reference=@n1, Index:int=0) -> (ReturnValue:Widget Object Reference)) -> (AsCommon Button Base:Common Button Base Object Reference, bSuccess:bool=~) {is=title K2Node_DynamicCast}
   )
  ]}
 HandleCollapsedStateChanged {nodes [
   HandleCollapsedStateChanged!() -> (then, IsCollapsed:bool=~, SkipAnimation:bool=~) {is=title K2Node_FunctionEntry}
   > n1 = Sequence!(execute) -> (then_0, then_1, then_2:Exec, then_3:Exec)
   n2 = Get InnerTabGroup()
   n3 = For Each Loop(
    Exec=@n1.then_3
    Array:Array of Widget Object References=GetAllChildren(self=Get TabsContainer!(self:WBP UIKit Tab Group Compact Base Object Reference=@n2) -> (TabsContainer:Grid Panel Object Reference))
   ) -> (Array Element:Widget Object Reference)
   > n4 = Cast To BPI_UIKit_Collapsible!(execute, Object=@n3) -> (then, CastFailed, AsBPI UIKit Collapsible:BPI UIKit Collapsible Interface) {is=title K2Node_DynamicCast}
   > SetIsCollapsed!(
    execute
    self:BPI UIKit Collapsible Interface=@n4
    IsCollapsed:bool=Get IsCollapsed() -> (IsCollapsed:bool)
    SkipAnimation:bool=Get SkipAnimation() -> (SkipAnimation:bool)
   ) -> (then)
   n5 = Cast To BPI_UIKit_Collapsible(execute=@n1.then_2, Object=GetTabButton() {is=title K2Node_CallFunction})
   > SetIsCollapsed(self=@n5, IsCollapsed=Get IsCollapsed(), SkipAnimation=Get SkipAnimation())
   n6 = Branch(execute=@n1.then_1, Condition=Get SkipAnimation())
   > SetArrowVisibilty!(execute, self, IsVisible:bool=NOT Boolean(A=Get IsCollapsed())) -> (then)
   n7 = Override_GetHideArrowAnimation!(self) -> (Animation:Widget Animation Object Reference)
   Is Valid(exec=@n6.else, InputObject=@n7)
   > n8 = Branch(Condition=Get IsCollapsed())
   > PlayAnimationForward(InAnimation=@n7)
   PlayAnimationReverse(execute=@n8.else, InAnimation=@n7)
   n9 = GetSelectedTabId(self=GetInnerTabGroupWidget!(self) -> (InnerTabGroupWidget:WBP UIKit Tab Group Compact Base Object Reference) {is=title K2Node_CallFunction}) -> (ReturnValue=None)
   SetIsExpanded(
    execute=@n1.then_0
    InIsExpanded=Select(Option 0:bool="Not Equal (Name)"(A=@n9) -> (ReturnValue=~), Option 1:bool=false, Index=Get IsCollapsed()) -> (ReturnValue:bool)
    SkipAnim=Get SkipAnimation()
   )
  ]}
 InitContainer {nodes [
   n1 = InitContainer!() -> (then, ParentTabButton:Common Button Base Object Reference, TabID:Name) {is=title K2Node_FunctionEntry}
   > n2 = Branch(Condition="Not Equal (Name)"(A=@n1.TabID) -> (ReturnValue=~))
   > Set ContainerTabID(ContainerTabID:Name=@n1.TabID) -> (Output_Get:Name=None)
   n3 = Get ParentTabButton() -> (ParentTabButton:Common Button Base Object Reference)
   > AddChild(self=Override_GetTabSlot(), Content=@n3)
   > RefreshTabSlotEnabledState()
   > Bind Event to On Selected Changed(self:Common Button Base Object Reference=@n3, Delegate=Create Event()) {is=title K2Node_AddDelegate}
   n4 = Self-Reference()
   PrintString!(
    execute=@n2.else
    InString=Append!(A=GetDisplayName(Object=@n4), B=.InitContainer:, C:String=" - invalid TabID") -> (ReturnValue)
    bPrintToScreen
    bPrintToLog
    TextColor
    Duration
    Key
   ) -> (then)
  ]}
 OnContainerExpandedStateChanged {nodes [OnContainerExpandedStateChanged!() -> (then, TabID:Name, IsExpanded:bool=~) {is=title K2Node_FunctionEntry}]}
 OnContainerTabGroupCreation {nodes [OnContainerTabGroupCreation!() -> (then, TabGroupWidget:WBP UIKit Tab Group Compact Base Object Reference) {is=title K2Node_FunctionEntry}]}
 OnInnerTabClicked {nodes [OnInnerTabClicked!() -> (then, TabID:Name, InnerTabID:Name, TabViewModel:MVVM Base Viewmodel Object Reference) {is=title K2Node_FunctionEntry}]}
 OnInnerTabCreation {nodes [OnInnerTabCreation!() -> (then, TabID:Name, InnerTabID:Name, TabButton:Common Button Base Object Reference) {is=title K2Node_FunctionEntry}]}
 OnInnerTabSelected {nodes [OnInnerTabSelected!() -> (then, TabID:Name, InnerTabID:Name, TabViewModel:MVVM Base Viewmodel Object Reference) {is=title K2Node_FunctionEntry}]}
 Override_GetArrow {nodes [
   Override_GetArrow() {is=title K2Node_FunctionEntry}
   > Return Node!(execute, Image_Arrow:Widget Object Reference)
  ]}
 Override_GetCollapseAnimation {nodes [
   Override_GetCollapseAnimation() {is=title K2Node_FunctionEntry}
   > Return Node!(execute, Animation:Widget Animation Object Reference)
  ]}
 Override_GetContentContainer {nodes [
   Override_GetContentContainer() {is=title K2Node_FunctionEntry}
   > Return Node!(execute, Container:Panel Widget Object Reference)
  ]}
 Override_GetHideArrowAnimation {nodes [
   Override_GetHideArrowAnimation() {is=title K2Node_FunctionEntry}
   > Return Node!(execute, Animation:Widget Animation Object Reference)
  ]}
 Override_GetTabSlot {nodes [
   Override_GetTabSlot() {is=title K2Node_FunctionEntry}
   > Return Node!(execute, Slot:Named Slot Object Reference)
  ]}
 Override_SetInstantTransition {nodes [Override_SetInstantTransition!() -> (then, IsExpanded:bool=~) {is=title K2Node_FunctionEntry}]}
 RefreshTabSlotEnabledState {nodes [
   RefreshTabSlotEnabledState() {is=title K2Node_FunctionEntry}
   n1 = Get CommonInputSubsystem()
   > Is Valid(InputObject=@n1)
   > n2 = Switch on ECommonInputType(Selection=GetCurrentInputType(self=@n1))
   n3 = Branch(execute=@n2.Gamepad, Condition=Get AllowContainerInteractionOnGamepad())
   n4 = Override_GetTabSlot() {is=title K2Node_CallFunction}
   Set bIsEnabled!(execute=[@n3.then, @n2.MouseAndKeyboard, @n2.Touch], bIsEnabled:bool=true, self:Widget Object Reference=@n4) -> (then, Output_Get:bool)
   > Set Visibility(Visibility=SelfHitTestInvisible, self=@n4)
   n5 = Override_GetTabSlot()
   n6 = Get IsExpanded()
   Set bIsEnabled(execute=@n3.else, bIsEnabled=NOT Boolean(A=@n6), self=@n5)
   > Set Visibility(
    Visibility=Select(Option 0:ESlateVisibility Enum=SelfHitTestInvisible, Option 1:ESlateVisibility Enum=HitTestInvisible, Index=@n6) -> (ReturnValue:ESlateVisibility Enum)
    self=@n5
   )
  ]}
 RefreshTabSlotVisuals {nodes [
   RefreshTabSlotVisuals() {is=title K2Node_FunctionEntry}
   > Set TabButton(TabButton:Common Button Base Object Reference=GetTabButton()) -> (Output_Get:Common Button Base Object Reference)
   > n1 = Get!(execute) -> (then, else, TabButton:Common Button Base Object Reference)
   n2 = Make Array("[0]"=(TagName="StyleSheet.Style.TabExpanded"))
   n3 = Make Array!() -> (Array)
   n4 = Get IsExpanded()
   > UpdateWidgetStyleTags!(
    execute
    InWidget:Widget Object Reference=@n1
    InAddedTags:Array of Gameplay Tag Structures=Select(Option 0:Array of Gameplay Tag Structures=@n3, Option 1:Array of Gameplay Tag Structures=@n2, Index=@n4) -> (ReturnValue:Array of Gameplay Tag Structures)
    InRemovedTags:Array of Gameplay Tag Structures=Select(Option 0=@n2, Option 1=@n3, Index=@n4)
   ) -> (then, ReturnValue:bool)
   > SetArrowIsSelected(IsSelected=GetSelected!(self:Common Button Base Object Reference=@n1) -> (ReturnValue:bool))
  ]}
 SetArrowIsSelected {nodes [SetArrowIsSelected!() -> (then, IsSelected:bool=~) {is=title K2Node_FunctionEntry}]}
 SetArrowVisibilty {nodes [
   n1 = SetArrowVisibilty!() -> (then, IsVisible:bool=~) {is=title K2Node_FunctionEntry}
   n2 = Override_GetArrow!(self) -> (Image_Arrow:Widget Object Reference) {is=title K2Node_CallFunction}
   > Is Valid(InputObject=@n2)
   > Set Visibility(
    Visibility=Select(Option 0:ESlateVisibility Enum=Collapsed, Option 1:ESlateVisibility Enum=SelfHitTestInvisible, Index=@n1) -> (ReturnValue:ESlateVisibility Enum)
    self=@n2
   )
  ]}
 SetBackground {nodes [
   n1 = SetBackground!() -> (then, InBrush:"Slate Brush Structure (by ref)") {is=title K2Node_FunctionEntry}
   > n2 = Cast To Border!(execute, Object=Override_GetContentContainer() {is=title K2Node_CallFunction}) -> (then, CastFailed, AsBorder:Border Object Reference) {is=title K2Node_DynamicCast}
   > SetBrush(self:Border Object Reference=@n2, InBrush=@n1)
  ]}
 SetIsExpanded {nodes [
   n1 = SetIsExpanded!() -> (then, InIsExpanded:bool=~, SkipAnim:bool=true) {is=title K2Node_FunctionEntry}
   > n2 = Sequence!(execute) -> (then_0, then_1, then_2)
   > n3 = Branch(Condition=NOT Boolean(A=Get SkipAnim() -> (SkipAnim:bool)))
   > Branch(Condition="Not Equal (Boolean)"(A=Get InIsExpanded() -> (InIsExpanded:bool), B=Get IsExpanded()) -> (ReturnValue=~))
   > PlayAnimationDuringTime!(
    execute
    Widget:User Widget Object Reference=Self-Reference()
    Animation:Widget Animation Object Reference=Override_GetCollapseAnimation!(self) -> (Animation:Widget Animation Object Reference) {is=title K2Node_CallFunction}
    PlayTime:float=Get ExpandDuration()
    bPlayForward:bool=NOT Boolean(A=Get InIsExpanded())
   ) -> (then)
   Override_SetInstantTransition!(execute=@n3.else, self, IsExpanded:bool=Get InIsExpanded()) -> (then) {is=title K2Node_CallFunction}
   Branch(execute=@n2.then_1, Condition="Not Equal (Boolean)"(A=Get IsExpanded(), B=@n1) -> (ReturnValue=~))
   > n4 = Set IsExpanded(IsExpanded:bool=@n1) -> (Output_Get:bool)
   > Call On Container Expanded State Changed!(
    execute:Exec
    self:WBP UIKit Tab Container No Layout Object Reference
    TabID:Name=Get ContainerTabID() -> (ContainerTabID=None)
    IsExpanded:bool=@n4
   ) -> (then:Exec) {is=title K2Node_CallDelegate}
   RefreshTabSlotVisuals(execute=@n2.then_2) {is=title K2Node_CallFunction}
   > RefreshTabSlotEnabledState() {is=title K2Node_CallFunction}
   n5 = Get IsExpanded()
   > Set Visibility(Visibility=Select(Option 0=SelfHitTestInvisible, Option 1=HitTestInvisible, Index=NOT Boolean(A=@n5)), self=Override_GetContentContainer())
  ]}
]
variables [
 AllowContainerInteractionOnGamepad {type=boolean}
 ContainerTabgroupClass {type=Class, default=&/WBP_UIKit_TabGroupCompact_Base.WBP_UIKit_TabGroupCompact_Base_C}
 ContainerTabID {type=string, default=None}
 ExpandDuration {type=number, default=0.25}
 InnerTabGroup {type=WBP_UIKit_TabGroupCompact_Base_C, default=null}
 InnerTabGroupPadding {type=Margin, default=<ltrb 0 5 0 0>}
 IsExpanded {type=boolean}
 OnContainerExpandedStateChanged {}
 OnContainerTabGroupCreation {}
 OnInnerTabClicked {}
 OnInnerTabCreation {}
 OnInnerTabSelected {}
 SkipExpandAnimation {type=boolean, default=true}
]
widgets {info {inheritedWidgetCount=0, parentClass=&UserWidget, rootWidgetClass=null, widgetCount=0}, widgets []}
