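# Base UI Kit tab container widget that adds a collapsible layout to WBP_UIKit_TabContainer_NoLayout: backplate, tab slot, expand arrow and clipped content border.
# Overrides hand these widgets and the Collapse and HideArrow animations to the parent.
# Collapse scales content height and rotates the arrow's material, with a repeater event track alongside Update Border Opacity, which derives backplate alpha from playback progress.
# SetInstantTransition applies expanded or collapsed visuals without animating.
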WidgetBlueprint /FortUILibrary/Systems/Tabs/WBP_UIKit_TabContainer_Base
animations [
 Collapse {
  binding_count=3
  length_seconds=0.5
  object=&self:Collapse
  sequence {
   bindings [
    Border_Background {
     id=#1
     possessed_class=Border
     tracks [{type=MovieSceneEventTrack, sections [{type=MovieSceneEventRepeaterSection, end_frame=50, start_frame=0}]}]
    }
    Image_Arrow {
     id=#2
     possessed_class=Image
     tracks [{
       type=MovieSceneWidgetMaterialTrack
       sections [{
         type=MovieSceneParameterSection
         end_frame=50
         start_frame=0
         channels [Rotation {type=MovieSceneScriptingFloatChannel, keys [[0,0.25],[50,0.75]]}]
        }]
      }]
    }
    Border_Content {
     id=#3
     possessed_class=Border
     tracks [{
       type=MovieSceneDoubleVectorTrack
       display_name=Desired Size Scale
       sections [{
         type=MovieSceneDoubleVectorSection
         end_frame=50
         start_frame=0
         channels [
          Vector.X {type=MovieSceneScriptingDoubleChannel}
          Vector.Y {type=MovieSceneScriptingDoubleChannel, keys [[0,1],[50,0]]}
         ]
        }]
      }]
    }
   ]
   display_rate {numerator=100}
  }
 }
 HideArrow {
  binding_count=1
  length_seconds=0.25
  object=&self:HideArrow
  sequence {
   bindings [Image_Arrow {
     id=#4
     possessed_class=Image
     tracks [
      {
       type=MovieSceneFloatTrack
       display_name=Render Opacity
       sections [{type=MovieSceneFloatSection, end_frame=25, start_frame=0, channels [None {type=MovieSceneScriptingFloatChannel, keys [[0,1],[25,0]]}]}]
      }
      {
       type=MovieSceneByteTrack
       display_name=Visibility
       sections [{type=MovieSceneByteSection, end_frame=25, start_frame=0, channels [None {type=MovieSceneScriptingByteChannel, keys [[0,4],[25,1]]}]}]
      }
     ]
    }]
   display_rate {numerator=100}
  }
 }
]
graphs [
 EventGraph {links [], nodes []}
 Override_GetArrow {nodes [
   Override_GetArrow() {is=title K2Node_FunctionEntry}
   > Return Node!(execute, Image_Arrow:Widget Object Reference=Get Image_Arrow() -> (Image_Arrow:Image Object Reference))
  ]}
 Override_GetCollapseAnimation {nodes [
   Override_GetCollapseAnimation() {is=title K2Node_FunctionEntry}
   > Return Node!(execute, Animation:Widget Animation Object Reference=Get Collapse() -> (Collapse:Widget Animation Object Reference))
  ]}
 Override_GetContentContainer {nodes [
   Override_GetContentContainer() {is=title K2Node_FunctionEntry}
   > Return Node!(execute, Container:Panel Widget Object Reference=Get Border_Content() -> (Border_Content:Border Object Reference))
  ]}
 Override_GetHideArrowAnimation {nodes [
   Override_GetHideArrowAnimation() {is=title K2Node_FunctionEntry}
   > Return Node!(execute, Animation:Widget Animation Object Reference=Get HideArrow() -> (HideArrow:Widget Animation Object Reference))
  ]}
 Override_GetTabSlot {nodes [
   Override_GetTabSlot() {is=title K2Node_FunctionEntry}
   > Return Node!(execute, Slot:Named Slot Object Reference=Get Slot_Tab() -> (Slot_Tab:Named Slot Object Reference))
  ]}
 Override_SetInstantTransition {nodes [
   n1 = Override_SetInstantTransition!() -> (then, IsExpanded:bool) {is=title K2Node_FunctionEntry}
   > Parent: Override_SetInstantTransition!(execute, IsExpanded:bool=@n1) -> (then) {is=title K2Node_CallParentFunction}
   > SetDesiredSizeScale!(
    execute
    self:Border Object Reference=Get Border_Content()
    InScale:Vector 2D Structure=Select(Option 0:Vector 2D Structure=(X=0,Y=0), Option 1:Vector 2D Structure=(X=1,Y=1), Index=@n1) -> (ReturnValue:Vector 2D Structure)
   ) -> (then)
   > SetBrushColor(
    self=Get Border_Background() -> (Border_Background:Border Object Reference)
    InBrushColor=SelectColor(A=Get Backplate Color() -> (Backplate Color:Linear Color Structure), B=(R=1,G=1,B=1,A=0), bPickA=@n1)
   )
   n2 = Get Image_Arrow()
   > Set RenderOpacity!(execute, RenderOpacity:float=SelectFloat(A=0, B=1, bPickA=@n1), self:Widget Object Reference=@n2) -> (then, Output_Get:float)
   > n3 = GetDynamicMaterial(self=@n2)
   > Is Valid(InputObject=@n3)
   > SetScalarParameterValue(self=@n3, ParameterName=Rotation, Value=SelectFloat(A=0.25, B=0.75, bPickA=@n1))
  ]}
 Update Border Opacity {nodes [
   Update Border Opacity() {is=title K2Node_FunctionEntry}
   n1 = Get Collapse()
   > n2 = IsAnimationPlayingForward!(execute, self:User Widget Object Reference, InAnimation:Widget Animation Object Reference=@n1) -> (then, ReturnValue:bool)
   n3 = GetAnimationCurrentTime!(self:User Widget Object Reference, InAnimation:Widget Animation Object Reference=@n1) -> (ReturnValue:float)
   n4 = SelectFloat(A="To Float (Boolean)"(InBool=@n2), B=SafeDivide(A=@n3, B=GetEndTime(self=@n1)), bPickA="Equal (Float)"(A=@n3))
   n5 = Get!() -> (Backplate Color_R:float=~, Backplate Color_G:float=~, Backplate Color_B:float=~, Backplate Color_A:float=~)
   > SetBrushColor!(
    execute
    self=Get Border_Background()
    InBrushColor_R:float=@n5
    InBrushColor_G:float=@n5.Backplate Color_G
    InBrushColor_B:float=@n5.Backplate Color_B
    InBrushColor_A:float=float * float(A=@n5.Backplate Color_A, B=float - float(A=1, B=@n4)) {is=title K2Node_PromotableOperator}
   ) -> (then)
  ]}
]
variables [Backplate Color {type=LinearColor, default=<rgba 1 1 1 1>}]
widgets {
 info {
  inheritedWidgetCount=0
  parentClass=&/WBP_UIKit_TabContainer_NoLayout.WBP_UIKit_TabContainer_NoLayout_C
  rootWidgetClass=&Border
  widgetCount=5
 }
 widgets [
  Border_Background {
   widgetClassPath=&Border
   bIsVariable=true
   widget=Border_Background
   properties {
    background {
     bIsDynamicallyLoaded=false
     drawAs=Image
     imageType=FullColor
     mirroring=NoMirror
     resourceName=None
     resourceObject=/FortUILibrary/Systems/Backplate/MaterialInstances/MI_UIKit_Backplate
     tiling=NoTile
     imageSize=<32 32>
     margin=<ltrb 0 0 0 0>
     outlineSettings {
      bUseBrushTransparency=false
      roundingType=HalfHeightRadius
      width=0
      color {colorUseRule=UseColor_Specified, specifiedColor=<rgba 0 0 0 0>}
      cornerRadii=<0 0 0 0>
     }
     tintColor {colorUseRule=UseColor_Specified, specifiedColor=<rgba 1 1 1 0.2>}
     uVRegion {bIsValid=false, max=<0 0>, min=<0 0>}
    }
    brushColor=<rgba 1 1 1 0>
    padding=<ltrb 0 0 0 0>
   }
  }
  Grid_Layout {
   widgetClassPath=&GridPanel
   bIsVariable=true
   parent=Border_Background
   widget=Grid_Layout
   properties {columnFill [1]}
   slot {class=BorderSlot, properties {padding=<ltrb 0 0 0 0>}}
  }
  Slot_Tab {widgetClassPath=&NamedSlot, bIsVariable=true, parent=Grid_Layout, widget=Slot_Tab, slot {class=GridSlot}}
  Image_Arrow {
   widgetClassPath=&Image
   bIsVariable=true
   parent=Grid_Layout
   widget=Image_Arrow
   properties {
    visibility=SelfHitTestInvisible
    brush {
     bIsDynamicallyLoaded=false
     drawAs=Image
     imageType=FullColor
     mirroring=NoMirror
     resourceName=None
     resourceObject=/FortUILibrary/Textures/MaterialInstances/MI_UIKit_Icon_ExpandArrow
     tiling=NoTile
     imageSize=<32 32>
     margin=<ltrb 0 0 0 0>
     outlineSettings {
      bUseBrushTransparency=false
      roundingType=HalfHeightRadius
      width=0
      color {colorUseRule=UseColor_Specified, specifiedColor=<rgba 0 0 0 0>}
      cornerRadii=<0 0 0 0>
     }
     tintColor {colorUseRule=UseColor_Specified, specifiedColor=<rgba 1 1 1 1>}
     uVRegion {bIsValid=false, max=<0 0>, min=<0 0>}
    }
   }
   slot {class=GridSlot, properties {horizontalAlignment=HAlign_Right, layer=10, verticalAlignment=VAlign_Center, padding=<ltrb 0 0 16 0>}}
  }
  Border_Content {
   widgetClassPath=&Border
   bIsVariable=true
   parent=Grid_Layout
   widget=Border_Content
   properties {
    clipping=ClipToBounds
    visibility=SelfHitTestInvisible
    background {
     bIsDynamicallyLoaded=false
     drawAs=Image
     imageType=NoImage
     mirroring=NoMirror
     resourceName=None
     resourceObject=null
     tiling=NoTile
     imageSize=<200 200>
     margin=<ltrb 0 0 0 0>
     outlineSettings {
      bUseBrushTransparency=false
      roundingType=HalfHeightRadius
      width=0
      color {colorUseRule=UseColor_Specified, specifiedColor=<rgba 0 0 0 0>}
      cornerRadii=<0 0 0 0>
     }
     tintColor {colorUseRule=UseColor_Specified, specifiedColor=<rgba 1 1 1 1>}
     uVRegion {bIsValid=false, max=<0 0>, min=<0 0>}
    }
    padding=<ltrb 0 0 0 0>
   }
   slot {class=GridSlot, properties {row=1}}
  }
 ]
}
