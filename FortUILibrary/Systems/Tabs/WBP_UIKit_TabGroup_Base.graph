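# WBP_UIKit_TabGroup_Base is the UIKit base tab bar widget, built on Common Tab List, with optional dividers, backplates, arrow buttons and input prompts.
# Tabs register from data or counts, with per-ID button classes and view models.
# Creation events lay out buttons, spacers and dividers in a horizontal or vertical grid, using doubled indices to interleave them.
# Selection scrolls tabs into view; input method changes update prompt and arrow visibility.
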WidgetBlueprint /FortUILibrary/Systems/Tabs/WBP_UIKit_TabGroup_Base
animations [Anim_PromptSpaceChange {
  binding_count=0
  length_seconds=1
  object=&self:Anim_PromptSpaceChange
  sequence {
   bindings []
   display_rate {numerator=100}
   tracks [{
     type=MovieSceneEventTrack
     sections [
      {type=MovieSceneEventRepeaterSection, end_frame=100, start_frame=0}
      {
       type=MovieSceneEventTriggerSection
       channels [None {
         type=MovieSceneScriptingEventChannel
         keys [[0, "<Struct 'MovieSceneEvent' (0x467360f9ba0) {}>"], [100, "<Struct 'MovieSceneEvent' (0x467360f8e80) {}>"]]
        }]
      }
     ]
    }]
  }
 }]
graphs [
 AddArrowButtons {nodes [
   AddArrowButtons() {is=title K2Node_FunctionEntry}
   > Remove Widget!(execute:Exec, Widget:Widget Object Reference=Get Next_ArrowButton() -> (Next_ArrowButton:WBP UIKit Button Generic Base Object Reference)) -> (then:Exec) {is=title K2Node_MacroInstance}
   > Remove Widget(Widget=Get Prev_ArrowButton() -> (Prev_ArrowButton:WBP UIKit Button Generic Base Object Reference))
   > Branch(Condition=Get UseArrowNavButtons() -> (UseArrowNavButtons:bool))
   > n1 = Sequence()
   > n2 = CreateArrowButton!(execute, self) -> (then, NewButton:WBP UIKit Button Generic Base Object Reference)
   > n3 = Set Next_ArrowButton(Next_ArrowButton:WBP UIKit Button Generic Base Object Reference=@n2) -> (Output_Get:WBP UIKit Button Generic Base Object Reference)
   > Branch(Condition=IsValid(Object=@n3))
   n4 = Create Event()
   > Unbind Event from On Clicked(self:Common Button Base Object Reference=@n3, Delegate=@n4) {is=title K2Node_RemoveDelegate}
   > Bind Event to On Clicked(self:Common Button Base Object Reference=@n3, Delegate=@n4) {is=title K2Node_AddDelegate}
   > AddChildToGrid!(
    execute
    self:Grid Panel Object Reference=Get Root() -> (Root:Grid Panel Object Reference)
    Content:Widget Object Reference=@n3
    InRow:int=1
    InColumn:int=2
   ) -> (then, ReturnValue:Grid Slot Object Reference)
   > Set Grid Slot Layer!(execute:Exec, Widget:Widget Object Reference=@n3, Layer:int=20) -> (then:Exec) {is=title K2Node_MacroInstance}
   n5 = CreateArrowButton(execute=@n1.then_1)
   > n6 = Set Prev_ArrowButton(Prev_ArrowButton:WBP UIKit Button Generic Base Object Reference=@n5) -> (Output_Get:WBP UIKit Button Generic Base Object Reference)
   > Branch(Condition=IsValid(Object=@n6))
   n7 = Create Event()
   > Unbind Event from On Clicked(self=@n6, Delegate=@n7)
   > Bind Event to On Clicked(self=@n6, Delegate=@n7)
   > AddChildToGrid(self=Get Root(), Content=@n6, InRow=1, InColumn=0)
   > Set Grid Slot Layer(Widget=@n6, Layer=20)
  ]}
 AddBackplate {nodes [
   AddBackplate() {is=title K2Node_FunctionEntry}
   n1 = Get BackplateClass() -> (BackplateClass:WBP UIKit Backplate Base Class Reference)
   > n2 = Set BackplateClass(BackplateClass:WBP UIKit Backplate Base Class Reference=Select(
     Option 0:WBP UIKit Backplate Base Class Reference=/FortUILibrary/Systems/Backplate/WBP_UIKit_Backplate_Base.WBP_UIKit_Backplate_Base_C
     Option 1:WBP UIKit Backplate Base Class Reference=@n1
     Index=IsValidClass!(Class:Object Class Reference=@n1) -> (ReturnValue:bool)
    ) -> (ReturnValue:WBP UIKit Backplate Base Class Reference)) -> (Output_Get:WBP UIKit Backplate Base Class Reference)
   > n3 = Get!(execute:Exec) -> (then:Exec, else:Exec, ContentBackplate:WBP UIKit Backplate Base Object Reference)
   n4 = Get ContentBackplate() -> (ContentBackplate:WBP UIKit Backplate Base Object Reference)
   n5 = AddChildToGrid(execute=[@n12.then, @n3.then], self=Get Root(), Content=@n4, InRow=1, InColumn=1)
   > n6 = Set BackplateGridSlot(BackplateGridSlot:Grid Slot Object Reference=@n5) -> (Output_Get:Grid Slot Object Reference)
   > n7 = Branch(Condition=AND Boolean(A=Get PromptsInsideTabsBackplate() -> (PromptsInsideTabsBackplate:bool), B=GetPromptsVisibility!(self) -> (IsVisible:bool)))
   n8 = Get IsVertical() -> (IsVertical:bool)
   > SetRow!(execute, self:Grid Slot Object Reference=@n6, InRow:int=SelectInt(A=0, B=1, bPickA=@n8)) -> (then)
   > SetRowSpan!(execute, self:Grid Slot Object Reference=@n6, InRowSpan:int=SelectInt(A=3, B=0, bPickA=@n8)) -> (then)
   > SetColumn!(execute, self:Grid Slot Object Reference=@n6, InColumn:int=SelectInt(A=1, B=0, bPickA=@n8)) -> (then)
   > n9 = SetColumnSpan!(execute, self:Grid Slot Object Reference=@n6, InColumnSpan:int=SelectInt(A=0, B=3, bPickA=@n8)) -> (then)
   SetBackplateProperties!(
    execute=[@n9.then, @n10.then]
    self
    Backplate:WBP UIKit Backplate Base Object Reference=@n4
    Brightness:E_UI_BackplateBrightness Enum=Get BackplateBrightness() -> (BackplateBrightness:E_UI_BackplateBrightness Enum=NewEnumerator0)
    Corner Radius:E_UI_BackplateCornerRadius Enum=Get BackplateCornerRadius() -> (BackplateCornerRadius:E_UI_BackplateCornerRadius Enum=NewEnumerator0)
   ) -> (then)
   > SetPadding!(execute, self:Grid Slot Object Reference=@n6, InPadding:Margin Structure=Get BackplatePadding() -> (BackplatePadding:Margin Structure)) -> (then)
   SetRowSpan(execute=@n7.else, self=@n6, InRowSpan=0)
   > n10 = SetColumnSpan(self=@n6, InColumnSpan=0)
   n11 = Construct Object from Class(execute=@n3.else, Class=@n2) -> (ReturnValue:WBP UIKit Backplate Base Object Reference)
   > n12 = Set ContentBackplate(ContentBackplate:WBP UIKit Backplate Base Object Reference=@n11) -> (Output_Get:WBP UIKit Backplate Base Object Reference)
  ]}
 AddBackplateForPrompts {nodes [
   AddBackplateForPrompts() {is=title K2Node_FunctionEntry}
   n1 = Get PromptsBackplateClass() -> (PromptsBackplateClass:WBP UIKit Backplate Base Class Reference)
   > n2 = Set PromptsBackplateClass(PromptsBackplateClass:WBP UIKit Backplate Base Class Reference=Select(Option 0=&WBP_UIKit_Backplate_Base_C, Option 1=@n1, Index=IsValidClass(Class=@n1))) -> (Output_Get:WBP UIKit Backplate Base Class Reference)
   > n3 = Sequence()
   > n4 = Get!(execute) -> (then, else, PrevPromptBackplate:WBP UIKit Backplate Base Object Reference)
   n5 = Get PrevPromptBackplate() -> (PrevPromptBackplate:WBP UIKit Backplate Base Object Reference)
   AddChild(execute=[@n11.then, @n4.then], self=Get Prev_SizeBox() -> (Prev_SizeBox:Size Box Object Reference), Content=@n5)
   n6 = Get PromptBackplateCornerRadius() -> (PromptBackplateCornerRadius:E_UI_BackplateCornerRadius Enum=NewEnumerator0)
   n7 = Get PromptBackplateBrightness() -> (PromptBackplateBrightness:E_UI_BackplateBrightness Enum=NewEnumerator0)
   > SetBackplateProperties(Backplate=@n5, Brightness=@n7, Corner Radius=@n6)
   n8 = BreakVector2D(InVec=Get PromptBackplateMinSize() -> (PromptBackplateMinSize:Vector 2D Structure))
   n9 = Get Prev_SizeBox()
   > SetMinDesiredHeight!(execute, self:Size Box Object Reference=@n9, InMinDesiredHeight:float=@n8.Y) -> (then)
   > SetMinDesiredWidth!(execute, self:Size Box Object Reference=@n9, InMinDesiredWidth:float=@n8) -> (then)
   n10 = Construct Object from Class(execute=@n4.else, Class=@n2)
   > n11 = Set PrevPromptBackplate(PrevPromptBackplate:WBP UIKit Backplate Base Object Reference=@n10) -> (Output_Get:WBP UIKit Backplate Base Object Reference)
   n12 = Get!(execute=@n3.then_1) -> (then, else, NextPromptBackplate:WBP UIKit Backplate Base Object Reference)
   n13 = Get NextPromptBackplate() -> (NextPromptBackplate:WBP UIKit Backplate Base Object Reference)
   AddChild(execute=[@n16.then, @n12.then], self=Get Next_SizeBox() -> (Next_SizeBox:Size Box Object Reference), Content=@n13)
   > SetBackplateProperties(Backplate=@n13, Brightness=@n7, Corner Radius=@n6)
   n14 = Get Next_SizeBox()
   > SetMinDesiredHeight(self=@n14, InMinDesiredHeight=@n8.Y)
   > SetMinDesiredWidth(self=@n14, InMinDesiredWidth=@n8)
   n15 = Construct Object from Class(execute=@n12.else, Class=@n2)
   > n16 = Set NextPromptBackplate(NextPromptBackplate:WBP UIKit Backplate Base Object Reference=@n15) -> (Output_Get:WBP UIKit Backplate Base Object Reference)
  ]}
 AddDivider {nodes [
   n1 = AddDivider!() -> (then, TabID:Name, AddAfter:bool=true) {is=title K2Node_FunctionEntry}
   > n2 = Set CurrentTabIndex(CurrentTabIndex:int=GetTabIndexByID!(self:UIKit Tab Group Object Reference, TabNameID:Name=@n1) -> (ReturnValue:int=0)) -> (Output_Get:int=0)
   > n3 = Branch(Condition="integer >= integer"(A=@n2))
   > HandleAddDivider!(execute, self, TabID:"Name (by ref)"=@n1, TabIndex:int=@n2, AddAfter:bool=@n1.AddAfter) -> (then)
   Add!(
    execute=@n3.else
    TargetMap:Map of Names to S UI Divider Data Structures=Get CachedDividers() -> (CachedDividers:Map of Names to S UI Divider Data Structures)
    Key:"Name (by ref)"=@n1
    Value:"S UI Divider Data Structure (by ref)"=Make S UI Divider Data!(
     Widget_6_#1:WBP UIKit Divider Base Object Reference
     IsAddedAfter_3_#2:bool=@n1.AddAfter
    ) -> (S_UI_DividerData:S UI Divider Data Structure) {is=title K2Node_MakeStruct}
   ) -> (then) {is=title K2Node_CallFunction}
  ]}
 AddDividerByIndex {nodes [
   n1 = AddDividerByIndex!() -> (then, TabIndex:int, AddAfter:bool=true) {is=title K2Node_FunctionEntry}
   > n2 = Set CurrentTabID(CurrentTabID:Name=GetTabIDbyIndex!(self:UIKit Tab Group Object Reference, Index:int=@n1) -> (ReturnValue:Name=None)) -> (Output_Get:Name=None)
   > n3 = Branch(Condition="Not Equal (Name)"(A=@n2) -> (ReturnValue=~))
   > HandleAddDivider(TabID=@n2, TabIndex=@n1, AddAfter=@n1.AddAfter)
   Add(
    execute=@n3.else
    TargetMap=Get CachedDividers()
    Key=String To Name(InString="To String (Integer)"(InInt=@n1))
    Value=Make S UI Divider Data(IsAddedAfter_3_#2=@n1.AddAfter)
   )
  ]}
 AddDividerToGrid {nodes [
   n1 = AddDividerToGrid!() -> (then, Divider:WBP UIKit Divider Base Object Reference, Index:int, AddAfter:bool=~) {is=title K2Node_FunctionEntry}
   > n2 = Set CurrentGridCell(CurrentGridCell:int=int + int(A=int * int(A=@n1.Index, B=2), B=SelectInt(A=2, B=0, bPickA=@n1.AddAfter))) -> (Output_Get:int=0)
   n3 = Get IsVertical()
   > AddChildToGrid(
    self=Get TabsContainer() -> (TabsContainer:Grid Panel Object Reference)
    Content=@n1
    InRow=SelectInt(A=@n2, B=0, bPickA=@n3)
    InColumn=SelectInt(A=0, B=@n2, bPickA=@n3)
   )
  ]}
 AddSpacerToGrid {nodes [
   n1 = AddSpacerToGrid!() -> (then, TabID:"Name (by ref)", Index:int) {is=title K2Node_FunctionEntry}
   > Branch(Condition="integer > integer"(A=@n1.Index))
   > n2 = Construct Object from Class(Class=&Spacer) -> (ReturnValue:Spacer Object Reference)
   > n3 = Set CurrentSpacer(CurrentSpacer:Spacer Object Reference=@n2) -> (Output_Get:Spacer Object Reference)
   n4 = Get TabSpacing() -> (TabSpacing:double)
   n5 = Get IsVertical()
   > SetSize!(
    execute
    self:Spacer Object Reference=@n3
    InSize_X:double=Select(Option 0:double=@n4, Option 1:double=1, Index=@n5) -> (ReturnValue:double)
    InSize_Y:double=Select(Option 0=1, Option 1=@n4, Index=@n5)
   ) -> (then)
   n6 = int * int(A=@n1.Index, B=2)
   n7 = Get IsVertical()
   > AddChildToGrid(
    self=Get TabsContainer()
    Content=@n3
    InRow=Select(Option 0:int=0, Option 1:int=@n6, Index=@n7) -> (ReturnValue:int)
    InColumn=Select(Option 0=@n6, Option 1=0, Index=@n7)
   )
   > Add(
    TargetMap:Map of Names to Spacer Object References=Get CachedSpacers() -> (CachedSpacers:Map of Names to Spacer Object References)
    Key=Get TabID() -> (TabID:"Name (by ref)"=None)
    Value:"Spacer Object Reference (by ref)"=@n3
   )
  ]}
 AddTab {nodes [
   n1 = AddTab!() -> (then, Tab VM:MVVM Base Viewmodel Object Reference, Tab ID:Name, Index:int=-1) {is=title K2Node_FunctionEntry}
   > n2 = Set CurrentTabIndex(CurrentTabIndex=SelectInt(
     A=GetTabCount!(self:Common Tab List Widget Base Object Reference) -> (ReturnValue:int=0)
     B=@n1.Index
     bPickA="integer < integer"(A=@n1.Index)
    )) -> (Output_Get=0)
   n3 = "To String (Integer)"(InInt=@n2)
   > n4 = Set CurrentTabID(CurrentTabID=Select(Option 0:Name=@n1.Tab ID, Option 1:Name=String To Name(InString=@n3), Index="Equal (Name)"(A=@n1.Tab ID)) -> (ReturnValue:Name)) -> (Output_Get=None)
   > Add(
    TargetMap:Map of Names to MVVM Base Viewmodel Object References=Get CachedViewModels() -> (CachedViewModels:Map of Names to MVVM Base Viewmodel Object References)
    Key=@n4
    Value:"MVVM Base Viewmodel Object Reference (by ref)"=@n1
   )
   > Add(
    TargetMap:Map of Names to Integers=Get CachedButtonIndex() -> (CachedButtonIndex:Map of Names to Integers)
    Key=@n4
    Value:"Integer (by ref)"=@n2
   )
   > n5 = RegisterTab!(
    execute
    self:Common Tab List Widget Base Object Reference
    TabNameID:Name=@n4
    ButtonWidgetType:Common Button Base Class Reference=GetButtonClass!(TabID:Name=@n4) -> (ButtonClass:Common Button Base Class Reference)
    ContentWidget:Widget Object Reference
    TabIndex:int=@n2
   ) -> (then, ReturnValue:bool)
   > n6 = Branch(Condition=@n5)
   > Branch(Condition=IsValid(Object=@n1))
   > SetTabViewModel!(execute, self, TabID:Name=@n4, VM:Notify Field Value Changed Interface=@n1) -> (then)
   Remove!(execute=@n6.else, TargetMap:Map of Names to MVVM Base Viewmodel Object References=Get CachedViewModels(), Key:"Name (by ref)"=@n4) -> (then, ReturnValue:bool)
   > Remove(TargetMap:Map of Names to Integers=Get CachedButtonIndex(), Key=@n4)
  ]}
 AddTabToGrid {nodes [
   n1 = AddTabToGrid!() -> (then, Widget:Widget Object Reference, Index:int) {is=title K2Node_FunctionEntry}
   > Set CurrentCell(CurrentCell:int=int + int(A=int * int(A=@n1.Index, B=2))) -> (Output_Get:int=0)
   > n2 = Branch(Condition=Get IsVertical())
   n3 = Get CurrentCell() -> (CurrentCell:int=0)
   n4 = Get TabsContainer()
   > n5 = AddChildToGrid(self=@n4, Content=@n1, InRow=@n3, InColumn=0)
   > SetHorizontalAlignment(
    self:Grid Slot Object Reference=@n5
    InHorizontalAlignment=Get ButtonHorizontalAlignment() -> (ButtonHorizontalAlignment:EHorizontalAlignment Enum=HAlign_Fill)
   )
   n6 = Get CurrentCell() -> (CurrentCell=0)
   > SetRowFill!(
    execute
    self:Grid Panel Object Reference=@n4
    RowIndex:int=@n6
    Coefficient:float=SelectFloat(A=1, B=0, bPickA=IsFillEnabled!() -> (ReturnValue:bool))
   ) -> (then)
   n7 = Get CurrentCell() -> (CurrentCell=0)
   > n8 = SetColumnFill!(execute, self:Grid Panel Object Reference=@n4, ColumnIndex:int=@n7, Coefficient:float=0) -> (then)
   Set Grid Slot Padding!(
    execute:Exec=[@n11.then, @n8.then]
    Widget:Widget Object Reference=Get Widget() -> (Widget:Widget Object Reference)
    Padding:Margin Structure=Get ButtonPadding() -> (ButtonPadding:Margin Structure)
   ) -> (then:Exec) {is=title K2Node_MacroInstance}
   n9 = Get TabsContainer()
   n10 = AddChildToGrid(execute=@n2.else, self=@n9, Content=@n1, InRow=0, InColumn=@n3)
   > SetVerticalAlignment(
    self:Grid Slot Object Reference=@n10
    InVerticalAlignment=Get ButtonVerticalAlignment() -> (ButtonVerticalAlignment:EVerticalAlignment Enum=VAlign_Fill)
   )
   > SetRowFill(self=@n9, RowIndex=@n6, Coefficient=0)
   > n11 = SetColumnFill(self=@n9, ColumnIndex=@n7, Coefficient=SelectFloat(A=1, B=0, bPickA=IsFillEnabled()))
  ]}
 CacheInitialDividers {nodes [
   CacheInitialDividers() {is=title K2Node_FunctionEntry}
   > Clear!(execute, TargetMap:Map of Names to S UI Divider Data Structures=Get CachedDividers()) -> (then) {is=title K2Node_CallFunction}
   > n1 = For Each Loop(Array:Array of S UI Divider Position Structures=Get Dividers() -> (Dividers:Array of S UI Divider Position Structures)) -> (Array Element:S UI Divider Position Structure)
   n2 = Break S UI Divider Position!(S_UI_DividerPosition:"S UI Divider Position Structure (by ref)"=@n1) -> (
    TabID_8_#1:Name
    TabIndex_12_#3:int
    AddAfterTab_9_#2:bool=~
   ) {is=title K2Node_BreakStruct}
   > n3 = Branch(Condition="Not Equal (Name)"(A=@n2) -> (ReturnValue=~))
   n4 = Make S UI Divider Data(IsAddedAfter_3_#2=@n2.AddAfterTab_9_#2)
   n5 = Get CachedDividers()
   > Add(TargetMap:Map of Names to S UI Divider Data Structures=@n5, Key=@n2, Value:"S UI Divider Data Structure (by ref)"=@n4)
   Branch(execute=@n3.else, Condition="integer >= integer"(A=@n2.TabIndex_12_#3))
   > Add(TargetMap=@n5, Key=String To Name(InString="To String (Integer)"(InInt=@n2.TabIndex_12_#3)), Value=@n4)
  ]}
 CreateArrowButton {nodes [
   CreateArrowButton() {is=title K2Node_FunctionEntry}
   > Branch(Condition=Get UseArrowNavButtons())
   > n1 = Create Widget!(
    execute:Exec
    Class:User Widget Class Reference=Get ArrowButtonsClass() -> (ArrowButtonsClass:WBP UIKit Button Generic Base Class Reference)
    OwningPlayer:Player Controller Object Reference
    Style:Common Button Style Class Reference=/FortUILibrary/UIKit/Buttons/Blocks/ButtonStyle_UIKit_Empty.ButtonStyle_UIKit_Empty_C
    bApplyAlphaOnDisable:bool=False
    bLocked:bool=False
    bSelectable:bool=false
    bShouldSelectUponReceivingFocus:bool=false
    bInteractableWhenSelected:bool=false
    bToggleable:bool=False
    bTriggerClickedAfterSelection:bool=False
    bDisplayInputActionWhenNotInteractable:bool=True
    bHideInputActionWithKeyboard:bool=False
    bShouldUseFallbackDefaultInputAction:bool=True
    bRequiresHold:bool=False
    HoldData:Common UIHold Data Class Reference=&CommonUI_HoldData_Base_C
    ClickMethod:EButtonClickMethod Enum=DownAndUp
    InputPriority:int=0
    TriggeringInputAction:Data Table Row Handle Structure
   ) -> (then:Exec, ReturnValue:WBP UIKit Button Generic Base Object Reference)
   > Return Node!(execute, NewButton:WBP UIKit Button Generic Base Object Reference=@n1)
  ]}
 CreateEmptyTabs {nodes [
   n1 = CreateEmptyTabs!() -> (then, TabsCount:int) {is=title K2Node_FunctionEntry}
   > RemoveTabGroupWidgets()
   > n2 = For Loop(LastIndex=int - int(A=@n1))
   InitialTabSelection(execute=@n2.Completed)
   n3 = Set TabID(execute=@n2.LoopBody, TabID:Name=String To Name(InString="To String (Integer)"(InInt=@n2))) -> (Output_Get:Name=None)
   > RegisterTab(TabNameID=@n3, ButtonWidgetType=GetButtonClass(TabID=@n3), TabIndex=@n2)
  ]}
 EventGraph {nodes [
   n1 = Event On Tab Selected!() -> (OutputDelegate, then, TabId:Name) {is=title K2Node_CustomEvent}
   > ScrollOnTabSelection!(execute, self, TabNameID:Name=@n1.TabId) -> (then)
   > UpdateArrowButtonsEnabled()
   n2 = Event On Input Method Changed!() -> (OutputDelegate, then, bNewInputType:ECommonInputType Enum) {is=title K2Node_CustomEvent}
   > n3 = Set CurrentInputType(CurrentInputType:ECommonInputType Enum=@n2.bNewInputType) -> (Output_Get:ECommonInputType Enum)
   > UpdatePromptsVisibility()
   > UpdateArrowButtonsVisibility!(execute, self, CurrentInputType:ECommonInputType Enum=@n3) -> (then)
   OnNextArrowButtonClicked!() -> (OutputDelegate, then, Button:Common Button Base Object Reference) {is=title K2Node_CustomEvent}
   n4 = Get ArrowButtonsAllowWrap() -> (ArrowButtonsAllowWrap:bool)
   > n5 = Branch(Condition=@n4)
   SelectNextTab!(execute=[@n8.then, @n5.then], self:UIKit Tab Group Object Reference, bSuppressClickFeedback:bool=true) -> (then)
   n6 = GetTabCount() -> (ReturnValue=0)
   n7 = GetSelectedTabIndex!(self:UIKit Tab Group Object Reference) -> (ReturnValue:int=0)
   n8 = Branch(execute=@n5.else, Condition="Not Equal (Integer)"(A=int - int(A=@n6), B=@n7))
   n9 = Event OnTab Button Removal!() -> (OutputDelegate, then, TabId:Name, TabButton:Common Button Base Object Reference) {is=title K2Node_CustomEvent}
   > RemoveDividerWidget!(execute, self, TabID:Name=@n9.TabId) -> (then)
   > RemoveSpacer!(execute, self, TabID:Name=@n9.TabId) -> (then)
   > RemoveCellFill!(execute, self, TabNameID:Name=@n9.TabId) -> (then)
   > n10 = Branch(Condition=IsRebuildingTabList!(self:UIKit Tab Group Object Reference) -> (ReturnValue:bool))
   Remove(execute=@n10.else, TargetMap:Map of Names to MVVM Base Viewmodel Object References=Get CachedViewModels(), Key=@n9.TabId)
   OnPreviousArrowButtonClicked!() -> (OutputDelegate, then, Button:Common Button Base Object Reference) {is=title K2Node_CustomEvent}
   > n11 = Branch(Condition=@n4)
   SelectPreviousTab!(execute=[@n11.then, @n12.then], self:UIKit Tab Group Object Reference, bSuppressClickFeedback:bool=true) -> (then)
   n12 = Branch(execute=@n11.else, Condition="Not Equal (Integer)"(A=@n7))
   n13 = Event PreConstruct()
   > n14 = Sequence!(execute) -> (then_0, then_1, then_2:Exec)
   > CacheInitialDividers() {is=title K2Node_CallFunction}
   > Set Tabs Alignment!(execute:Exec) -> (then:Exec) {is=title K2Node_MacroInstance}
   n15 = Branch(execute=@n14.then_1, Condition=@n13.IsDesignTime)
   > n16 = PresetForDesignTime()
   PopulateTabsFromData!(
    execute=[@n17.then, @n16.then]
    self
    Data:Array of UIKit Tab Data Structures=Get InitialTabsData() -> (InitialTabsData:Array of UIKit Tab Data Structures)
   ) -> (then)
   > InitialTabSelection()
   n17 = Branch(execute=@n15.else, Condition=Get UseInitialTabs() -> (UseInitialTabs:bool))
   n18 = Branch(execute=@n14.then_2, Condition=Get UseBackplate() -> (UseBackplate:bool))
   > n19 = AddBackplate() {is=title K2Node_CallFunction}
   n20 = Branch(execute=[@n19.then, @n23.then], Condition=Get UsePromptBackplate() -> (UsePromptBackplate:bool))
   > n21 = AddBackplateForPrompts() {is=title K2Node_CallFunction}
   AddArrowButtons(execute=[@n21.then, @n22.then]) {is=title K2Node_CallFunction}
   > UpdateTabLayout()
   > UpdateInputActions()
   > UpdatePromptsVisibility()
   n22 = RemoveBackplateForPrompts(execute=@n20.else)
   n23 = RemoveBackplate(execute=@n18.else)
   n24 = Event HandleTabCreation!() -> (OutputDelegate, then, TabNameID:Name=None, TabButton:Common Button Base Object Reference) {is=title K2Node_Event}
   > HandleCreatedTab!(execute, self, Tab ID:Name=@n24.TabNameID, Button:Common Button Base Object Reference=@n24.TabButton) -> (then)
   Event Construct()
   n25 = Create Event()
   > Bind Event to On Tab Selected(self:Common Tab List Widget Base Object Reference, Delegate=@n25) {is=title K2Node_AddDelegate}
   n26 = Create Event()
   > Assign On Tab Button Removal!(execute:Exec, self:Common Tab List Widget Base Object Reference, Delegate:"Delegate (by ref)"=@n26) -> (then:Exec) {is=title K2Node_AssignDelegate}
   n27 = Get CommonInputSubsystem()
   > Branch(Condition=IsValid(Object=@n27))
   n28 = Create Event()
   > Bind Event to On Input Method Changed(self=@n27, Delegate=@n28)
   > Event On Input Method Changed!(execute, self, bNewInputType:ECommonInputType Enum=GetCurrentInputType(self=@n27)) -> (then) {is=title K2Node_CallFunction}
   Event Destruct()
   > Unbind Event from On Tab Selected(self:Common Tab List Widget Base Object Reference, Delegate=@n25) {is=title K2Node_RemoveDelegate}
   > Unbind Event from On Tab Button Removal(self:Common Tab List Widget Base Object Reference, Delegate=@n26) {is=title K2Node_RemoveDelegate}
   > Unbind Event from On Input Method Changed(self=Get CommonInputSubsystem(), Delegate=@n28)
   > RemoveCachedWidgets()
  ]}
 GetButtonClass {nodes [
   n1 = GetButtonClass!() -> (then, TabID:Name) {is=title K2Node_FunctionEntry}
   n2 = Get ButtonClassOverride() -> (ButtonClassOverride:Map of Names to Common Button Base Class References)
   > n3 = Branch(Condition=Is Not Empty!(TargetMap:Map of Names to Common Button Base Class References=@n2) -> (ReturnValue:bool) {is=title K2Node_CallFunction})
   > n4 = Keys!(execute, TargetMap:Map of Names to Common Button Base Class References=@n2) -> (then, Keys:Array of Names)
   > n5 = Set FoundIndex(FoundIndex:int=Find Item!(TargetArray:Array of Names=@n4, ItemToFind:"Name (by ref)"=@n1) -> (ReturnValue)) -> (Output_Get:int=0)
   > n6 = Branch(Condition="Not Equal (Integer)"(A=@n5, B=-1))
   > n7 = Values!(execute, TargetMap:Map of Names to Common Button Base Class References=@n2) -> (then, Values:Array of Common Button Base Class References)
   n8 = "Get (a copy)"(Array:Array of Common Button Base Class References=@n7, Dimension 1=@n5) -> (Output:Common Button Base Class Reference)
   > n9 = Branch(Condition=IsValidClass(Class=@n8))
   > Return Node!(execute, ButtonClass:Common Button Base Class Reference=@n8)
   Return Node(execute=[@n9.else, @n6.else, @n3.else], ButtonClass=Get ButtonClass() -> (ButtonClass:Common Button Base Class Reference))
  ]}
 GetPromptPositionOffsetValue {nodes [
   n1 = Inputs!() -> (DefaultValue:int, AddValue:int)
   n2 = Get GamepadHideArrows() -> (GamepadHideArrows:bool)
   Outputs!(ReturnValue:int=int + int(A=MakeLiteralInt!(Value:int=@n1) -> (ReturnValue:int=0), B=Select(Option 0:int=@n1.AddValue, Option 1:int=0, Index=@n2) -> (ReturnValue:int)))
  ]}
 GetPromptsVisibility {nodes [
   GetPromptsVisibility() {is=title K2Node_FunctionEntry}
   > n1 = Branch(Condition=Get HidePrompts() -> (HidePrompts:bool))
   Return Node!(execute=[@n10.else, @n5.then, @n1.then, @n4.then], IsVisible:bool=false)
   n2 = Literal enum ECommonInputType!(Enum:ECommonInputType Enum=MouseAndKeyboard) -> (ReturnValue:ECommonInputType Enum=~) {is=title K2Node_EnumLiteral}
   n3 = Get CurrentInputType() -> (CurrentInputType:ECommonInputType Enum)
   n4 = Branch(
    execute=@n1.else
    Condition=AND Boolean(A=Get HideMouseAndKeyboardPrompts() -> (HideMouseAndKeyboardPrompts:bool), B="Equal (Enum)"(A=@n3, B=@n2))
   )
   n5 = Branch(
    execute=@n4.else
    Condition="Equal (Enum)"(
     A:E_UI_TabsInputActionType Enum=Get InputActions() -> (InputActions:E_UI_TabsInputActionType Enum=NewEnumerator2)
     B:E_UI_TabsInputActionType Enum=Literal enum E_UI_TabsInputActionType!(Enum:E_UI_TabsInputActionType Enum=NewEnumerator2) -> (ReturnValue:E_UI_TabsInputActionType Enum) {is=title K2Node_EnumLiteral}
    )
   )
   n6 = Get NextTabAction() -> (NextTabAction:Common Action Widget Object Reference)
   n7 = IsValid(Object=Break Slate Brush!(SlateBrush:"Slate Brush Structure (by ref)"=GetIcon!(self:Common Action Widget Object Reference=@n6) -> (ReturnValue:Slate Brush Structure)) -> (ResourceObject:Object Reference) {is=title K2Node_BreakStruct})
   n8 = Get PrevTabAction() -> (PrevTabAction:Common Action Widget Object Reference)
   n9 = IsValid(Object=Break Slate Brush(SlateBrush=GetIcon(self=@n8)))
   n10 = Branch(execute=@n5.else, Condition=OR Boolean(A=AND Boolean(A=@n9, B=@n7), B=InEditor!() -> (ReturnValue:bool)))
   > Return Node(IsVisible=true)
  ]}
 HandleAddDivider {nodes [
   n1 = HandleAddDivider!() -> (then, TabID:"Name (by ref)", TabIndex:int, AddAfter:bool=~) {is=title K2Node_FunctionEntry}
   n2 = Find(TargetMap:Map of Names to S UI Divider Data Structures=Get CachedDividers(), Key:"Name (by ref)"=@n1) -> (Value:S UI Divider Data Structure)
   > n3 = Branch(Condition=@n2.ReturnValue)
   > n4 = Set CurrentDivider(CurrentDivider:WBP UIKit Divider Base Object Reference=Break S UI Divider Data!(S_UI_DividerData:"S UI Divider Data Structure (by ref)"=@n2) -> (Widget_6_#1:WBP UIKit Divider Base Object Reference, IsAddedAfter_3_#2:bool=~) {is=title K2Node_BreakStruct}) -> (Output_Get:WBP UIKit Divider Base Object Reference)
   > n5 = Branch(Condition=IsValid(Object=@n4))
   n6 = Get CurrentDivider() -> (CurrentDivider:WBP UIKit Divider Base Object Reference)
   AddDividerToGrid!(execute=[@n10.then, @n5.then], self, Divider:WBP UIKit Divider Base Object Reference=@n6, Index:int=@n1.TabIndex, AddAfter:bool=@n1.AddAfter) -> (then) {is=title K2Node_CallFunction}
   > SetupDivider!(execute, self, InputPin:WBP UIKit Divider Base Object Reference=@n6) -> (then)
   n7 = Get DividerClass() -> (DividerClass:WBP UIKit Divider Base Class Reference)
   n8 = Construct Object from Class(
    execute=[@n5.else, @n3.else]
    Class=Select(
     Option 0:WBP UIKit Divider Base Class Reference=&/WBP_UIKit_Divider_Base.WBP_UIKit_Divider_Base_C
     Option 1:WBP UIKit Divider Base Class Reference=@n7
     Index=IsValidClass(Class=@n7)
    ) -> (ReturnValue:WBP UIKit Divider Base Class Reference)
   ) -> (ReturnValue:WBP UIKit Divider Base Object Reference)
   > n9 = Set CurrentDivider(CurrentDivider=@n8)
   > n10 = Add!(
    execute
    TargetMap=Get CachedDividers()
    Key=Get TabID() -> (TabID=None)
    Value_Widget_6_#1:WBP UIKit Divider Base Object Reference=@n9
    Value_IsAddedAfter_3_#2:bool=@n1.AddAfter
   ) -> (then)
  ]}
 HandleCreatedTab {nodes [
   n1 = HandleCreatedTab!() -> (then, Tab ID:Name, Button:Common Button Base Object Reference) {is=title K2Node_FunctionEntry}
   > n2 = Set CurrentTabIndex(CurrentTabIndex=GetTabIndexByID(TabNameID=@n1) -> (ReturnValue=0)) -> (Output_Get=0)
   > SetupTabButton!(execute, self, Tab ID:Name=@n1, Tab Button:Common Button Base Object Reference=@n1.Button) -> (then)
   > AddSpacerToGrid!(execute, self, TabID:"Name (by ref)"=@n1, Index:int=@n2) -> (then) {is=title K2Node_CallFunction}
   > AddTabToGrid!(execute, self, Widget:Widget Object Reference=@n1.Button, Index:int=@n2) -> (then) {is=title K2Node_CallFunction}
   n3 = Find(TargetMap=Get CachedDividers(), Key=@n1)
   > Branch(Condition=@n3.ReturnValue)
   > HandleAddDivider(
    TabID=@n1
    TabIndex=@n2
    AddAfter=Break S UI Divider Data(S_UI_DividerData=@n3) -> (IsAddedAfter_3_#2=~).IsAddedAfter_3_#2
   ) {is=title K2Node_CallFunction}
  ]}
 InitialTabSelection {nodes [
   InitialTabSelection() {is=title K2Node_FunctionEntry}
   n1 = Get InitialSelectedTabID() -> (InitialSelectedTabID:Name=None)
   n2 = GetTabButtonBaseByID!(self:Common Tab List Widget Base Object Reference, TabNameID:Name=@n1) -> (ReturnValue:Common Button Base Object Reference)
   > n3 = Cast To WBP_UIKit_ButtonGeneric_Base!(execute, Object=@n2) -> (then, CastFailed, AsWBP UIKit Button Generic Base:WBP UIKit Button Generic Base Object Reference) {is=title K2Node_DynamicCast}
   > n4 = RequestInstantSelection(self:WBP UIKit Button Generic Base Object Reference=@n3)
   SelectTabByID!(execute=[@n4.then, @n5.then], self:Common Tab List Widget Base Object Reference, TabNameID:Name=@n1, bSuppressClickFeedback:bool=true) -> (then, ReturnValue:bool)
   n5 = Branch(execute=@n3.CastFailed, Condition=IsValid(Object=@n2))
   n6 = GetTabIdAtIndex!(self:Common Tab List Widget Base Object Reference, Index:int=Get InitialSelectedTabIndex() -> (InitialSelectedTabIndex:int=0)) -> (ReturnValue:Name=None)
   n7 = Cast To WBP_UIKit_ButtonGeneric_Base(execute=@n5.else, Object=GetTabButtonBaseByID(TabNameID=@n6))
   > n8 = RequestInstantSelection(self=@n7)
   SelectTabByID(execute=[@n8.then, @n7.CastFailed], TabNameID=@n6, bSuppressClickFeedback=true)
  ]}
 IsFillEnabled {nodes [
   IsFillEnabled() {is=title K2Node_FunctionEntry}
   n1 = "Equal (Enum)"(
    A:EVerticalAlignment Enum=Get TabGroupVerticalAlignment() -> (TabGroupVerticalAlignment:EVerticalAlignment Enum=VAlign_Fill)
    B:EVerticalAlignment Enum=VAlign_Fill
   )
   n2 = "Equal (Enum)"(A=Get ButtonVerticalAlignment() -> (ButtonVerticalAlignment=VAlign_Fill), B=VAlign_Fill)
   n3 = "Equal (Enum)"(
    A:EHorizontalAlignment Enum=Get TabGroupHorizontalAlignment() -> (TabGroupHorizontalAlignment:EHorizontalAlignment Enum=HAlign_Fill)
    B:EHorizontalAlignment Enum=HAlign_Fill
   )
   n4 = "Equal (Enum)"(A=Get ButtonHorizontalAlignment() -> (ButtonHorizontalAlignment=HAlign_Fill), B=HAlign_Fill)
   > Return Node!(
    execute
    ReturnValue:bool=Select(Option 0:bool=AND Boolean(A=@n3, B=@n4), Option 1:bool=AND Boolean(A=@n1, B=@n2), Index=Get IsVertical()) -> (ReturnValue:bool)
   )
  ]}
 OverrideInputActions {nodes [
   n1 = OverrideInputActions!() -> (then, Next Tab:Data Table Row Handle Structure, PreviousTab:Data Table Row Handle Structure) {is=title K2Node_FunctionEntry}
   > SetInputAction!(execute, self:Common Action Widget Object Reference=Get NextTabAction(), InputActionRow:Data Table Row Handle Structure=@n1) -> (then)
   > SetInputAction(self=Get PrevTabAction(), InputActionRow=@n1.PreviousTab)
   > SetInputActions!(
    execute
    self:UIKit Tab Group Object Reference
    NextTab:Data Table Row Handle Structure=@n1
    PreviousTab:Data Table Row Handle Structure=@n1.PreviousTab
   ) -> (then)
  ]}
 PopulateTabsFromData {nodes [
   n1 = PopulateTabsFromData!() -> (then, Data:Array of UIKit Tab Data Structures) {is=title K2Node_FunctionEntry}
   > Set CurrentSelection(CurrentSelection:Name=GetSelectedTabId!(self:Common Tab List Widget Base Object Reference) -> (ReturnValue:Name=None)) -> (Output_Get:Name=None)
   > RemoveTabGroupWidgets()
   > n2 = For Each Loop(Array:Array of UIKit Tab Data Structures=@n1) -> (Array Element:UIKit Tab Data Structure)
   n3 = Branch(execute=@n2.Completed, Condition=InEditor())
   RestoreTabSelection!(execute=@n3.else, self, TabNameID:Name=Get CurrentSelection() -> (CurrentSelection:Name=None)) -> (then)
   n4 = Break UIKit Tab Data!(UIKitTabData:"UIKit Tab Data Structure (by ref)"=@n2) -> (bEnableTabID:bool=~, TabID:Name, TabViewModel:MVVM Base Viewmodel Object Reference) {is=title K2Node_BreakStruct}
   AddTab!(
    execute=@n2.LoopBody
    self
    Tab VM:MVVM Base Viewmodel Object Reference=@n4.TabViewModel
    Tab ID:Name=SelectName!(A:Name=@n4.TabID, B:Name=None, bPickA:bool=@n4) -> (ReturnValue:Name=None)
    Index:int=@n2.Array Index
   ) -> (then) {is=title K2Node_CallFunction}
  ]}
 PresetForDesignTime {nodes [
   PresetForDesignTime() {is=title K2Node_FunctionEntry}
   > n1 = Branch(Condition=GetPromptsVisibility() {is=title K2Node_CallFunction})
   > Set Visibility(Visibility=Visible, self=Get Prev_Prompt() -> (Prev_Prompt:Grid Panel Object Reference))
   > Set Visibility(Visibility=Visible, self=Get Next_Prompt() -> (Next_Prompt:Grid Panel Object Reference))
   n2 = Get ReservedPromptSpace() -> (ReservedPromptSpace:double)
   > SetSize!(
    execute
    self=[Get Next_Spacer() -> (Next_Spacer:Spacer Object Reference), Get Prev_Spacer() -> (Prev_Spacer:Spacer Object Reference)]
    InSize:Vector 2D Structure=Select(Option 0:Vector 2D Structure=MakeVector2D(X=@n2, Y=0), Option 1:Vector 2D Structure=MakeVector2D(X=0, Y=@n2), Index=Get IsVertical()) -> (ReturnValue:Vector 2D Structure)
   ) -> (then)
   Set Visibility(execute=@n1.else, Visibility=Collapsed, self=Get Prev_Prompt())
   > Set Visibility(Visibility=Collapsed, self=Get Next_Prompt())
   > SetSize(self=[Get Next_Spacer(), Get Prev_Spacer()])
  ]}
 RemoveBackplate {nodes [
   RemoveBackplate() {is=title K2Node_FunctionEntry}
   > n1 = Get!(execute) -> (then, else, ContentBackplate:WBP UIKit Backplate Base Object Reference)
   > RemoveFromParent(self=@n1)
  ]}
 RemoveBackplateForPrompts {nodes [
   RemoveBackplateForPrompts() {is=title K2Node_FunctionEntry}
   > n1 = Sequence!(execute) -> (then_0, then_1)
   > n2 = Get!(execute) -> (then, else, PrevPromptBackplate:WBP UIKit Backplate Base Object Reference)
   > RemoveFromParent(self=@n2)
   n3 = Get!(execute=@n1.then_1) -> (then, else, NextPromptBackplate:WBP UIKit Backplate Base Object Reference)
   > RemoveFromParent(self=@n3)
  ]}
 RemoveCachedWidgets {nodes [
   RemoveCachedWidgets() {is=title K2Node_FunctionEntry}
   > Clear(TargetMap:Map of Names to Spacer Object References=Get CachedSpacers())
   n1 = Get CachedDividers()
   > n2 = Keys(TargetMap:Map of Names to S UI Divider Data Structures=@n1)
   > n3 = Values(TargetMap:Map of Names to S UI Divider Data Structures=@n1) -> (Values:Array of S UI Divider Data Structures)
   > n4 = For Each Loop(Array:Array of Names=@n2) -> (Array Element:Name)
   > Add(
    TargetMap=@n1
    Key=@n4
    Value_IsAddedAfter_3_#2="Get (a ref)"!(Array:Array of S UI Divider Data Structures=@n3, Dimension 1=@n4.Array Index) -> (
     Output_Widget_6_#1:WBP UIKit Divider Base Object Reference
     Output_IsAddedAfter_3_#2:bool=~
    ).Output_IsAddedAfter_3_#2
   )
  ]}
 RemoveCellFill {nodes [
   n1 = RemoveCellFill!() -> (then, TabNameID:Name) {is=title K2Node_FunctionEntry}
   > Branch(Condition=IsFillEnabled() {is=title K2Node_CallFunction})
   n2 = Find(TargetMap:Map of Names to Integers=Get CachedButtonIndex(), Key=@n1) -> (Value:int)
   > Branch(Condition=@n2.ReturnValue)
   > n3 = Branch(Condition=Get IsVertical())
   n4 = int + int(A=int * int(A=@n2, B=2))
   n5 = Get TabsContainer()
   > n6 = SetRowFill(self=@n5, RowIndex=@n4, Coefficient=0)
   Remove(execute=[@n7.then, @n6.then], TargetMap:Map of Names to Integers=Get CachedButtonIndex(), Key=@n1)
   n7 = SetColumnFill(execute=@n3.else, self=@n5, ColumnIndex=@n4, Coefficient=0)
  ]}
 RemoveDivider {nodes [
   n1 = RemoveDivider!() -> (then, TabID:Name) {is=title K2Node_FunctionEntry}
   n2 = Find(TargetMap:Map of Names to S UI Divider Data Structures=Get CachedDividers(), Key=@n1) -> (Value:S UI Divider Data Structure)
   > n3 = Branch(Condition=@n2.ReturnValue)
   > RemoveFromParent(self=Break S UI Divider Data(S_UI_DividerData=@n2) -> (IsAddedAfter_3_#2=~))
   > Remove(TargetMap:Map of Names to S UI Divider Data Structures=Get CachedDividers(), Key=@n1)
   > Return Node!(execute, DividerRemoved:bool=true)
   Return Node(execute=@n3.else, DividerRemoved=false)
  ]}
 RemoveDividerByIndex {nodes [
   n1 = RemoveDividerByIndex!() -> (then, TabIndex:int) {is=title K2Node_FunctionEntry}
   > n2 = RemoveDivider!(execute, self, TabID:Name=GetTabIDbyIndex(Index=@n1) -> (ReturnValue=None)) -> (then, DividerRemoved:bool) {is=title K2Node_CallFunction}
   > Return Node(DividerRemoved=@n2)
  ]}
 RemoveDividerWidget {nodes [
   n1 = RemoveDividerWidget!() -> (then, TabID:Name) {is=title K2Node_FunctionEntry}
   n2 = Find(TargetMap=Get CachedDividers(), Key=@n1)
   > Branch(Condition=@n2.ReturnValue)
   n3 = Break S UI Divider Data(S_UI_DividerData=@n2) -> (IsAddedAfter_3_#2=~)
   > RemoveFromParent(self=@n3)
   > Add!(
    execute
    TargetMap=Get CachedDividers()
    Key=@n1
    Value:"S UI Divider Data Structure (by ref)"=Make S UI Divider Data(IsAddedAfter_3_#2=@n3.IsAddedAfter_3_#2)
   ) -> (then)
  ]}
 RemoveSpacer {nodes [
   n1 = RemoveSpacer!() -> (then, TabID:Name) {is=title K2Node_FunctionEntry}
   n2 = Find(TargetMap:Map of Names to Spacer Object References=Get CachedSpacers(), Key=@n1) -> (Value:Spacer Object Reference)
   > Branch(Condition=@n2.ReturnValue)
   > RemoveFromParent(self=@n2)
   > Remove(TargetMap:Map of Names to Spacer Object References=Get CachedSpacers(), Key=@n1)
  ]}
 RemoveTabByIndex {nodes [
   n1 = RemoveTabByIndex!() -> (then, Index:int) {is=title K2Node_FunctionEntry}
   > n2 = RemoveTab!(execute, self:Common Tab List Widget Base Object Reference, TabNameID:Name=GetTabIDbyIndex(Index=@n1) -> (ReturnValue=None)) -> (then, ReturnValue:bool)
   > Return Node!(execute, TabRemoved:bool=@n2)
  ]}
 RemoveTabGroupWidgets {nodes [
   RemoveTabGroupWidgets() {is=title K2Node_FunctionEntry}
   > RemoveAllTabs(self:Common Tab List Widget Base Object Reference)
   n1 = Get TabsContainer()
   > ClearChildren(self=@n1)
   > ClearFill(self:Grid Panel Object Reference=@n1)
   > RemoveCachedWidgets() {is=title K2Node_CallFunction}
  ]}
 RemoveTabs {nodes [
   RemoveTabs() {is=title K2Node_FunctionEntry}
   > Clear(TargetMap:Map of Names to MVVM Base Viewmodel Object References=Get CachedViewModels())
   > Clear(TargetMap:Map of Names to Integers=Get CachedButtonIndex())
   > RemoveTabGroupWidgets() {is=title K2Node_CallFunction}
  ]}
 RestoreTabSelection {nodes [
   n1 = RestoreTabSelection!() -> (then, TabNameID:Name) {is=title K2Node_FunctionEntry}
   > n2 = Cast To WBP_UIKit_ButtonGeneric_Base(Object=GetTabButtonBaseByID(TabNameID=@n1))
   > n3 = RequestInstantSelection(self=@n2)
   SelectTabByID(execute=[@n3.then, @n2.CastFailed], TabNameID=@n1, bSuppressClickFeedback=true)
  ]}
 ScrollOnTabSelection {nodes [
   n1 = ScrollOnTabSelection!() -> (then, TabNameID:Name) {is=title K2Node_FunctionEntry}
   n2 = Break Margin!(Margin:"Margin Structure (by ref)"=Get TabGroupPadding() -> (TabGroupPadding:Margin Structure)) -> (Left:float=~, Top:float=~) {is=title K2Node_BreakStruct}
   > ScrollWidgetIntoView!(
    execute
    self:Scroll Box Object Reference=Get ScrollBox_Tabs() -> (ScrollBox_Tabs:Scroll Box Object Reference)
    WidgetToFind:Widget Object Reference=GetTabButtonBaseByID(TabNameID=@n1)
    AnimateScroll:bool=true
    ScrollDestination:EDescendantScrollDestination Enum=IntoView
    Padding:float=SelectFloat(A=@n2.Top, B=@n2, bPickA=Get IsVertical())
   ) -> (then)
  ]}
 Sequencer Events {nodes [
   Repeater_PromptSpaceChange() {is=title K2Node_CustomEvent}
   > n1 = Branch(Condition=Get IsVertical())
   n2 = float * float(
    A=GetAnimationCurrentTime!(
     self:User Widget Object Reference
     InAnimation:Widget Animation Object Reference=Get Anim_PromptSpaceChange() -> (Anim_PromptSpaceChange:Widget Animation Object Reference)
    ) -> (ReturnValue:float)
    B=Get ReservedPromptSpace()
   ) {is=title K2Node_PromotableOperator}
   > SetSize(self=[Get Next_Spacer(), Get Prev_Spacer()], InSize=MakeVector2D(X=0, Y=@n2))
   SetSize(execute=@n1.else, self=[Get Next_Spacer(), Get Prev_Spacer()], InSize=MakeVector2D(X=@n2, Y=0))
   Trigger_Start() {is=title K2Node_CustomEvent}
   > SetSize(self=[Get Next_Spacer(), Get Prev_Spacer()])
   Trigger_End() {is=title K2Node_CustomEvent}
   > n3 = Branch(Condition=Get IsPromptsVisible() -> (IsPromptsVisible:bool))
   > Set Visibility(Visibility=Visible, self=Get Next_Prompt())
   > n4 = Set Visibility(Visibility=Visible, self=Get Prev_Prompt())
   n5 = Branch(execute=[@n4.then, @n7.then], Condition=Get IsVertical())
   n6 = Get ReservedPromptSpace()
   > SetSize(self=[Get Prev_Spacer(), Get Next_Spacer()], InSize=MakeVector2D(X=0, Y=@n6))
   SetSize(execute=@n5.else, self=[Get Next_Spacer(), Get Prev_Spacer()], InSize=MakeVector2D(X=@n6, Y=0))
   Set Visibility(execute=@n3.else, Visibility=Collapsed, self=Get Next_Prompt())
   > n7 = Set Visibility(Visibility=Collapsed, self=Get Prev_Prompt())
  ]}
 SetBackplateProperties {nodes [
   n1 = SetBackplateProperties!() -> (
    then
    Backplate:WBP UIKit Backplate Base Object Reference
    Brightness:E_UI_BackplateBrightness Enum=NewEnumerator0
    Corner Radius:E_UI_BackplateCornerRadius Enum=NewEnumerator0
   ) {is=title K2Node_FunctionEntry}
   > SetBrightness!(execute, self:WBP UIKit Backplate Base Object Reference=@n1, Brightness:E_UI_BackplateBrightness Enum=@n1.Brightness) -> (then)
   > SetCornerRadius!(execute, self:WBP UIKit Backplate Base Object Reference=@n1, Corner Radius:E_UI_BackplateCornerRadius Enum=@n1.Corner Radius) -> (then)
  ]}
 SetInputActionsType {nodes [
   n1 = SetInputActionsType!() -> (then, InputActions:E_UI_TabsInputActionType Enum=NewEnumerator0, InputActionsWithMouseScroll:bool=~) {is=title K2Node_FunctionEntry}
   > Set InputActions(InputActions:E_UI_TabsInputActionType Enum=@n1) -> (Output_Get:E_UI_TabsInputActionType Enum=NewEnumerator2)
   > Set InputActionsWithMouseScroll(InputActionsWithMouseScroll:bool=@n1.InputActionsWithMouseScroll) -> (Output_Get:bool)
   > UpdateInputActions()
  ]}
 SetPromptsVisibility {nodes [
   n1 = SetPromptsVisibility!() -> (then, ShowPrompts:bool=~) {is=title K2Node_FunctionEntry}
   > Branch(Condition="Not Equal (Boolean)"(A=Get IsPromptsVisible(), B=@n1) -> (ReturnValue=~))
   > n2 = Set IsPromptsVisible(IsPromptsVisible:bool=@n1) -> (Output_Get:bool)
   > n3 = Branch(Condition=@n2)
   > n4 = Branch(Condition=Get AnimateReservedPromptSpace() -> (AnimateReservedPromptSpace:bool))
   > PlayAnimationForward(InAnimation=Get Anim_PromptSpaceChange(), PlaybackSpeed=SafeDivide(A=1, B=Get TimeToShowPrompts() -> (TimeToShowPrompts:double)))
   Set Visibility(execute=@n4.else, Visibility=Visible, self=Get Prev_Prompt())
   > Set Visibility(Visibility=Visible, self=Get Next_Prompt())
   n5 = Get ReservedPromptSpace()
   > SetSize(
    self=[Get Next_Spacer(), Get Prev_Spacer()]
    InSize=Select(Option 0=MakeVector2D(X=@n5, Y=0), Option 1=MakeVector2D(X=0, Y=@n5), Index=Get IsVertical())
   )
   n6 = Branch(execute=@n3.else, Condition=Get AnimateReservedPromptSpace())
   > PlayAnimationReverse(InAnimation=Get Anim_PromptSpaceChange(), PlaybackSpeed=SafeDivide(A=1, B=Get TimeToHidePrompts() -> (TimeToHidePrompts:double)))
   Set Visibility(execute=@n6.else, Visibility=Collapsed, self=Get Prev_Prompt())
   > Set Visibility(Visibility=Collapsed, self=Get Next_Prompt())
   > SetSize(self=[Get Next_Spacer(), Get Prev_Spacer()])
  ]}
 SetTabViewModel {nodes [
   n1 = SetTabViewModel!() -> (then, TabID:Name, VM:Notify Field Value Changed Interface) {is=title K2Node_FunctionEntry}
   n2 = Get View From User Widget!(
    self:Viewmodel Engine Subsytem Object Reference=Get MVVMSubsystem() -> (ReturnValue:Viewmodel Engine Subsytem Object Reference) {is=title K2Node_GetEngineSubsystem}
    UserWidget:User Widget Object Reference=GetTabButtonBaseByID(TabNameID=@n1)
   ) -> (ReturnValue:MVVM View Object Reference) {is=title K2Node_CallFunction}
   > Branch(Condition=IsValid(Object=@n2))
   > SetViewModelByClass!(execute, self:MVVM View Object Reference=@n2, NewValue:Notify Field Value Changed Interface=@n1.VM) -> (then, ReturnValue:bool)
  ]}
 SetTabsAlignment {nodes [
   Inputs!() -> (execute:Exec)
   n1 = Get TabsContainer()
   > ClearFill(self=@n1)
   n2 = Get IsVertical()
   > SetColumnFill(self=@n1, ColumnIndex=0, Coefficient=SelectFloat(A=1, B=0, bPickA=@n2))
   > SetRowFill(self=@n1, RowIndex=0, Coefficient=SelectFloat(A=0, B=1, bPickA=@n2))
   > n3 = Cast To ScrollBoxSlot!(execute, Object=Get Slot!(self:Widget Object Reference=Get TabsContainer()) -> (Slot:Panel Slot Object Reference)) -> (then, CastFailed, AsScroll Box Slot:Scroll Box Slot Object Reference) {is=title K2Node_DynamicCast}
   n4 = IsFillEnabled()
   > n5 = Set Size!(
    execute
    Size:Slate Child Size Structure=Make Slate Child Size!(
     Value:float=1
     SizeRule:ESlateSizeRule Enum=Select(Option 0:ESlateSizeRule Enum=Automatic, Option 1:ESlateSizeRule Enum=Fill, Index=@n4) -> (ReturnValue:ESlateSizeRule Enum)
    ) -> (SlateChildSize:Slate Child Size Structure) {is=title K2Node_MakeStruct}
    self:Scroll Box Slot Object Reference=@n3
   ) -> (then, Output_Get:Slate Child Size Structure)
   Outputs!(then:Exec=[@n5.then, @n3.CastFailed])
  ]}
 SetupDivider {nodes [
   n1 = SetupDivider!() -> (then, InputPin:WBP UIKit Divider Base Object Reference) {is=title K2Node_FunctionEntry}
   > SetDividerOrientation!(execute, self:WBP UIKit Divider Base Object Reference=@n1, IsVertical:bool=Get IsVertical()) -> (then)
   > SetDividerSpacing!(
    execute
    self:WBP UIKit Divider Base Object Reference=@n1
    Spacing:double=Get DividerSpacing() -> (DividerSpacing:double)
    IsVertical:bool=Get IsVertical()
   ) -> (then)
   > SetPadding(self:User Widget Object Reference=@n1, InPadding=Get DividerPadding() -> (DividerPadding:Margin Structure))
  ]}
 SetupTabButton {nodes [
   n1 = SetupTabButton!() -> (then, Tab ID:Name, Tab Button:Common Button Base Object Reference) {is=title K2Node_FunctionEntry}
   > n2 = Sequence()
   > SetStyle!(
    execute
    self:Common Button Base Object Reference=@n1.Tab Button
    InStyle:Common Button Style Class Reference=&ButtonStyle_UIKit_Empty_C
   ) -> (then)
   > SetIsFocusable!(
    execute
    self:Common Button Base Object Reference=@n1.Tab Button
    bInIsFocusable:bool=Get IsNaturalNavigation() -> (IsNaturalNavigation:bool)
   ) -> (then)
   n3 = Cast To WBP_UIKit_ButtonGeneric_Base(execute=@n2.then_1, Object=@n1.Tab Button)
   > SetIconSize!(execute, self:WBP UIKit Button Generic Base Object Reference=@n3, Icon:Vector 2D Structure=Get IconSize() -> (IconSize:Vector 2D Structure)) -> (then)
   > n4 = Branch(Condition=Get UseMinButtonSize() -> (UseMinButtonSize:bool))
   n5 = Get!() -> (MinButtonSize_X:int, MinButtonSize_Y:int)
   > n6 = SetButtonMinSize!(execute, self:WBP UIKit Button Generic Base Object Reference=@n3, Min Width:int=@n5, Min Height:int=@n5.MinButtonSize_Y) -> (then)
   Branch(execute=[@n6.then, @n4.else], Condition=InEditor())
   > SetPreviewData!(
    execute
    self:WBP UIKit Button Generic Base Object Reference=@n3
    VM:MVVM Base Viewmodel Object Reference=GetTabViewModel!(self:UIKit Tab Group Object Reference, TabID:Name=Get Tab ID() -> (Tab ID:Name=None)) -> (ReturnValue:MVVM Base Viewmodel Object Reference)
   ) -> (then)
  ]}
 UpdateArrowButtonsEnabled {nodes [
   UpdateArrowButtonsEnabled() {is=title K2Node_FunctionEntry}
   > Branch(Condition=Get UseArrowNavButtons())
   > n1 = Branch(Condition=Get ArrowButtonsAllowWrap())
   > SetIsEnabled(self=[Get Next_ArrowButton(), Get Prev_ArrowButton()], bInIsEnabled=true)
   n2 = GetSelectedTabIndex() -> (ReturnValue=0)
   n3 = GetTabCount() -> (ReturnValue=0)
   SetIsEnabled(execute=@n1.else, self=Get Next_ArrowButton(), bInIsEnabled="Not Equal (Integer)"(A=int - int(A=@n3), B=@n2))
   > SetIsEnabled(self=Get Prev_ArrowButton(), bInIsEnabled="Not Equal (Integer)"(A=@n2))
  ]}
 UpdateArrowButtonsVisibility {nodes [
   n1 = UpdateArrowButtonsVisibility!() -> (then, CurrentInputType:ECommonInputType Enum) {is=title K2Node_FunctionEntry}
   n2 = Select(
    Option 0:bool=true
    Option 1:bool="Not Equal (Enum)"!(A:ECommonInputType Enum=@n1, B:ECommonInputType Enum=Gamepad) -> (ReturnValue:bool)
    Index=Get GamepadHideArrows()
   ) -> (ReturnValue:bool)
   n3 = Get UseArrowNavButtons()
   > SetVisibility(
    self=[Get Prev_ArrowButton(), Get Next_ArrowButton()]
    InVisibility=Select(Option 0:ESlateVisibility Enum=Collapsed, Option 1:ESlateVisibility Enum=SelfHitTestInvisible, Index=AND Boolean(A=@n3, B=@n2)) -> (ReturnValue:ESlateVisibility Enum)
   )
   > UpdateArrowButtonsEnabled() {is=title K2Node_CallFunction}
  ]}
 UpdateInputActions {nodes [
   UpdateInputActions() {is=title K2Node_FunctionEntry}
   > n1 = Switch on E_UI_TabsInputActionType!(execute, Selection:E_UI_TabsInputActionType Enum=Get InputActions() -> (InputActions=NewEnumerator2)) -> (NewEnumerator2:Exec, NewEnumerator0:Exec, NewEnumerator1:Exec, NewEnumerator3:Exec) {is=title K2Node_SwitchEnum}
   n2 = Make Data Table Row Handle!(DataTable:Data Table Object Reference, RowName:Name=None) -> (DataTableRowHandle:Data Table Row Handle Structure) {is=title K2Node_MakeStruct}
   > OverrideInputActions!(execute, self, Next Tab:Data Table Row Handle Structure=@n2, PreviousTab:Data Table Row Handle Structure=@n2) -> (then) {is=title K2Node_CallFunction}
   n3 = Get InputActionsWithMouseScroll() -> (InputActionsWithMouseScroll:bool)
   OverrideInputActions(
    execute=@n1.NewEnumerator0
    Next Tab=Select(
     Option 0:Data Table Row Handle Structure=Get Bumper_NextTab() -> (Bumper_NextTab:Data Table Row Handle Structure)
     Option 1:Data Table Row Handle Structure=Get BumperAndMouseScroll_NextTab() -> (BumperAndMouseScroll_NextTab:Data Table Row Handle Structure)
     Index=@n3
    ) -> (ReturnValue:Data Table Row Handle Structure)
    PreviousTab=Select(
     Option 0=Get Bumper_PrevTab() -> (Bumper_PrevTab:Data Table Row Handle Structure)
     Option 1=Get BumperAndMouseScroll_PrevTab() -> (BumperAndMouseScroll_PrevTab:Data Table Row Handle Structure)
     Index=@n3
    )
   )
   n4 = Get InputActionsWithMouseScroll()
   OverrideInputActions(
    execute=@n1.NewEnumerator1
    Next Tab=Select(
     Option 0=Get Trigger_NextTab() -> (Trigger_NextTab:Data Table Row Handle Structure)
     Option 1=Get TriggerAndMouseScroll_NextTab() -> (TriggerAndMouseScroll_NextTab:Data Table Row Handle Structure)
     Index=@n4
    )
    PreviousTab=Select(
     Option 0=Get Trigger_PrevTab() -> (Trigger_PrevTab:Data Table Row Handle Structure)
     Option 1=Get TriggerAndMouseScroll_PrevTab() -> (TriggerAndMouseScroll_PrevTab:Data Table Row Handle Structure)
     Index=@n4
    )
   )
   SetInputAction(
    execute=@n1.NewEnumerator3
    self=Get NextTabAction()
    InputActionRow=Get NextTabInputActionData() -> (NextTabInputActionData:Data Table Row Handle Structure)
   )
   > SetInputAction(self=Get PrevTabAction(), InputActionRow=Get PreviousTabInputActionData() -> (PreviousTabInputActionData:Data Table Row Handle Structure))
  ]}
 UpdatePromptsVisibility {nodes [
   UpdatePromptsVisibility() {is=title K2Node_FunctionEntry}
   > SetPromptsVisibility!(execute, self, ShowPrompts:bool=GetPromptsVisibility()) -> (then) {is=title K2Node_CallFunction}
  ]}
 UpdateTabLayout {nodes [
   UpdateTabLayout() {is=title K2Node_FunctionEntry}
   n1 = Get IsVertical()
   n2 = SelectInt(A=0, B=1, bPickA=@n1)
   n3 = SelectInt(A=1, B=0, bPickA=@n1)
   > Set Grid Slot Column and Row!(execute:Exec, Widget:Widget Object Reference=Get Prev_Prompt(), Row:int=@n2, RowSpan:int, Column:int=@n3, ColumnSpan:int) -> (then:Exec) {is=title K2Node_MacroInstance}
   > Set Grid Slot Column and Row(Widget=Get Prev_Spacer(), Row=@n2, Column=@n3)
   n4 = Get Prompt Position Offset Value!(DefaultValue:int=2, AddValue:int=2) -> (ReturnValue:int) {is=title K2Node_MacroInstance}
   n5 = Get IsVertical()
   n6 = SelectInt(A=1, B=@n4, bPickA=@n5)
   n7 = SelectInt(A=@n4, B=1, bPickA=@n5)
   > Set Grid Slot Column and Row(Widget=Get Next_Prompt(), Row=@n7, Column=@n6)
   > Set Grid Slot Column and Row(Widget=Get Next_Spacer(), Row=@n7, Column=@n6)
   > SetPadding(
    self:Scroll Box Slot Object Reference=SlotAsScrollBoxSlot!(Widget:Widget Object Reference=Get TabsContainer()) -> (ReturnValue:Scroll Box Slot Object Reference)
    InPadding=Get TabGroupPadding()
   )
   > Set Grid Slot Alignment!(
    execute:Exec
    Widget:Widget Object Reference=Get Prev_Prompt()
    Horizontal:EHorizontalAlignment Enum=Get PrevPromptHorizontalAlignment() -> (PrevPromptHorizontalAlignment:EHorizontalAlignment Enum=HAlign_Fill)
    Vertical:EVerticalAlignment Enum=Get PrevPromptVerticalAlignment() -> (PrevPromptVerticalAlignment:EVerticalAlignment Enum=VAlign_Fill)
   ) -> (then:Exec) {is=title K2Node_MacroInstance}
   > Set Grid Slot Alignment(
    Widget=Get Next_Prompt()
    Horizontal=Get NextPromptHorizontalAlignment() -> (NextPromptHorizontalAlignment:EHorizontalAlignment Enum=HAlign_Fill)
    Vertical=Get NextPromptVerticalAlignment() -> (NextPromptVerticalAlignment:EVerticalAlignment Enum=VAlign_Fill)
   )
   > Set Grid Slot Padding(Widget=Get Prev_Prompt(), Padding=Get PrevPromptPadding() -> (PrevPromptPadding:Margin Structure))
   > Set Grid Slot Padding(Widget=Get Next_Prompt(), Padding=Get NextPromptPadding() -> (NextPromptPadding:Margin Structure))
   n8 = Get ScrollBox_Tabs()
   > Set Grid Slot Alignment(
    Widget=@n8
    Horizontal=Get TabGroupHorizontalAlignment() -> (TabGroupHorizontalAlignment=HAlign_Fill)
    Vertical=Get TabGroupVerticalAlignment() -> (TabGroupVerticalAlignment=VAlign_Fill)
   )
   n9 = Get Prompt Position Offset Value(DefaultValue=1, AddValue=1)
   n10 = Get IsVertical()
   > Set Grid Slot Column and Row(
    Widget=@n8
    Row=Select(Option 0:int=1, Option 1:int=@n9, Index=@n10) -> (ReturnValue:int)
    Column=Select(Option 0=@n9, Option 1=1, Index=@n10)
   )
   > SetOrientation!(
    execute
    self:Scroll Box Object Reference=@n8
    NewOrientation:EOrientation Enum=Select(Option 0:EOrientation Enum=Orient_Horizontal, Option 1:EOrientation Enum=Orient_Vertical, Index=Get IsVertical()) -> (ReturnValue:EOrientation Enum)
   ) -> (then)
   > SetScrollBarVisibility!(
    execute
    self:Scroll Box Object Reference=@n8
    NewScrollBarVisibility:ESlateVisibility Enum=Get ScrollBarVisibility() -> (ScrollBarVisibility:ESlateVisibility Enum=Visible)
   ) -> (then)
   > SetClipping!(
    execute
    self:Widget Object Reference=Get ScrollBox_Tabs()
    InClipping:EWidgetClipping Enum=Get ScrollClipping() -> (ScrollClipping:EWidgetClipping Enum=Inherit)
   ) -> (then)
   > Branch(Condition=Get UseArrowNavButtons())
   n11 = Get Prompt Position Offset Value(AddValue=1)
   n12 = Get IsVertical()
   > Set Grid Slot Column and Row(Widget=Get Prev_ArrowButton(), Row=SelectInt(A=@n11, B=1, bPickA=@n12), Column=SelectInt(A=1, B=@n11, bPickA=@n12))
   n13 = Get Prompt Position Offset Value(DefaultValue=2, AddValue=1)
   n14 = Get IsVertical()
   > Set Grid Slot Column and Row(Widget=Get Next_ArrowButton(), Row=SelectInt(A=@n13, B=1, bPickA=@n14), Column=SelectInt(A=1, B=@n13, bPickA=@n14))
   n15 = Get IsVertical()
   > SetRenderTransformAngle!(execute, self:Widget Object Reference=Get Next_ArrowButton(), Angle:float=SelectFloat(A=90, B=0, bPickA=@n15)) -> (then)
   > SetRenderTransformAngle(self=Get Prev_ArrowButton(), Angle=SelectFloat(A=-90, B=180, bPickA=@n15))
   > Set Grid Slot Alignment(
    Widget=Get Prev_ArrowButton()
    Horizontal=Get PrevPromptHorizontalAlignment() -> (PrevPromptHorizontalAlignment=HAlign_Fill)
    Vertical=Get PrevPromptVerticalAlignment() -> (PrevPromptVerticalAlignment=VAlign_Fill)
   )
   > Set Grid Slot Alignment(
    Widget=Get Next_ArrowButton()
    Horizontal=Get NextPromptHorizontalAlignment() -> (NextPromptHorizontalAlignment=HAlign_Fill)
    Vertical=Get NextPromptVerticalAlignment() -> (NextPromptVerticalAlignment=VAlign_Fill)
   )
   > Set Grid Slot Padding(Widget=Get Prev_ArrowButton(), Padding=Get PrevPromptPadding())
   > Set Grid Slot Padding(Widget=Get Next_ArrowButton(), Padding=Get NextPromptPadding())
  ]}
]
variables [
 AnimateReservedPromptSpace {type=boolean, default=true}
 ArrowButtonsAllowWrap {type=boolean, default=true}
 ArrowButtonsClass {type=Class, default=/FortUILibrary/UIKit/Buttons/TabButtons/WBP_UIKit_Button_TabNavArrow.WBP_UIKit_Button_TabNavArrow_C}
 BackplateBrightness {type=E_UI_BackplateBrightness, default=Medium}
 BackplateClass {type=Class, default=/FortUILibrary/UIKit/Backplates/WBP_UIKit_Backplate.WBP_UIKit_Backplate_C}
 BackplateCornerRadius {type=E_UI_BackplateCornerRadius, default=Backplate}
 BackplatePadding {type=Margin, default=<ltrb 0 0 0 0>}
 Bumper_NextTab {type=DataTableRowHandle, default {dataTable=&InputActionDataTable, rowName=MainTabRight}}
 Bumper_PrevTab {type=DataTableRowHandle, default {dataTable=&InputActionDataTable, rowName=MainTabLeft}}
 BumperAndMouseScroll_NextTab {type=DataTableRowHandle, default {dataTable=&InputActionDataTable, rowName=MainTabRight_WithMouse}}
 BumperAndMouseScroll_PrevTab {type=DataTableRowHandle, default {dataTable=&InputActionDataTable, rowName=MainTabLeft_WithMouse}}
 ButtonClass {type=Class, default=/FortUILibrary/UIKit/Buttons/TabButtons/WBP_UIKit_Button_Tab.WBP_UIKit_Button_Tab_C}
 ButtonClassOverride {type="map<Class>", default {}}
 ButtonHorizontalAlignment {type=EHorizontalAlignment, default=HAlign_Center}
 ButtonPadding {type=Margin, default=<ltrb 0 0 0 0>}
 ButtonVerticalAlignment {type=EVerticalAlignment, default=VAlign_Center}
 CachedButtonIndex {type="map<integer>", default {}}
 CachedDividers {type="map<S_UI_DividerData>", default {}}
 CachedSpacers {type="map<Spacer>", default {}}
 ContentBackplate {type=WBP_UIKit_Backplate_Base_C, default=null}
 CurrentInputType {type=ECommonInputType, default=MouseAndKeyboard}
 DividerClass {type=Class, default=&WBP_UIKit_Divider_Base_C}
 DividerPadding {type=Margin, default=<ltrb 0 0 0 0>}
 Dividers {type="array<S_UI_DividerPosition>", default []}
 DividerSpacing {type=number, default=24}
 GamepadHideArrows {type=boolean, default=true}
 HideMouseAndKeyboardPrompts {type=boolean}
 HidePrompts {type=boolean}
 IconSize {type=Vector2D, default=<32 32>}
 InitialSelectedTabID {type=string, default=None}
 InitialSelectedTabIndex {type=integer, default=-1}
 InitialTabsData {type="array<UIKitTabData>", default []}
 InputActions {type=E_UI_TabsInputActionType, default=Bumpers}
 InputActionsWithMouseScroll {type=boolean}
 IsNaturalNavigation {type=boolean}
 IsPromptsVisible {type=boolean, default=true}
 IsVertical {type=boolean}
 MinButtonSize {type=IntPoint, default=<0 0>}
 Next_ArrowButton {type=WBP_UIKit_ButtonGeneric_Base_C, default=null}
 NextPromptBackplate {type=WBP_UIKit_Backplate_Base_C, default=null}
 NextPromptHorizontalAlignment {type=EHorizontalAlignment, default=HAlign_Center}
 NextPromptPadding {type=Margin, default=<ltrb 0 0 0 0>}
 NextPromptVerticalAlignment {type=EVerticalAlignment, default=VAlign_Center}
 Prev_ArrowButton {type=WBP_UIKit_ButtonGeneric_Base_C, default=null}
 PrevPromptBackplate {type=WBP_UIKit_Backplate_Base_C, default=null}
 PrevPromptHorizontalAlignment {type=EHorizontalAlignment, default=HAlign_Center}
 PrevPromptPadding {type=Margin, default=<ltrb 0 0 0 0>}
 PrevPromptVerticalAlignment {type=EVerticalAlignment, default=VAlign_Center}
 PromptBackplateBrightness {type=E_UI_BackplateBrightness, default=Medium}
 PromptBackplateCornerRadius {type=E_UI_BackplateCornerRadius, default=Circle}
 PromptBackplateMinSize {type=Vector2D, default=<0 0>}
 PromptsBackplateClass {type=Class, default=&WBP_UIKit_Backplate_C}
 PromptsInsideTabsBackplate {type=boolean}
 ReservedPromptSpace {type=number, default=70}
 ScrollBarVisibility {type=ESlateVisibility, default=Collapsed}
 ScrollClipping {type=EWidgetClipping, default=ClipToBounds}
 TabGroupHorizontalAlignment {type=EHorizontalAlignment, default=HAlign_Left}
 TabGroupPadding {type=Margin, default=<ltrb 8 8 8 8>}
 TabGroupVerticalAlignment {type=EVerticalAlignment, default=VAlign_Center}
 TabSpacing {type=number, default=30}
 TimeToHidePrompts {type=number, default=0.2}
 TimeToShowPrompts {type=number, default=0.2}
 Trigger_NextTab {type=DataTableRowHandle, default {dataTable=&InputActionDataTable, rowName=SecondaryTabRight}}
 Trigger_PrevTab {type=DataTableRowHandle, default {dataTable=&InputActionDataTable, rowName=SecondaryTabLeft}}
 TriggerAndMouseScroll_NextTab {type=DataTableRowHandle, default {dataTable=&InputActionDataTable, rowName=SecondaryTabRight_WithMouse}}
 TriggerAndMouseScroll_PrevTab {type=DataTableRowHandle, default {dataTable=&InputActionDataTable, rowName=SecondaryTabLeft_WithMouse}}
 UseArrowNavButtons {type=boolean}
 UseBackplate {type=boolean, default=true}
 UseInitialTabs {type=boolean}
 UseMinButtonSize {type=boolean, default=true}
 UsePromptBackplate {type=boolean, default=true}
]
widgets {
 info {inheritedWidgetCount=0, parentClass=/Script/FortniteUI.UIKitTabGroup, rootWidgetClass=&GridPanel, widgetCount=11}
 widgets [
  Root {widgetClassPath=&GridPanel, bIsVariable=true, widget=Root, properties {columnFill [0,1,0], rowFill [0,1,0]}}
  Prev_Prompt {
   widgetClassPath=&GridPanel
   bIsVariable=true
   parent=Root
   widget=Prev_Prompt
   properties {columnFill [1], rowFill [1]}
   slot {class=GridSlot, properties {layer=10, row=1}}
  }
  Prev_SizeBox {widgetClassPath=&SizeBox, bIsVariable=true, parent=Prev_Prompt, widget=Prev_SizeBox, slot {class=GridSlot}}
  PrevTabAction {
   widgetClassPath=&CommonActionWidget
   bIsVariable=true
   parent=Prev_Prompt
   widget=PrevTabAction
   properties {visibility=Collapsed}
   slot {class=GridSlot, properties {horizontalAlignment=HAlign_Center, layer=10, verticalAlignment=VAlign_Center}}
  }
  Prev_Spacer {
   widgetClassPath=&Spacer
   bIsVariable=true
   parent=Root
   widget=Prev_Spacer
   properties {visibility=Visible, size=<48 32>}
   slot {class=GridSlot, properties {horizontalAlignment=HAlign_Center, row=1, verticalAlignment=VAlign_Center}}
  }
  ScrollBox_Tabs {
   widgetClassPath=&ScrollBox
   bIsVariable=true
   parent=Root
   widget=ScrollBox_Tabs
   properties {
    clipping=OnDemand
    scrollWhenFocusChanges=AnimatedScroll
    scrollbarPadding=<ltrb 0 0 0 0>
    scrollbarThickness=<16 4>
    widgetBarStyle {
     thickness=12
     draggedThumbImage {
      bIsDynamicallyLoaded=false
      drawAs=RoundedBox
      imageType=FullColor
      mirroring=NoMirror
      resourceName=../../../Engine/Content/Slate/Common/Scrollbar_Thumb.png
      resourceObject=null
      tiling=NoTile
      imageSize=<12 12>
      margin=<ltrb 0.25 0.25 0.25 0.25>
      outlineSettings {
       bUseBrushTransparency=false
       roundingType=FixedRadius
       width=2
       color {colorUseRule=UseColor_Specified, specifiedColor=<rgba 0 0 0 0>}
       cornerRadii=<6 6 6 6>
      }
      tintColor {colorUseRule=UseColor_Specified, specifiedColor=<rgba 1 1 1 1>}
      uVRegion {bIsValid=false, max=<0 0>, min=<0 0>}
     }
     horizontalBackgroundImage {
      bIsDynamicallyLoaded=false
      drawAs=Box
      imageType=FullColor
      mirroring=NoMirror
      resourceName=None
      resourceObject=/FortUILibrary/UIKit/ScrollBar/MaterialInstances/MI_UIKit_ScrollbarBackground
      tiling=NoTile
      imageSize=<12 12>
      margin=<ltrb 0.5 0.5 0.5 0.5>
      outlineSettings {
       bUseBrushTransparency=false
       roundingType=HalfHeightRadius
       width=0
       color {colorUseRule=UseColor_Specified, specifiedColor=<rgba 0 0 0 0>}
       cornerRadii=<0 0 0 0>
      }
      tintColor {colorUseRule=UseColor_Specified, specifiedColor=<rgba 1 1 1 1>}
      uVRegion {bIsValid=false, max=<0 0>, min=<0 0>}
     }
     horizontalBottomSlotImage {
      bIsDynamicallyLoaded=false
      drawAs=NoDrawType
      imageType=FullColor
      mirroring=NoMirror
      resourceName=../../../Engine/Content/Slate/Common/Scrollbar_Background_Horizontal.png
      resourceObject=null
      tiling=NoTile
      imageSize=<8 8>
      margin=<ltrb 0 0 0 0>
      outlineSettings {
       bUseBrushTransparency=false
       roundingType=HalfHeightRadius
       width=0
       color {colorUseRule=UseColor_Specified, specifiedColor=<rgba 0 0 0 0>}
       cornerRadii=<0 0 0 0>
      }
      tintColor {colorUseRule=UseColor_Specified, specifiedColor=<rgba 1 1 1 1>}
      uVRegion {bIsValid=false, max=<0 0>, min=<0 0>}
     }
     horizontalTopSlotImage {
      bIsDynamicallyLoaded=false
      drawAs=NoDrawType
      imageType=FullColor
      mirroring=NoMirror
      resourceName=../../../Engine/Content/Slate/Common/Scrollbar_Background_Horizontal.png
      resourceObject=null
      tiling=NoTile
      imageSize=<8 8>
      margin=<ltrb 0 0 0 0>
      outlineSettings {
       bUseBrushTransparency=false
       roundingType=HalfHeightRadius
       width=0
       color {colorUseRule=UseColor_Specified, specifiedColor=<rgba 0 0 0 0>}
       cornerRadii=<0 0 0 0>
      }
      tintColor {colorUseRule=UseColor_Specified, specifiedColor=<rgba 1 1 1 1>}
      uVRegion {bIsValid=false, max=<0 0>, min=<0 0>}
     }
     hoveredThumbImage {
      bIsDynamicallyLoaded=false
      drawAs=RoundedBox
      imageType=FullColor
      mirroring=NoMirror
      resourceName=../../../Engine/Content/Slate/Common/Scrollbar_Thumb.png
      resourceObject=null
      tiling=NoTile
      imageSize=<12 12>
      margin=<ltrb 0.25 0.25 0.25 0.25>
      outlineSettings {
       bUseBrushTransparency=false
       roundingType=FixedRadius
       width=2
       color {colorUseRule=UseColor_Specified, specifiedColor=<rgba 0 0 0 0>}
       cornerRadii=<6 6 6 6>
      }
      tintColor {colorUseRule=UseColor_Specified, specifiedColor=<rgba 1 1 1 1>}
      uVRegion {bIsValid=false, max=<0 0>, min=<0 0>}
     }
     normalThumbImage {
      bIsDynamicallyLoaded=false
      drawAs=RoundedBox
      imageType=FullColor
      mirroring=NoMirror
      resourceName=../../../Engine/Content/Slate/Common/Scrollbar_Thumb.png
      resourceObject=null
      tiling=NoTile
      imageSize=<12 12>
      margin=<ltrb 0.25 0.25 0.25 0.25>
      outlineSettings {
       bUseBrushTransparency=false
       roundingType=FixedRadius
       width=2
       color {colorUseRule=UseColor_Specified, specifiedColor=<rgba 0 0 0 0>}
       cornerRadii=<6 6 6 6>
      }
      tintColor {colorUseRule=UseColor_Specified, specifiedColor=<rgba 1 1 1 0.75>}
      uVRegion {bIsValid=false, max=<0 0>, min=<0 0>}
     }
     verticalBackgroundImage {
      bIsDynamicallyLoaded=false
      drawAs=Box
      imageType=NoImage
      mirroring=NoMirror
      resourceName=None
      resourceObject=&MI_UIKit_ScrollbarBackground
      tiling=NoTile
      imageSize=<12 12>
      margin=<ltrb 0.5 0.5 0.5 0.5>
      outlineSettings {
       bUseBrushTransparency=false
       roundingType=HalfHeightRadius
       width=0
       color {colorUseRule=UseColor_Specified, specifiedColor=<rgba 0 0 0 0>}
       cornerRadii=<0 0 0 0>
      }
      tintColor {colorUseRule=UseColor_Specified, specifiedColor=<rgba 1 1 1 1>}
      uVRegion {bIsValid=false, max=<0 0>, min=<0 0>}
     }
     verticalBottomSlotImage {
      bIsDynamicallyLoaded=false
      drawAs=NoDrawType
      imageType=FullColor
      mirroring=NoMirror
      resourceName=../../../Engine/Content/Slate/Common/Scrollbar_Background_Vertical.png
      resourceObject=null
      tiling=NoTile
      imageSize=<8 8>
      margin=<ltrb 0 0 0 0>
      outlineSettings {
       bUseBrushTransparency=false
       roundingType=HalfHeightRadius
       width=0
       color {colorUseRule=UseColor_Specified, specifiedColor=<rgba 0 0 0 0>}
       cornerRadii=<0 0 0 0>
      }
      tintColor {colorUseRule=UseColor_Specified, specifiedColor=<rgba 1 1 1 1>}
      uVRegion {bIsValid=false, max=<0 0>, min=<0 0>}
     }
     verticalTopSlotImage {
      bIsDynamicallyLoaded=false
      drawAs=NoDrawType
      imageType=FullColor
      mirroring=NoMirror
      resourceName=../../../Engine/Content/Slate/Common/Scrollbar_Background_Vertical.png
      resourceObject=null
      tiling=NoTile
      imageSize=<8 8>
      margin=<ltrb 0 0 0 0>
      outlineSettings {
       bUseBrushTransparency=false
       roundingType=HalfHeightRadius
       width=0
       color {colorUseRule=UseColor_Specified, specifiedColor=<rgba 0 0 0 0>}
       cornerRadii=<0 0 0 0>
      }
      tintColor {colorUseRule=UseColor_Specified, specifiedColor=<rgba 1 1 1 1>}
      uVRegion {bIsValid=false, max=<0 0>, min=<0 0>}
     }
    }
    widgetStyle {
     barThickness=12
     bottomShadowBrush {
      bIsDynamicallyLoaded=false
      drawAs=NoDrawType
      imageType=FullColor
      mirroring=NoMirror
      resourceName=../../../Engine/Content/Slate/Common/ScrollBoxShadowBottom.png
      resourceObject=null
      tiling=NoTile
      imageSize=<16 8>
      margin=<ltrb 0.5 0 0.5 1>
      outlineSettings {
       bUseBrushTransparency=false
       roundingType=HalfHeightRadius
       width=0
       color {colorUseRule=UseColor_Specified, specifiedColor=<rgba 0 0 0 0>}
       cornerRadii=<0 0 0 0>
      }
      tintColor {colorUseRule=UseColor_Specified, specifiedColor=<rgba 1 1 1 1>}
      uVRegion {bIsValid=false, max=<0 0>, min=<0 0>}
     }
     horizontalScrolledContentPadding=<ltrb 0 0 1 0>
     leftShadowBrush {
      bIsDynamicallyLoaded=false
      drawAs=NoDrawType
      imageType=FullColor
      mirroring=NoMirror
      resourceName=../../../Engine/Content/Slate/Common/ScrollBoxShadowLeft.png
      resourceObject=null
      tiling=NoTile
      imageSize=<8 16>
      margin=<ltrb 1 0.5 0 0.5>
      outlineSettings {
       bUseBrushTransparency=false
       roundingType=HalfHeightRadius
       width=0
       color {colorUseRule=UseColor_Specified, specifiedColor=<rgba 0 0 0 0>}
       cornerRadii=<0 0 0 0>
      }
      tintColor {colorUseRule=UseColor_Specified, specifiedColor=<rgba 1 1 1 1>}
      uVRegion {bIsValid=false, max=<0 0>, min=<0 0>}
     }
     rightShadowBrush {
      bIsDynamicallyLoaded=false
      drawAs=NoDrawType
      imageType=FullColor
      mirroring=NoMirror
      resourceName=../../../Engine/Content/Slate/Common/ScrollBoxShadowRight.png
      resourceObject=null
      tiling=NoTile
      imageSize=<8 16>
      margin=<ltrb 0 0.5 1 0.5>
      outlineSettings {
       bUseBrushTransparency=false
       roundingType=HalfHeightRadius
       width=0
       color {colorUseRule=UseColor_Specified, specifiedColor=<rgba 0 0 0 0>}
       cornerRadii=<0 0 0 0>
      }
      tintColor {colorUseRule=UseColor_Specified, specifiedColor=<rgba 1 1 1 1>}
      uVRegion {bIsValid=false, max=<0 0>, min=<0 0>}
     }
     topShadowBrush {
      bIsDynamicallyLoaded=false
      drawAs=Box
      imageType=FullColor
      mirroring=NoMirror
      resourceName=../../../Engine/Content/Slate/Common/ScrollBoxShadowTop.png
      resourceObject=null
      tiling=NoTile
      imageSize=<16 8>
      margin=<ltrb 0.5 1 0.5 0>
      outlineSettings {
       bUseBrushTransparency=false
       roundingType=HalfHeightRadius
       width=0
       color {colorUseRule=UseColor_Specified, specifiedColor=<rgba 0 0 0 0>}
       cornerRadii=<0 0 0 0>
      }
      tintColor {colorUseRule=UseColor_Specified, specifiedColor=<rgba 1 1 1 1>}
      uVRegion {bIsValid=false, max=<0 0>, min=<0 0>}
     }
     verticalScrolledContentPadding=<ltrb 0 0 0 1>
    }
   }
   slot {class=GridSlot, properties {column=1, row=1}}
  }
  TabsContainer {
   widgetClassPath=&GridPanel
   bIsVariable=true
   parent=ScrollBox_Tabs
   widget=TabsContainer
   properties {columnFill [1], rowFill [1,0,0,0,0,0]}
   slot {class=ScrollBoxSlot, properties {padding=<ltrb 8 8 8 8>}}
  }
  Next_Spacer {
   widgetClassPath=&Spacer
   bIsVariable=true
   parent=Root
   widget=Next_Spacer
   properties {visibility=Visible, size=<48 32>}
   slot {class=GridSlot, properties {column=2, horizontalAlignment=HAlign_Center, row=1, verticalAlignment=VAlign_Center}}
  }
  Next_Prompt {
   widgetClassPath=&GridPanel
   bIsVariable=true
   parent=Root
   widget=Next_Prompt
   properties {columnFill [1], rowFill [1]}
   slot {class=GridSlot, properties {column=2, layer=10, row=1}}
  }
  Next_SizeBox {widgetClassPath=&SizeBox, bIsVariable=true, parent=Next_Prompt, widget=Next_SizeBox, slot {class=GridSlot}}
  NextTabAction {
   widgetClassPath=&CommonActionWidget
   bIsVariable=true
   parent=Next_Prompt
   widget=NextTabAction
   properties {visibility=Collapsed}
   slot {class=GridSlot, properties {horizontalAlignment=HAlign_Center, layer=10, verticalAlignment=VAlign_Center}}
  }
 ]
}
