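# Base side navigation widget: a vertical, scrollable tab group whose tabs can expand to show nested inner tabs.
# Inner-tab containers are created on demand, mapped by tab ID, inserted into the grid and inherit the parent's styling.
# Click and selection handlers coordinate expansion, deselection, scroll-into-view and de-duplicated selection-changed notifications.
# An optional collapsed state follows mouse hover, while input method changes refresh selectability and select-on-focus behavior.
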WidgetBlueprint /FortUILibrary/Systems/Tabs/WBP_UIKit_SideNavigation_Base
graphs [
 AddSideNavigationInnerTab {nodes [
   n1 = AddSideNavigationInnerTab!() -> (then, TabViewModel:MVVM Base Viewmodel Object Reference, TabID:Name, InnerTabID:Name, Index:int=-1) {is=title K2Node_FunctionEntry}
   > n2 = Branch(Condition="Not Equal (Name)"(A=@n1.TabID) -> (ReturnValue=~))
   > n3 = Set TabContainer(TabContainer:WBP UIKit Tab Container No Layout Object Reference=GetTabContainer!(self, TabID:"Name (by ref)"=@n1.TabID) -> (TabContainerWidget:WBP UIKit Tab Container No Layout Object Reference)) -> (Output_Get:WBP UIKit Tab Container No Layout Object Reference)
   > n4 = Is Valid(InputObject=@n3)
   n5 = InjectTabContainer!(execute=@n4.Is Not Valid, self, TabID:Name=@n1.TabID) -> (then, TabContainerWidget:WBP UIKit Tab Container No Layout Object Reference)
   > n6 = Set TabContainer(TabContainer=@n5)
   > n7 = Is Valid(InputObject=@n6)
   PrintString!(execute=[@n7.Is Not Valid, @n2.else], InString=AddSideNavigationInnerTab - failed, bPrintToScreen, bPrintToLog, TextColor, Duration, Key) -> (then)
   AddTabToContainer!(
    execute=[@n7.Is Valid, @n4.Is Valid]
    self:WBP UIKit Tab Container No Layout Object Reference=Get TabContainer() -> (TabContainer:WBP UIKit Tab Container No Layout Object Reference)
    TabViewModel:MVVM Base Viewmodel Object Reference=Get TabViewModel() -> (TabViewModel:MVVM Base Viewmodel Object Reference)
    TabID:Name=Get InnerTabID() -> (InnerTabID:Name=None)
    Index:int=Get Index() -> (Index:int=0)
   ) -> (then)
  ]}
 AddSideNavigationTab {nodes [
   n1 = AddSideNavigationTab!() -> (then, TabViewModel:MVVM Base Viewmodel Object Reference, TabID:Name, Index:int=-1) {is=title K2Node_FunctionEntry}
   > n2 = Branch(Condition="Not Equal (Name)"(A=@n1.TabID) -> (ReturnValue=~))
   n3 = Get TabContainers() -> (TabContainers:Map of Names to WBP UIKit Tab Container No Layout Object References)
   > n4 = Branch(Condition=NOT Boolean(A=Find(TargetMap:Map of Names to WBP UIKit Tab Container No Layout Object References=@n3, Key:"Name (by ref)"=@n1.TabID) -> (Value:WBP UIKit Tab Container No Layout Object Reference).ReturnValue))
   AddTab!(
    execute=[@n4.then, @n2.else]
    self:WBP UIKit Tab Group Base Object Reference
    Tab VM:MVVM Base Viewmodel Object Reference=Get TabViewModel()
    Tab ID:Name=Get TabID() -> (TabID:Name=None)
    Index:int=Get Index() -> (Index=0)
   ) -> (then)
   > n5 = Cast To WBP_UIKit_SideNavigation_Button!(
    execute
    Object=GetTabButtonBaseByID!(self:Common Tab List Widget Base Object Reference, TabNameID:Name=Get TabID() -> (TabID=None)) -> (ReturnValue:Common Button Base Object Reference)
   ) -> (then, CastFailed, AsWBP UIKit Side Navigation Button:WBP UIKit Side Navigation Button Object Reference) {is=title K2Node_DynamicCast}
   n6 = Create Event()
   > Unbind Event from On Collapsed Anim Finished(self:WBP UIKit Side Navigation Button Object Reference=@n5, Delegate=@n6) {is=title K2Node_RemoveDelegate}
   > Bind Event to On Collapsed Anim Finished(self:WBP UIKit Side Navigation Button Object Reference=@n5, Delegate=@n6) {is=title K2Node_AddDelegate}
  ]}
 CheckTabChanged {nodes [
   n1 = CheckTabChanged!() -> (then, TabID:Name) {is=title K2Node_FunctionEntry}
   n2 = Get PreviousSelectedTab() -> (PreviousSelectedTab:Name=None)
   > Branch(Condition="Not Equal (Name)"(A=@n1, B=@n2) -> (ReturnValue=~))
   > RefreshTabSelectableState!(execute, self, TabID:Name=@n2) -> (then)
   > ClearTabContainerSelection!(execute, self, TabID:Name=@n2) -> (then)
   > Set PreviousSelectedTab(PreviousSelectedTab:Name=Get TabID() -> (TabID=None)) -> (Output_Get:Name=None)
  ]}
 ClearSideNavigationTabs {nodes [
   ClearSideNavigationTabs() {is=title K2Node_FunctionEntry}
   > Clear!(execute, TargetMap:Map of Names to WBP UIKit Tab Container No Layout Object References=Get TabContainers()) -> (then) {is=title K2Node_CallFunction}
   > RemoveTabs(self:WBP UIKit Tab Group Base Object Reference)
   > Set SideNavigationIsCollapsed(SideNavigationIsCollapsed:bool=false) -> (Output_Get:bool)
  ]}
 ClearTabContainerSelection {nodes [
   n1 = ClearTabContainerSelection!() -> (then, TabID:Name) {is=title K2Node_FunctionEntry}
   > n2 = Set TabContainer(TabContainer=GetTabContainer(TabID=@n1))
   > Is Valid(InputObject=@n2)
   > ClearInnerTabsSelection(self:WBP UIKit Tab Container No Layout Object Reference=@n2)
  ]}
 CopyParentTabGroupProps {nodes [
   CopyParentTabGroupProps!() -> (then, Target:WBP UIKit Tab Group Compact Base Object Reference) {is=title K2Node_FunctionEntry}
   > n1 = Sequence!(execute) -> (then_0, then_1, then_2:Exec)
   n2 = Self-Reference()
   > Set IsVertical!(
    execute
    IsVertical:bool=Get IsVertical(None=@n2) -> (IsVertical:bool)
    self:WBP UIKit Tab Group Compact Base Object Reference=Get Target() -> (Target:WBP UIKit Tab Group Compact Base Object Reference)
   ) -> (then, Output_Get:bool)
   > Set IsNaturalNavigation!(
    execute
    IsNaturalNavigation:bool=Get IsNaturalNavigation(None=@n2) -> (IsNaturalNavigation:bool)
    self:WBP UIKit Tab Group Compact Base Object Reference=Get Target()
   ) -> (then, Output_Get:bool)
   > Set TabGroupHorizontalAlignment!(
    execute
    TabGroupHorizontalAlignment:EHorizontalAlignment Enum=Get TabGroupHorizontalAlignment(None=@n2) -> (TabGroupHorizontalAlignment:EHorizontalAlignment Enum=HAlign_Fill)
    self:WBP UIKit Tab Group Compact Base Object Reference=Get Target()
   ) -> (then, Output_Get:EHorizontalAlignment Enum=HAlign_Fill)
   > Set TabGroupVerticalAlignment!(
    execute
    TabGroupVerticalAlignment:EVerticalAlignment Enum=Get TabGroupVerticalAlignment(None=@n2) -> (TabGroupVerticalAlignment:EVerticalAlignment Enum=VAlign_Fill)
    self:WBP UIKit Tab Group Compact Base Object Reference=Get Target()
   ) -> (then, Output_Get:EVerticalAlignment Enum=VAlign_Fill)
   > Set TabGroupPadding!(
    execute
    TabGroupPadding:Margin Structure=Get TabGroupPadding(None=@n2) -> (TabGroupPadding:Margin Structure)
    self:WBP UIKit Tab Group Compact Base Object Reference=Get Target()
   ) -> (then, Output_Get:Margin Structure)
   > Set TabSpacing!(
    execute
    TabSpacing:double=Get TabSpacing(None=@n2) -> (TabSpacing:double)
    self:WBP UIKit Tab Group Compact Base Object Reference=Get Target()
   ) -> (then, Output_Get:double)
   n3 = Get InnerTabButtonClassOverride(None=@n2) -> (InnerTabButtonClassOverride:Common Button Base Class Reference)
   Set ButtonClass!(
    execute=@n1.then_1
    ButtonClass:Common Button Base Class Reference=Select(
     Option 0:Common Button Base Class Reference=Get ButtonClass(None=@n2) -> (ButtonClass:Common Button Base Class Reference)
     Option 1:Common Button Base Class Reference=@n3
     Index=IsValidClass!(Class:Object Class Reference=@n3) -> (ReturnValue:bool)
    ) -> (ReturnValue:Common Button Base Class Reference)
    self:WBP UIKit Tab Group Compact Base Object Reference=Get Target()
   ) -> (then, Output_Get:Common Button Base Class Reference)
   > Set MinButtonSize!(
    execute
    MinButtonSize:Int Point Structure=Get MinButtonSize(None=@n2) -> (MinButtonSize:Int Point Structure)
    self:WBP UIKit Tab Group Compact Base Object Reference=Get Target()
   ) -> (then, Output_Get:Int Point Structure)
   > Set ButtonHorizontalAlignment!(
    execute
    ButtonHorizontalAlignment:EHorizontalAlignment Enum=Get ButtonHorizontalAlignment(None=@n2) -> (ButtonHorizontalAlignment:EHorizontalAlignment Enum=HAlign_Fill)
    self:WBP UIKit Tab Group Compact Base Object Reference=Get Target()
   ) -> (then, Output_Get:EHorizontalAlignment Enum=HAlign_Fill)
   > Set ButtonVerticalAlignment!(
    execute
    ButtonVerticalAlignment:EVerticalAlignment Enum=Get ButtonVerticalAlignment(None=@n2) -> (ButtonVerticalAlignment:EVerticalAlignment Enum=VAlign_Fill)
    self:WBP UIKit Tab Group Compact Base Object Reference=Get Target()
   ) -> (then, Output_Get:EVerticalAlignment Enum=VAlign_Fill)
   > Set IconSize!(
    execute
    IconSize:Vector 2D Structure=Get IconSize(None=@n2) -> (IconSize:Vector 2D Structure)
    self:WBP UIKit Tab Group Compact Base Object Reference=Get Target()
   ) -> (then, Output_Get:Vector 2D Structure)
   Set DividerClass!(execute=@n1.then_2, DividerClass:WBP UIKit Divider Base Class Reference, self:WBP UIKit Tab Group Compact Base Object Reference=Get Target()) -> (then, Output_Get:WBP UIKit Divider Base Class Reference)
  ]}
 CreateTabContainer {nodes [
   n1 = CreateTabContainer!() -> (then, TabID:"Name (by ref)") {is=title K2Node_FunctionEntry}
   n2 = Get TabContainers()
   > n3 = Branch(Condition=NOT Boolean(A=Contains!(TargetMap:Map of Names to WBP UIKit Tab Container No Layout Object References=@n2, Key:"Name (by ref)"=@n1) -> (ReturnValue)))
   > n4 = Create Widget!(
    execute:Exec
    Class:User Widget Class Reference=Get TabContainerClass() -> (TabContainerClass:WBP UIKit Tab Container No Layout Class Reference)
    OwningPlayer:Player Controller Object Reference=GetOwningPlayer()
   ) -> (then:Exec, ReturnValue:WBP UIKit Tab Container No Layout Object Reference)
   > n5 = Set ContainerWidget(ContainerWidget:WBP UIKit Tab Container No Layout Object Reference=@n4) -> (Output_Get:WBP UIKit Tab Container No Layout Object Reference)
   > Add!(
    execute
    TargetMap:Map of Names to WBP UIKit Tab Container No Layout Object References=Get TabContainers()
    Key:"Name (by ref)"=Get TabID() -> (TabID:"Name (by ref)"=None)
    Value:"WBP UIKit Tab Container No Layout Object Reference (by ref)"=@n5
   ) -> (then) {is=title K2Node_CallFunction}
   n6 = Get ContainerWidget() -> (ContainerWidget:WBP UIKit Tab Container No Layout Object Reference)
   > Set ContainerTabgroupClass!(
    execute
    ContainerTabgroupClass:WBP UIKit Tab Group Compact Base Class Reference=Get InnerTabGroupClass() -> (InnerTabGroupClass:WBP UIKit Tab Group Compact Base Class Reference)
    self:WBP UIKit Tab Container No Layout Object Reference=@n6
   ) -> (then, Output_Get:WBP UIKit Tab Group Compact Base Class Reference)
   > Set SkipExpandAnimation!(
    execute
    SkipExpandAnimation:bool=Get SkipExpandAnimation() -> (SkipExpandAnimation:bool)
    self:WBP UIKit Tab Container No Layout Object Reference=@n6
   ) -> (then, Output_Get:bool)
   > Set ExpandDuration!(execute, ExpandDuration:float=Get ExpandDuration() -> (ExpandDuration:float), self:WBP UIKit Tab Container No Layout Object Reference=@n6) -> (then, Output_Get:float)
   > Set InnerTabGroupPadding!(
    execute
    InnerTabGroupPadding:Margin Structure=Get InnerTabGroupPadding() -> (InnerTabGroupPadding:Margin Structure)
    self:WBP UIKit Tab Container No Layout Object Reference=@n6
   ) -> (then, Output_Get:Margin Structure)
   > SetBackground!(
    execute
    self:WBP UIKit Tab Container No Layout Object Reference=@n6
    InBrush:"Slate Brush Structure (by ref)"=Get InnerTabGroupBackround() -> (InnerTabGroupBackround:Slate Brush Structure)
   ) -> (then)
   > Bind Event to On Inner Tab Selected(self:WBP UIKit Tab Container No Layout Object Reference=@n6, Delegate=Create Event()) {is=title K2Node_AddDelegate}
   > Bind Event to On Inner Tab Clicked(self:WBP UIKit Tab Container No Layout Object Reference=@n6, Delegate=Create Event()) {is=title K2Node_AddDelegate}
   > Bind Event to On Container Tab Group Creation(self:WBP UIKit Tab Container No Layout Object Reference=@n6, Delegate=Create Event()) {is=title K2Node_AddDelegate}
   > Bind Event to On Inner Tab Creation(self:WBP UIKit Tab Container No Layout Object Reference=@n6, Delegate=Create Event()) {is=title K2Node_AddDelegate}
   > Bind Event to On Container Expanded State Changed(self:WBP UIKit Tab Container No Layout Object Reference=@n6, Delegate=Create Event()) {is=title K2Node_AddDelegate}
   > Return Node!(execute, TabContainerWidget:WBP UIKit Tab Container No Layout Object Reference=Get ContainerWidget())
   Return Node(execute=@n3.else)
  ]}
 EventGraph {nodes [
   n1 = Event OnInnerTabCreation!() -> (OutputDelegate, then, TabID:Name, InnerTabID:Name, TabButton:Common Button Base Object Reference) {is=title K2Node_CustomEvent}
   > RefreshTabSelectableState(TabID=@n1.TabID)
   > SetShouldSelectUponReceivingFocus!(
    execute
    self:Common Button Base Object Reference=@n1.TabButton
    bInShouldSelectUponReceivingFocus:bool=ShouldSelectOnFocus!(self) -> (ReturnValue:bool)
   ) -> (then)
   n2 = ResetButton!() -> (OutputDelegate, then, Button:Common Button Base Object Reference) {is=title K2Node_CustomEvent}
   > n3 = Cast To WBP_UIKit_ButtonGeneric_Base!(execute, Object=@n2.Button) -> (then, CastFailed, AsWBP UIKit Button Generic Base:WBP UIKit Button Generic Base Object Reference) {is=title K2Node_DynamicCast}
   > n4 = For Each Loop(Array:Array of UIKit Block Object References=Get ModularBlocks!(self:UIKit Modular Button Object Reference=@n3) -> (ModularBlocks:Array of UIKit Block Object References)) -> (Array Element:UIKit Block Object Reference)
   SetIsCollapsed!(
    execute:Exec=@n4.Completed
    self:Object Reference=@n3
    IsCollapsed:bool=Get SideNavigationIsCollapsed() -> (SideNavigationIsCollapsed:bool)
    SkipAnimation:bool=true
   ) -> (then:Exec) {is=title K2Node_Message}
   StopAllAnimations(execute=@n4.LoopBody, self:User Widget Object Reference=@n4)
   > FlushAnimations(self:User Widget Object Reference=@n4)
   > RequestInstantTransition!(
    execute
    self:UIKit Block Object Reference=@n4
    State:EUIKitBlockInstantTransitionState Enum=Select(
     Option 0:EUIKitBlockInstantTransitionState Enum=InstantEnabled
     Option 1:EUIKitBlockInstantTransitionState Enum=InstantSelected
     Index=GetSelected!(self:Common Button Base Object Reference=@n3) -> (ReturnValue:bool)
    ) -> (ReturnValue:EUIKitBlockInstantTransitionState Enum)
   ) -> (then)
   n5 = Event OnInnerTabGroupCreation!() -> (OutputDelegate, then, TabGroupWidget:WBP UIKit Tab Group Compact Base Object Reference) {is=title K2Node_CustomEvent}
   > CopyParentTabGroupProps!(execute, self, Target:WBP UIKit Tab Group Compact Base Object Reference=@n5.TabGroupWidget) -> (then) {is=title K2Node_CallFunction}
   n6 = Event OnTabContainerExpandedStateChanged!() -> (OutputDelegate, then, TabID:Name, IsExpanded:bool=~) {is=title K2Node_CustomEvent}
   > HandleContainerExpandedStateChanged!(execute, self, TabID:Name=@n6.TabID, IsExpanded:bool=@n6.IsExpanded) -> (then)
   n7 = Event OnInnerTabSelected!() -> (OutputDelegate, then, TabID:Name, InnerTabID:Name, TabViewModel:MVVM Base Viewmodel Object Reference) {is=title K2Node_CustomEvent}
   > n8 = Sequence()
   > DeselectAll(self:UIKit Tab Group Object Reference)
   > n9 = Branch(Condition=IsTabContainerExpanded!(self, TabID:Name=@n7.TabID) -> (IsExpanded:bool))
   Branch(execute=@n9.else, Condition=NOT Boolean(A=Get AllowMultipleExpandedContainers() -> (AllowMultipleExpandedContainers:bool)))
   > UnexpandOtherContainers!(execute, self, ExceptTabID:Name=@n7.TabID) -> (then)
   ExpandContainer!(execute=@n8.then_1, self, TabID:Name=@n7.TabID) -> (then)
   > CheckTabChanged!(execute, self, TabID:Name=@n7.TabID) -> (then) {is=title K2Node_CallFunction}
   NotifySelectionChanged!(
    execute=@n8.then_2
    self
    TabID:Name=@n7.TabID
    InnerTabID:Name=@n7.InnerTabID
    TabViewModel:MVVM Base Viewmodel Object Reference=@n7.TabViewModel
   ) -> (then)
   > ScrollToWidget!(
    execute
    self
    Widget:Widget Object Reference=GetInnerTabButton!(self, TabID:Name=@n7.TabID, InnerTabID:Name=@n7.InnerTabID) -> (TabButton:Common Button Base Object Reference)
    IsDelayed:bool=IsAnyAnimationPlaying!(self:User Widget Object Reference=GetTabContainer(TabID=@n7.TabID)) -> (ReturnValue:bool)
   ) -> (then)
   n10 = ScrollToWidget!() -> (OutputDelegate, then, Widget:Widget Object Reference, IsDelayed:bool=~) {is=title K2Node_CustomEvent}
   > Is Valid(InputObject=@n10.Widget)
   > n11 = Branch(Condition=@n10.IsDelayed)
   > n12 = Delay(Duration=Get ExpandDuration())
   n13 = Get CommonInputSubsystem()
   n14 = Switch on ECommonInputType!(execute=[@n12.then, @n15.Is Valid], Selection:ECommonInputType Enum=GetCurrentInputType(self=@n13)) -> (MouseAndKeyboard:Exec, Gamepad:Exec, Touch:Exec, Count:Exec) {is=title K2Node_SwitchEnum}
   ScrollWidgetIntoView!(
    execute=@n14.Gamepad
    self:Scroll Box Object Reference=Get ScrollBox_Tabs() -> (ScrollBox_Tabs:Scroll Box Object Reference)
    WidgetToFind:Widget Object Reference=@n10.Widget
    AnimateScroll:bool=true
    ScrollDestination:EDescendantScrollDestination Enum=Get ScrollDestinationGamepad() -> (ScrollDestinationGamepad:EDescendantScrollDestination Enum=IntoView)
    Padding:float=Get ScrollPaddingGamepad() -> (ScrollPaddingGamepad:float)
   ) -> (then)
   ScrollWidgetIntoView(
    execute=[@n14.MouseAndKeyboard, @n14.Touch]
    self=Get ScrollBox_Tabs()
    WidgetToFind=@n10.Widget
    AnimateScroll=true
    ScrollDestination=Get ScrollDestinationMouse() -> (ScrollDestinationMouse:EDescendantScrollDestination Enum=IntoView)
    Padding=Get ScrollPaddingMouse() -> (ScrollPaddingMouse:float)
   )
   n15 = Is Valid(exec=@n11.else, InputObject=@n13)
   n16 = Event OnUIKitTabClicked!() -> (OutputDelegate, then, TabID:Name, TabVM:MVVM Base Viewmodel Object Reference) {is=title K2Node_CustomEvent}
   > n17 = Sequence!(execute) -> (then_0, then_1)
   > HandleTabClick!(execute, self, TabID:Name=@n16.TabID) -> (then)
   > Branch(Condition=NOT Boolean(A=Get AllowMultipleExpandedContainers()))
   > UnexpandOtherContainers(ExceptTabID=@n16.TabID)
   Call On Side Navigation Tab Clicked!(
    execute:Exec=@n17.then_1
    self:WBP UIKit Side Navigation Base Object Reference
    TabID:Name=@n16.TabID
    IsTabContainerExpanded:bool=IsTabContainerExpanded(TabID=@n16.TabID)
    TabViewModel:MVVM Base Viewmodel Object Reference=@n16.TabVM
   ) -> (then:Exec) {is=title K2Node_CallDelegate}
   > ScrollToWidget(Widget=GetTabButtonBaseByID(TabNameID=@n16.TabID), IsDelayed=NOT Boolean(A=Get SkipExpandAnimation())) {is=title K2Node_CallFunction}
   LatentDeselectAllTabs() {is=title K2Node_CustomEvent}
   > DelayUntilNextTick!(execute) -> (then)
   > DeselectAll()
   n18 = Event OnInnerTabClicked!() -> (OutputDelegate, then, TabID:Name, InnerTabID:Name, TabViewModel:MVVM Base Viewmodel Object Reference) {is=title K2Node_CustomEvent}
   > Call On Side Navigation Inner Tab Clicked!(
    execute:Exec
    self:WBP UIKit Side Navigation Base Object Reference
    TabID:Name=@n18.TabID
    InnerTabID:Name=@n18.InnerTabID
    TabViewModel:MVVM Base Viewmodel Object Reference=@n18.TabViewModel
   ) -> (then:Exec) {is=title K2Node_CallDelegate}
   n19 = Event OnUIKitTabSelected!() -> (OutputDelegate, then, TabID:Name, TabVM:MVVM Base Viewmodel Object Reference) {is=title K2Node_CustomEvent}
   > n20 = Sequence()
   > Branch(Condition="Not Equal (Name)"(A=Get PreviousSelectedTab() -> (PreviousSelectedTab=None), B=@n19.TabID) -> (ReturnValue=~))
   > CheckTabChanged(TabID=@n19.TabID)
   > Branch(Condition=NOT Boolean(A=Get AllowMultipleExpandedContainers()))
   > UnexpandOtherContainers(ExceptTabID=@n19.TabID)
   NotifySelectionChanged(
    execute=@n20.then_1
    TabID=@n19.TabID
    InnerTabID=GetSelectedInnerTabID!(self, TabID:Name=@n19.TabID) -> (InnerTabID:Name=None)
    TabViewModel=@n19.TabVM
   )
   n21 = Event PreConstruct()
   > Set IsVertical!(execute, IsVertical=true) -> (then, Output_Get)
   > Clear(TargetArray:Array of UIKit Tab Data Structures=Get InitialTabsData() -> (InitialTabsData:Array of UIKit Tab Data Structures)) {is=title K2Node_CallArrayFunction}
   > Parent: PreConstruct(IsDesignTime=@n21.IsDesignTime)
   > n22 = Sequence()
   > Clear(TargetMap=Get TabContainers()) {is=title K2Node_CallFunction}
   > Bind Event to On UIKit Tab Selected(self:UIKit Tab Group Object Reference, Delegate=Create Event()) {is=title K2Node_AddDelegate}
   > Bind Event to On UIKit Tab Clicked(self:UIKit Tab Group Object Reference, Delegate=Create Event()) {is=title K2Node_AddDelegate}
   Branch(execute=@n22.then_1, Condition=OR Boolean(A=@n21.IsDesignTime, B=Get UseInitialSideNavigationTabs() -> (UseInitialSideNavigationTabs:bool)))
   > PopulateSideNavigationTabs!(
    execute
    self
    Data:Array of S UI Tab Container Data Structures=Get InitialSideNavigationTabs() -> (InitialSideNavigationTabs:Array of S UI Tab Container Data Structures)
   ) -> (then)
   Event OnMouseEnter()
   > Branch(Condition=Get EnableCollapsedState() -> (EnableCollapsedState:bool))
   > SetIsCollapsed!(execute, self, IsCollapsed:bool=false, SkipAnimation:bool=false) -> (then) {is=title K2Node_CallFunction}
   n23 = Event Event On Input Method Changed!() -> (OutputDelegate, then, bNewInputType:ECommonInputType Enum) {is=title K2Node_Event}
   > Parent: Event On Input Method Changed!(execute, bNewInputType:ECommonInputType Enum=@n23.bNewInputType) -> (then) {is=title K2Node_CallParentFunction}
   > HandleInputMethodChanged()
   Event OnMouseLeave!() -> (OutputDelegate, then, MouseEvent:"Pointer Event Structure (by ref)") {is=title K2Node_Event}
   n24 = IsUsingTouch(PlayerController=GetOwningPlayer())
   > Branch(Condition=AND Boolean(A=Get EnableCollapsedState(), B=NOT Boolean(A=@n24)))
   > SetIsCollapsed(IsCollapsed=true, SkipAnimation=false)
   n25 = Event HandleTabCreation!() -> (OutputDelegate, then, TabNameID:Name=None, TabButton:Common Button Base Object Reference) {is=title K2Node_Event}
   > ResetButton!(execute, self, Button:Common Button Base Object Reference=@n25.TabButton) -> (then) {is=title K2Node_CallFunction}
   > Parent: HandleTabCreation!(execute, TabNameID:Name=@n25.TabNameID, TabButton:Common Button Base Object Reference=@n25.TabButton) -> (then) {is=title K2Node_CallParentFunction}
   > RefreshTabSelectableState(TabID=@n25.TabNameID)
  ]}
 ExpandContainer {nodes [
   n1 = ExpandContainer!() -> (then, TabID:Name) {is=title K2Node_FunctionEntry}
   > n2 = Branch(Condition=NOT Boolean(A=GetIsCollapsedStateOverriden!(self) -> (ReturnValue:bool)))
   > Set TabContainer(TabContainer=GetTabContainer(TabID=@n1))
   > n3 = Get!(execute:Exec) -> (then:Exec, else:Exec, TabContainer:WBP UIKit Tab Container No Layout Object Reference)
   > n4 = Branch(Condition=Get IsExpanded!(self:WBP UIKit Tab Container No Layout Object Reference=@n3) -> (IsExpanded:bool))
   SetIsExpanded!(
    execute=@n4.else
    self:WBP UIKit Tab Container No Layout Object Reference=@n3
    InIsExpanded:bool=true
    SkipAnim:bool=Get SkipExpandAnimation()
   ) -> (then)
   HandleContainerExpandedStateChanged(execute=@n2.else, TabID=@n1, IsExpanded=false)
  ]}
 GetInnerTabButton {nodes [
   n1 = GetInnerTabButton!() -> (then, TabID:Name, InnerTabID:Name) {is=title K2Node_FunctionEntry}
   > n2 = Set TabContainer(TabContainer=GetTabContainer(TabID=@n1))
   > n3 = Is Valid(InputObject=@n2)
   > Return Node!(
    execute
    TabButton:Common Button Base Object Reference=GetTabButtonBaseByID(
     self=GetInnerTabGroupWidget!(self:WBP UIKit Tab Container No Layout Object Reference=@n2) -> (InnerTabGroupWidget:WBP UIKit Tab Group Compact Base Object Reference)
     TabNameID=Get InnerTabID() -> (InnerTabID=None)
    )
   )
   Return Node(execute=@n3.Is Not Valid)
  ]}
 GetInnerTabIdAtIndex {nodes [
   n1 = GetInnerTabIdAtIndex!() -> (then, TabID:"Name (by ref)", Index:int) {is=title K2Node_FunctionEntry}
   > Set TabContainer(TabContainer=GetTabContainer(TabID=@n1))
   > n2 = Get()
   > Return Node!(
    execute
    InnerTabID:Name=GetTabIdAtIndex!(self:Common Tab List Widget Base Object Reference=GetInnerTabGroupWidget(self=@n2), Index:int=Get Index() -> (Index=0)) -> (ReturnValue:Name=None)
   )
   Return Node(execute=@n2.else, InnerTabID=None)
  ]}
 GetInnerTabsCount {nodes [
   n1 = GetInnerTabsCount!() -> (then, TabID:Name) {is=title K2Node_FunctionEntry}
   > Set TabContainer(TabContainer=GetTabContainer(TabID=@n1))
   > n2 = Get()
   > Return Node!(execute, Count:int=GetTabCount!(self:Common Tab List Widget Base Object Reference=GetInnerTabGroupWidget(self=@n2)) -> (ReturnValue:int=0))
   Return Node(execute=@n2.else, Count=0)
  ]}
 GetIsCollapsedStateOverriden {nodes [
   GetIsCollapsedStateOverriden() {is=title K2Node_FunctionEntry}
   > Return Node!(
    execute
    ReturnValue:bool=Select(
     Option 0:bool=Get SideNavigationIsCollapsed()
     Option 1:bool=false
     Index=Get AllowInnerTabExpandInCollapsedState() -> (AllowInnerTabExpandInCollapsedState:bool)
    ) -> (ReturnValue:bool)
   )
  ]}
 GetSelectedInnerTabID {nodes [
   n1 = GetSelectedInnerTabID!() -> (then, TabID:Name) {is=title K2Node_FunctionEntry}
   > n2 = Set TabContainer(TabContainer=GetTabContainer(TabID=@n1))
   > n3 = Is Valid(InputObject=@n2)
   Return Node!(execute=@n3.Is Not Valid, InnerTabID:Name=None)
   > Return Node(InnerTabID=GetSelectedTabId!(self:Common Tab List Widget Base Object Reference=GetInnerTabGroupWidget(self=@n2)) -> (ReturnValue:Name=None))
  ]}
 GetSideNavigationButtonBaseByViewModel {nodes [
   n1 = GetSideNavigationButtonBaseByViewModel!() -> (then, TabViewModel:MVVM Base Viewmodel Object Reference) {is=title K2Node_FunctionEntry}
   > n2 = Is Valid(InputObject=@n1)
   n3 = Get CachedViewModels() -> (CachedViewModels:Map of Names to MVVM Base Viewmodel Object References)
   > n4 = Keys!(execute, TargetMap:Map of Names to MVVM Base Viewmodel Object References=@n3) -> (then, Keys:Array of Names)
   > n5 = Values!(execute, TargetMap:Map of Names to MVVM Base Viewmodel Object References=@n3) -> (then, Values:Array of MVVM Base Viewmodel Object References)
   > n6 = For Each Loop(Array:Array of Names=@n4) -> (Array Element:Name)
   Return Node!(execute=[@n6.Completed, @n2.Is Not Valid], ButtonBase:Common Button Base Object Reference)
   > n7 = Branch(Condition="Equal (Object)"(
     A="Get (a copy)"(Array:Array of MVVM Base Viewmodel Object References=@n5, Dimension 1=@n6.Array Index) -> (Output:MVVM Base Viewmodel Object Reference)
     B=Get TabViewModel()
    ))
   > Return Node(ButtonBase=GetTabButtonBaseByID(TabNameID=@n6))
   Set CurrentTabCointainer(execute=@n7.else, CurrentTabCointainer:WBP UIKit Tab Container No Layout Object Reference=GetTabContainer(TabID=@n6)) -> (Output_Get:WBP UIKit Tab Container No Layout Object Reference)
   > n8 = Get!(execute) -> (then, else, CurrentTabCointainer:WBP UIKit Tab Container No Layout Object Reference)
   > Set InnerTabButton(InnerTabButton:Common Button Base Object Reference=GetTabButtonBaseByViewModel!(self:UIKit Tab Group Object Reference=GetInnerTabGroupWidget(self=@n8), TabViewModel:MVVM Base Viewmodel Object Reference=Get TabViewModel()) -> (ReturnValue:Common Button Base Object Reference)) -> (Output_Get:Common Button Base Object Reference)
   > n9 = Get!(execute) -> (then, else, InnerTabButton:Common Button Base Object Reference)
   > Return Node(ButtonBase=@n9)
  ]}
 GetSideNavigationSelectedButtonBase {nodes [
   GetSideNavigationSelectedButtonBase() {is=title K2Node_FunctionEntry}
   > n1 = Set SelectedTabID(SelectedTabID:Name=GetSelectedTabId() -> (ReturnValue=None)) -> (Output_Get:Name=None)
   > n2 = Branch(Condition="Not Equal (Name)"(A=@n1) -> (ReturnValue=~))
   Set TabContainer(execute=[@n2.then, @n6.then], TabContainer=GetTabContainer(TabID=Get SelectedTabID() -> (SelectedTabID:Name=None)))
   > n3 = Get!(execute) -> (then, else, TabContainer:WBP UIKit Tab Container No Layout Object Reference)
   n4 = GetInnerTabGroupWidget(self=@n3)
   > Set TabButton(TabButton:Common Button Base Object Reference=GetTabButtonBaseByID(self=@n4, TabNameID=GetSelectedTabId(self=@n4) -> (ReturnValue=None))) -> (Output_Get:Common Button Base Object Reference)
   > n5 = Get!(execute) -> (then, else, TabButton:Common Button Base Object Reference)
   > Return Node!(execute, Button:Common Button Base Object Reference=@n5)
   Return Node(execute=[@n5.else, @n3.else], Button=GetTabButtonBaseByID(TabNameID=Get SelectedTabID() -> (SelectedTabID=None)))
   n6 = Set SelectedTabID(execute=@n2.else, SelectedTabID=Get PreviousSelectedTab() -> (PreviousSelectedTab=None)) -> (Output_Get=None)
  ]}
 GetSideNavigationTabViewModel {nodes [
   n1 = GetSideNavigationTabViewModel!() -> (then, TabID:Name, InnerTabID:Name) {is=title K2Node_FunctionEntry}
   > n2 = Branch(Condition="Not Equal (Name)"(A=@n1.InnerTabID) -> (ReturnValue=~))
   > Set TabContainer(TabContainer=GetTabContainer(TabID=@n1))
   > n3 = Get!(execute) -> (then, else, TabContainer:WBP UIKit Tab Container No Layout Object Reference)
   > Return Node!(
    execute
    TabViewModel:MVVM Base Viewmodel Object Reference=GetTabViewModel!(self:UIKit Tab Group Object Reference=GetInnerTabGroupWidget(self=@n3), TabID:Name=Get InnerTabID() -> (InnerTabID=None)) -> (ReturnValue:MVVM Base Viewmodel Object Reference)
   )
   Return Node(execute=[@n3.else, @n2.else], TabViewModel=GetTabViewModel(TabID=@n1))
  ]}
 GetTabContainer {nodes [
   n1 = GetTabContainer!() -> (then, TabID:"Name (by ref)") {is=title K2Node_FunctionEntry}
   > Return Node!(execute, TabContainerWidget:WBP UIKit Tab Container No Layout Object Reference=Find(TargetMap=Get TabContainers(), Key=@n1))
  ]}
 HandleAudio {nodes [
   n1 = HandleAudio!() -> (then, SkipAnimation:bool=~, Remaining Anim Time:double=~, Total Anim Time:double=~) {is=title K2Node_FunctionEntry}
   > n2 = Branch(Condition=@n1)
   n3 = Get StubSoundThresholdMult() -> (StubSoundThresholdMult:double)
   n4 = Branch(
    execute=@n2.else
    Condition="float < float"(A=@n1.Remaining Anim Time, B=float * float(A=@n1.Total Anim Time, B=@n3) {is=title K2Node_PromotableOperator})
   )
   > n5 = Set Sound Start Time(Sound Start Time:double=float * float(A=Get Total Anim Time() -> (Total Anim Time:double), B=Get StubSoundStartOffsetMult() -> (StubSoundStartOffsetMult:double))) -> (Output_Get:double)
   PlaySound2D(
    execute=[@n5.then, @n6.then]
    Sound=Select(
     Option 0:Sound Base Object Reference=Get ExpandSound() -> (ExpandSound:Sound Base Object Reference)
     Option 1:Sound Base Object Reference=Get CollapseSound() -> (CollapseSound:Sound Base Object Reference)
     Index=Get SideNavigationIsCollapsed()
    ) -> (ReturnValue:Sound Base Object Reference)
    StartTime=Get Sound Start Time() -> (Sound Start Time:double)
   )
   n6 = Set Sound Start Time(execute=@n4.else, Sound Start Time=float - float(A=Get Total Anim Time(), B=Get Remaining Anim Time() -> (Remaining Anim Time:double)))
   n7 = Get Remaining Anim Time()
   n8 = Get Total Anim Time()
   PrintString(InString=Append!(A="Remaining Time: ", B="To String (Float)"(InDouble=@n7), C:String=" || Total Time: ", D:String="To String (Float)"(InDouble=@n8)) -> (ReturnValue))
  ]}
 HandleContainerExpandedStateChanged {nodes [
   n1 = HandleContainerExpandedStateChanged!() -> (then, TabID:Name, IsExpanded:bool=~) {is=title K2Node_FunctionEntry}
   > n2 = Set TabContainer(TabContainer=GetTabContainer(TabID=@n1) {is=title K2Node_CallFunction})
   > Is Valid(InputObject=@n2)
   > n3 = Branch(Condition=@n1.IsExpanded)
   n4 = GetInnerTabGroupWidget(self=Get TabContainer())
   > Branch(Condition="integer >= integer"(A=GetSelectedTabIndex!(self:UIKit Tab Group Object Reference=@n4) -> (ReturnValue:int=0)))
   > LatentDeselectAllTabs() {is=title K2Node_CallFunction}
   Branch(execute=@n3.else, Condition=ShouldDisplaySelectionWhenNotExpanded!(self) -> (ReturnValue:bool))
   n5 = GetInnerTabGroupWidget(self=Get TabContainer())
   > Branch(Condition="integer >= integer"(A=GetSelectedTabIndex(self=@n5) -> (ReturnValue=0)))
   > n6 = Set TabButton(TabButton=GetTabButton!(self:WBP UIKit Tab Container No Layout Object Reference=Get TabContainer()) -> (TabButton:Common Button Base Object Reference))
   > Is Valid(InputObject=@n6)
   n7 = Get TabButton() -> (TabButton:Common Button Base Object Reference)
   > SetIsSelectable!(execute, self:Common Button Base Object Reference=@n7, bInIsSelectable:bool=true) -> (then)
   > SetIsSelected!(execute, self:Common Button Base Object Reference=@n7, InSelected:bool=true, bGiveClickFeedback:bool=false) -> (then)
  ]}
 HandleInputMethodChanged {nodes [
   HandleInputMethodChanged() {is=title K2Node_FunctionEntry}
   > n1 = For Loop(LastIndex=int - int(A=GetTabCount() -> (ReturnValue=0)))
   > n2 = Set LocalTabID(LocalTabID:Name=GetTabIdAtIndex(Index=@n1) -> (ReturnValue=None)) -> (Output_Get:Name=None)
   > RefreshTabSelectableState(TabID=@n2)
   > n3 = Set LocalTabContainer(LocalTabContainer:WBP UIKit Tab Container No Layout Object Reference=GetTabContainer(TabID=@n2)) -> (Output_Get:WBP UIKit Tab Container No Layout Object Reference)
   > Is Valid(InputObject=@n3)
   > RefreshTabSlotEnabledState(self:WBP UIKit Tab Container No Layout Object Reference=@n3)
   > Set LocalShouldSelectOnFocus(LocalShouldSelectOnFocus:bool=ShouldSelectOnFocus()) -> (Output_Get:bool)
   n4 = GetInnerTabGroupWidget(self=@n3)
   > Is Valid(InputObject=@n4)
   > n5 = For Each Loop(Array:Array of Widget Object References=GetAllChildren(self=Get TabsContainer!(self:WBP UIKit Tab Group Compact Base Object Reference=@n4) -> (TabsContainer:Grid Panel Object Reference))) -> (Array Element:Widget Object Reference)
   > n6 = Cast To CommonButtonBase!(execute, Object=@n5) -> (then, CastFailed, AsCommon Button Base:Common Button Base Object Reference) {is=title K2Node_DynamicCast}
   > SetShouldSelectUponReceivingFocus(self=@n6, bInShouldSelectUponReceivingFocus=Get LocalShouldSelectOnFocus() -> (LocalShouldSelectOnFocus:bool))
  ]}
 HandleTabClick {nodes [
   n1 = HandleTabClick!() -> (then, TabID:Name) {is=title K2Node_FunctionEntry}
   > n2 = Set TabContainer(TabContainer=GetTabContainer(TabID=@n1))
   > Is Valid(InputObject=@n2)
   n3 = Get CommonInputSubsystem()
   > Is Valid(InputObject=@n3)
   > Set InputMethod(InputMethod:ECommonInputType Enum=GetCurrentInputType(self=@n3)) -> (Output_Get:ECommonInputType Enum)
   > n4 = Branch(Condition=Get IsExpanded(self=Get TabContainer()))
   n5 = Get InputMethod() -> (InputMethod:ECommonInputType Enum)
   > Branch(Condition=OR Boolean(
     A="Not Equal (Enum)"!(A:ECommonInputType Enum=@n5, B:ECommonInputType Enum=Gamepad) -> (ReturnValue:bool)
     B=Get AllowContainerInteractionOnGamepad() -> (AllowContainerInteractionOnGamepad:bool)
    ))
   > SetIsExpanded(self=Get TabContainer(), InIsExpanded=false, SkipAnim=Get SkipExpandAnimation())
   n6 = Sequence(execute=@n4.else)
   n7 = NOT Boolean(A=Get AllowContainerInteractionOnGamepad())
   n8 = "Equal (Enum)"(A=Get InputMethod(), B=Gamepad)
   > Branch(Condition=OR Boolean(A=AND Boolean(A=@n7, B=@n8), B=Get SelectInnerTabOnContainerExpand() -> (SelectInnerTabOnContainerExpand:bool)))
   n9 = GetInnerTabGroupWidget(self=Get TabContainer())
   > Is Valid(InputObject=@n9)
   > SelectTabByIndex!(execute, self:UIKit Tab Group Object Reference=@n9, Index:int=0, bSuppressClickFeedback:bool=true) -> (then, ReturnValue:bool)
   ExpandToSelectedItem(execute=@n6.then_1, self:WBP UIKit Tab Container No Layout Object Reference=Get TabContainer())
  ]}
 InjectTabContainer {nodes [
   n1 = InjectTabContainer!() -> (then, TabID:Name) {is=title K2Node_FunctionEntry}
   > n2 = Set TabButton(TabButton=GetTabButtonBaseByID(TabNameID=@n1))
   > n3 = Is Valid(InputObject=@n2)
   > n4 = CreateTabContainer!(execute, self, TabID:"Name (by ref)"=@n1) -> (then, TabContainerWidget:WBP UIKit Tab Container No Layout Object Reference) {is=title K2Node_CallFunction}
   > n5 = Is Valid(InputObject=@n4)
   n6 = Get TabID() -> (TabID:Name=None)
   > InitContainer!(
    execute
    self:WBP UIKit Tab Container No Layout Object Reference=@n4
    ParentTabButton:Common Button Base Object Reference=Get TabButton()
    TabID:Name=@n6
   ) -> (then)
   > AddTabToGrid!(
    execute
    self:WBP UIKit Tab Group Base Object Reference
    Widget:Widget Object Reference=@n4
    Index:int=GetTabIndexByID!(self:UIKit Tab Group Object Reference, TabNameID:Name=@n6) -> (ReturnValue:int=0)
   ) -> (then)
   > Return Node(TabContainerWidget=@n4)
   PrintString(execute=[@n5.Is Not Valid, @n3.Is Not Valid], InString=InjectTabContainer failed)
   > Return Node()
  ]}
 IsTabContainerEmpty {nodes [
   n1 = IsTabContainerEmpty!() -> (then, TabID:Name) {is=title K2Node_FunctionEntry}
   > Return Node!(execute, IsEmpty:bool="Equal (Integer)"(A=GetInnerTabsCount!(self, TabID:Name=@n1) -> (Count:int=0) {is=title K2Node_CallFunction}))
  ]}
 IsTabContainerExpanded {nodes [
   n1 = IsTabContainerExpanded!() -> (then, TabID:Name) {is=title K2Node_FunctionEntry}
   > n2 = Set TabContainer(TabContainer=GetTabContainer(TabID=@n1))
   > n3 = Is Valid(InputObject=@n2)
   Return Node!(execute=@n3.Is Not Valid, IsExpanded:bool=true)
   > Return Node(IsExpanded=Get IsExpanded(self=@n2))
  ]}
 NotifySelectionChanged {nodes [
   n1 = NotifySelectionChanged!() -> (then, TabID:Name, InnerTabID:Name, TabViewModel:MVVM Base Viewmodel Object Reference) {is=title K2Node_FunctionEntry}
   > n2 = Branch(Condition=Get ShouldWaitForInnerTabCallback() -> (ShouldWaitForInnerTabCallback:bool))
   n3 = Get LastNotifiedInnerTabID() -> (LastNotifiedInnerTabID:Name=None)
   n4 = Get LastNotifiedSelectedTabID() -> (LastNotifiedSelectedTabID:Name=None)
   Branch(
    execute=@n2.else
    Condition=OR Boolean(A="Not Equal (Name)"(A=@n1, B=@n4) -> (ReturnValue=~), B="Not Equal (Name)"(A=@n1.InnerTabID, B=@n3) -> (ReturnValue=~))
   )
   n5 = Get TabID() -> (TabID=None)
   n6 = Get InnerTabID() -> (InnerTabID=None)
   > Call On Side Navigation Selection Changed!(
    execute:Exec
    self:WBP UIKit Side Navigation Base Object Reference
    TabID:Name=@n5
    InnerTabID:Name=@n6
    TabViewModel:MVVM Base Viewmodel Object Reference=Get TabViewModel()
   ) -> (then:Exec) {is=title K2Node_CallDelegate}
   > Set LastNotifiedSelectedTabID(LastNotifiedSelectedTabID:Name=@n5) -> (Output_Get:Name=None)
   > Set LastNotifiedInnerTabID(LastNotifiedInnerTabID:Name=@n6) -> (Output_Get:Name=None)
  ]}
 OnButtonFinishedCollapsedAnim {nodes [
   OnButtonFinishedCollapsedAnim() {is=title K2Node_FunctionEntry}
   > n1 = Branch(Condition=Get SideNavigationIsCollapsed())
   Set TotalButtonsFinishedCollapsedAnim(
    execute=@n1.else
    TotalButtonsFinishedCollapsedAnim:int=int + int(A=Get TotalButtonsFinishedCollapsedAnim() -> (TotalButtonsFinishedCollapsedAnim:int=0))
   ) -> (Output_Get:int=0)
   > Branch(Condition="Equal (Integer)"(A=Get TotalButtonsFinishedCollapsedAnim() -> (TotalButtonsFinishedCollapsedAnim=0), B=1))
   > ScrollToWidget(
    Widget=GetSideNavigationSelectedButtonBase!(self) -> (Button:Common Button Base Object Reference) {is=title K2Node_CallFunction}
    IsDelayed=false
   )
  ]}
 OnSideNavigationCollapsedStateChanged {nodes [OnSideNavigationCollapsedStateChanged!() -> (then, IsCollapsed:bool=~) {is=title K2Node_FunctionEntry}]}
 OnSideNavigationInnerTabClicked {nodes [OnSideNavigationInnerTabClicked!() -> (then, TabID:Name, InnerTabID:Name, TabViewModel:MVVM Base Viewmodel Object Reference) {is=title K2Node_FunctionEntry}]}
 OnSideNavigationSelectionChanged {nodes [OnSideNavigationSelectionChanged!() -> (then, TabID:Name, InnerTabID:Name, TabViewModel:MVVM Base Viewmodel Object Reference) {is=title K2Node_FunctionEntry}]}
 OnSideNavigationTabClicked {nodes [OnSideNavigationTabClicked!() -> (then, TabID:Name, IsTabContainerExpanded:bool=~, TabViewModel:MVVM Base Viewmodel Object Reference) {is=title K2Node_FunctionEntry}]}
 PopulateSideNavigationTabs {nodes [
   n1 = PopulateSideNavigationTabs!() -> (then, Data:Array of S UI Tab Container Data Structures) {is=title K2Node_FunctionEntry}
   > n2 = Sequence!(execute) -> (then_0, then_1, then_2:Exec)
   > Set PreviousSelectedTab(PreviousSelectedTab=None) -> (Output_Get=None)
   > Clear(TargetMap=Get TabContainers())
   n3 = For Each Loop(Exec=@n2.then_1, Array:Array of S UI Tab Container Data Structures=@n1) -> (Array Element:S UI Tab Container Data Structure)
   n4 = Break S UI Tab Container Data!(S_UI_TabContainerData:"S UI Tab Container Data Structure (by ref)"=@n3) -> (
    Tab_6_#1:UIKit Tab Data Structure
    InnerTabs_7_#2:Array of UIKit Tab Data Structures
   ) {is=title K2Node_BreakStruct}
   n5 = Break UIKit Tab Data!(UIKitTabData:"UIKit Tab Data Structure (by ref)"=@n4) -> (bEnableTabID:bool=~, TabID:Name, TabViewModel:MVVM Base Viewmodel Object Reference) {is=title K2Node_BreakStruct}
   n6 = "To String (Integer)"(InInt=@n3.Array Index)
   > AddSideNavigationTab!(
    execute
    self
    TabViewModel:MVVM Base Viewmodel Object Reference=@n5.TabViewModel
    TabID:Name=SelectName!(A:Name=@n5.TabID, B:Name=String To Name(InString=@n6), bPickA:bool=@n5) -> (ReturnValue:Name=None)
    Index:int=-1
   ) -> (then) {is=title K2Node_CallFunction}
   n7 = GetTabCount() -> (ReturnValue=0)
   > Set CurrentTabID(CurrentTabID:Name=GetTabIDbyIndex!(self:UIKit Tab Group Object Reference, Index:int=int - int(A=@n7)) -> (ReturnValue:Name=None)) -> (Output_Get:Name=None)
   > n8 = For Each Loop(Array:Array of UIKit Tab Data Structures=@n4.InnerTabs_7_#2) -> (Array Element:UIKit Tab Data Structure)
   n9 = Break UIKit Tab Data(UIKitTabData=@n8) -> (bEnableTabID=~)
   > AddSideNavigationInnerTab!(
    execute
    self
    TabViewModel:MVVM Base Viewmodel Object Reference=@n9.TabViewModel
    TabID:Name=Get CurrentTabID() -> (CurrentTabID:Name=None)
    InnerTabID:Name=SelectName(A=@n9.TabID, B=None, bPickA=@n9) -> (ReturnValue=None)
    Index:int=-1
   ) -> (then) {is=title K2Node_CallFunction}
  ]}
 RefreshTabSelectableState {nodes [
   n1 = RefreshTabSelectableState!() -> (then, TabID:Name) {is=title K2Node_FunctionEntry}
   > n2 = Set TabButton(TabButton=GetTabButtonBaseByID(TabNameID=@n1))
   > Is Valid(InputObject=@n2)
   > n3 = Set LocalShouldSelectOnFocus(LocalShouldSelectOnFocus=ShouldSelectOnFocus())
   n4 = IsValid(Object=GetTabContainer(TabID=@n1))
   n5 = GetIsCollapsedStateOverriden() {is=title K2Node_CallFunction}
   n6 = GetSelected(self=Get TabButton())
   n7 = Get TabButton()
   > SetIsSelectable(self=@n7, bInIsSelectable=OR Boolean!(A=@n3, B=NOT Boolean(A=@n4), C:bool=AND Boolean(A=@n6, B=@n5)) -> (ReturnValue))
   > SetShouldSelectUponReceivingFocus(self=@n7, bInShouldSelectUponReceivingFocus=Get LocalShouldSelectOnFocus())
  ]}
 SelectSideNavigationTab {nodes [
   n1 = SelectSideNavigationTab!() -> (then, TabID:Name, InnerTabID:Name, bSuppressClickFeedback:bool=~) {is=title K2Node_FunctionEntry}
   > n2 = Branch(Condition="Not Equal (Name)"(A=@n1) -> (ReturnValue=~))
   n3 = Get TabID() -> (TabID=None)
   n4 = Get InnerTabID() -> (InnerTabID=None)
   > Set ShouldWaitForInnerTabCallback(ShouldWaitForInnerTabCallback:bool=IsValid(Object=GetInnerTabButton(TabID=@n3, InnerTabID=@n4) {is=title K2Node_CallFunction})) -> (Output_Get:bool)
   > n5 = SelectTabByID!(
    execute
    self:Common Tab List Widget Base Object Reference
    TabNameID:Name=Get TabID() -> (TabID=None)
    bSuppressClickFeedback:bool=Get bSuppressClickFeedback() -> (bSuppressClickFeedback:bool)
   ) -> (then, ReturnValue:bool)
   > Set ShouldWaitForInnerTabCallback(ShouldWaitForInnerTabCallback=false)
   > n6 = Branch(Condition=@n5)
   > Set TabContainer(TabContainer=GetTabContainer(TabID=Get TabID() -> (TabID=None)))
   > n7 = Get()
   > n8 = SelectTabByID(self=GetInnerTabGroupWidget(self=@n7), TabNameID=Get InnerTabID() -> (InnerTabID=None), bSuppressClickFeedback=Get bSuppressClickFeedback())
   > Return Node!(execute, Success:bool=@n8)
   Return Node(execute=[@n7.else, @n6.else, @n2.else], Success=false)
  ]}
 SelectSideNavigationViewModel {nodes [
   n1 = SelectSideNavigationViewModel!() -> (then, TabViewModel:MVVM Base Viewmodel Object Reference, bSuppressClickFeedback:bool=~) {is=title K2Node_FunctionEntry}
   > n2 = Is Valid(InputObject=@n1)
   n3 = Get CachedViewModels()
   > n4 = Keys(TargetMap=@n3)
   > n5 = Values(TargetMap=@n3)
   > n6 = For Each Loop(Array:Array of Names=@n4) -> (Array Element:Name)
   Return Node(execute=[@n6.Completed, @n2.Is Not Valid], Success=false)
   > n7 = Branch(Condition="Equal (Object)"(A="Get (a copy)"(Array=@n5, Dimension 1=@n6.Array Index), B=Get TabViewModel()))
   > n8 = SelectTabByID(TabNameID=@n6, bSuppressClickFeedback=Get bSuppressClickFeedback())
   > Return Node(Success=@n8)
   Set CurrentTabCointainer(execute=@n7.else, CurrentTabCointainer=GetTabContainer(TabID=@n6))
   > n9 = Get!(execute) -> (then, else, CurrentTabCointainer:WBP UIKit Tab Container No Layout Object Reference)
   n10 = Get CachedViewModels!(self:UIKit Tab Group Object Reference=GetInnerTabGroupWidget(self=@n9)) -> (CachedViewModels)
   > n11 = Values(TargetMap=@n10)
   > n12 = Keys(TargetMap=@n10)
   n13 = Find Item!(
    TargetArray:Array of MVVM Base Viewmodel Object References=@n11
    ItemToFind:"MVVM Base Viewmodel Object Reference (by ref)"=Get TabViewModel()
   ) -> (ReturnValue)
   > n14 = Set FoundIndex(FoundIndex:int=@n13) -> (Output_Get:int=0)
   > Branch(Condition="Not Equal (Integer)"(A=@n14, B=-1))
   > n15 = SelectSideNavigationTab!(
    execute
    self
    TabID:Name=Get ContainerTabID!(self:WBP UIKit Tab Container No Layout Object Reference=@n9) -> (ContainerTabID:Name=None)
    InnerTabID:Name="Get (a ref)"(Array:Array of Names=@n12, Dimension 1=@n13) -> (Output:"Name (by ref)")
    bSuppressClickFeedback:bool=false
   ) -> (then, Success:bool) {is=title K2Node_CallFunction}
   > Return Node(Success=@n15)
  ]}
 SetIsCollapsed {nodes [
   SetIsCollapsed!() -> (then, IsCollapsed:bool=~, SkipAnimation:bool=~) {is=title K2Node_FunctionEntry}
   n1 = Get IsCollapsed() -> (IsCollapsed:bool)
   > Branch(Condition="Not Equal (Boolean)"(A=Get SideNavigationIsCollapsed(), B=@n1) -> (ReturnValue=~))
   > Set SideNavigationIsCollapsed(SideNavigationIsCollapsed=@n1)
   > Set TotalButtonsFinishedCollapsedAnim(TotalButtonsFinishedCollapsedAnim=0) -> (Output_Get=0)
   > n2 = For Each Loop(Array:Array of Widget Object References=GetAllChildren(self=Get TabsContainer!() -> (TabsContainer))) -> (Array Element:Widget Object Reference)
   Call On Side Navigation Collapsed State Changed!(execute:Exec=@n2.Completed, self:WBP UIKit Side Navigation Base Object Reference, IsCollapsed:bool=Get IsCollapsed()) -> (then:Exec) {is=title K2Node_CallDelegate}
   n3 = Cast To BPI_UIKit_Collapsible!(execute=@n2.LoopBody, Object=@n2) -> (then, CastFailed, AsBPI UIKit Collapsible:BPI UIKit Collapsible Interface) {is=title K2Node_DynamicCast}
   > SetIsCollapsed(self:BPI UIKit Collapsible Interface=@n3, IsCollapsed=Get IsCollapsed(), SkipAnimation=Get SkipAnimation() -> (SkipAnimation:bool)) {is=title K2Node_CallFunction}
   > n4 = Branch(Condition=Get HasHandledAudio() -> (HasHandledAudio:bool))
   n5 = GetRemainingAnimTime!(execute=@n4.else, self:BPI UIKit Collapsible Interface=@n3) -> (then, FunctionIsImplemented:bool, RemainingTime:double, TotalTime:double)
   > n6 = Set HasHandledAudio(HasHandledAudio:bool=@n5) -> (Output_Get:bool)
   > Branch(Condition=@n6)
   > HandleAudio!(execute, self, SkipAnimation:bool=Get SkipAnimation(), Remaining Anim Time:double=@n5.RemainingTime, Total Anim Time:double=@n5.TotalTime) -> (then) {is=title K2Node_CallFunction}
  ]}
 ShouldDisplaySelectionWhenNotExpanded {nodes [
   ShouldDisplaySelectionWhenNotExpanded() {is=title K2Node_FunctionEntry}
   n1 = Get CommonInputSubsystem()
   > n2 = Is Valid(InputObject=@n1)
   Return Node!(execute=@n2.Is Not Valid, ReturnValue:bool=false)
   n3 = GetCurrentInputType(self=@n1)
   > Return Node(ReturnValue=OR Boolean!(A="Not Equal (Enum)"(A=@n3, B=Gamepad), B=GetIsCollapsedStateOverriden()) -> (ReturnValue))
  ]}
 ShouldSelectOnFocus {nodes [
   ShouldSelectOnFocus() {is=title K2Node_FunctionEntry}
   n1 = Get CommonInputSubsystem()
   > n2 = Is Valid(InputObject=@n1)
   Return Node(execute=@n2.Is Not Valid, ReturnValue=false)
   > Return Node(ReturnValue="Equal (Enum)"(A=GetCurrentInputType(self=@n1), B=Gamepad))
  ]}
 UnexpandOtherContainers {nodes [
   n1 = UnexpandOtherContainers!() -> (then, ExceptTabID:Name) {is=title K2Node_FunctionEntry}
   > Branch(Condition=NOT Boolean(A=GetIsCollapsedStateOverriden()))
   > n2 = Values(TargetMap:Map of Names to WBP UIKit Tab Container No Layout Object References=Get TabContainers()) -> (Values:Array of WBP UIKit Tab Container No Layout Object References)
   > n3 = For Each Loop(Array:Array of WBP UIKit Tab Container No Layout Object References=@n2) -> (Array Element:WBP UIKit Tab Container No Layout Object Reference)
   > Branch(Condition="Not Equal (Name)"(A=Get ContainerTabID(self=@n3) -> (ContainerTabID=None), B=@n1) -> (ReturnValue=~))
   > SetIsExpanded(self=@n3, InIsExpanded=false, SkipAnim=Get SkipExpandAnimation())
  ]}
]
variables [
 AllowContainerInteractionOnGamepad {type=boolean}
 AllowInnerTabExpandInCollapsedState {type=boolean}
 AllowMultipleExpandedContainers {type=boolean}
 CollapseSound {type=SoundBase, default=null}
 EnableCollapsedState {type=boolean}
 ExpandDuration {type=number, default=0.25}
 ExpandSound {type=SoundBase, default=null}
 InitialSideNavigationTabs {
  type="array<S_UI_TabContainerData>"
  default [
   {
    innerTabs [{bEnableTabId=false, tabId=None, tabViewModel=null}, {bEnableTabId=false, tabId=None, tabViewModel=null}]
    tab {bEnableTabId=false, tabId=None, tabViewModel=null}
   }
   {innerTabs [], tab {bEnableTabId=false, tabId=None, tabViewModel=null}}
   {innerTabs [], tab {bEnableTabId=false, tabId=None, tabViewModel=null}}
  ]
 }
 InnerTabButtonClassOverride {type=Class, default=null}
 InnerTabGroupBackround {
  type=SlateBrush
  default {
   bIsDynamicallyLoaded=false
   drawAs=NoDrawType
   imageType=NoImage
   mirroring=NoMirror
   resourceName=None
   resourceObject=null
   tiling=NoTile
   imageSize=<32 32>
   margin=<ltrb 0 0 0 0>
   outlineSettings {
    bUseBrushTransparency=false
    roundingType=HalfHeightRadius
    width=0
    color {colorUseRule=UseColor_Specified, specifiedColor=<rgba 0 0 0 0>}
    cornerRadii=<0 0 0 1>
   }
   tintColor {colorUseRule=UseColor_Specified, specifiedColor=<rgba 1 1 1 0>}
   uVRegion {bIsValid=false, max=<0 0>, min=<0 0>}
  }
 }
 InnerTabGroupClass {type=Class, default=&/WBP_UIKit_TabGroupCompact_Base.WBP_UIKit_TabGroupCompact_Base_C}
 InnerTabGroupPadding {type=Margin, default=<ltrb 0 8 0 0>}
 LastNotifiedInnerTabID {type=string, default=None}
 LastNotifiedSelectedTabID {type=string, default=None}
 OnSideNavigationCollapsedStateChanged {}
 OnSideNavigationInnerTabClicked {}
 OnSideNavigationSelectionChanged {}
 OnSideNavigationTabClicked {}
 PreviousSelectedTab {type=string, default=None}
 ScrollDestinationGamepad {type=EDescendantScrollDestination, default=IntoView}
 ScrollDestinationMouse {type=EDescendantScrollDestination, default=IntoView}
 ScrollPaddingGamepad {type=number, default=150}
 ScrollPaddingMouse {type=number, default=50}
 SelectInnerTabOnContainerExpand {type=boolean}
 ShouldWaitForInnerTabCallback {type=boolean}
 SideNavigationIsCollapsed {type=boolean}
 SkipExpandAnimation {type=boolean}
 StubSoundStartOffsetMult {type=number, default=0.95}
 StubSoundThresholdMult {type=number, default=0.2}
 TabContainerClass {type=Class, default=&/WBP_UIKit_TabContainer_Base.WBP_UIKit_TabContainer_Base_C}
 TabContainers {type="map<WBP_UIKit_TabContainer_NoLayout_C>", default {}}
 TotalButtons {type=integer}
 TotalButtonsFinishedCollapsedAnim {type=integer}
 UseInitialSideNavigationTabs {type=boolean}
]
widgets {
 info {
  inheritedWidgetCount=0
  parentClass=&/WBP_UIKit_TabGroup_Base.WBP_UIKit_TabGroup_Base_C
  rootWidgetClass=null
  widgetCount=0
 }
 widgets []
}
