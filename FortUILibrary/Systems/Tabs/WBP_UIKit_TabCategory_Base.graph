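# Base UIKit widget for a collapsible tab category: a header button in a named slot above or beside an inner tab group, laid out vertically or horizontally.
# InitContainer takes a category view model, creates the inner tabs widget and fills it; tab clicks and selections are re-broadcast with the category ID.
# SetIsExpanded snaps or animates content size scale, backplate and arrow.
# On gamepad, it disables the expanded header and supports delayed focus on the selected tab.
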WidgetBlueprint /FortUILibrary/Systems/Tabs/WBP_UIKit_TabCategory_Base
animations [
 Collapse_Vertical {
  binding_count=3
  length_seconds=0.5
  object=&self:Collapse_Vertical
  sequence {
   bindings [
    Border_Background {
     id=#1
     possessed_class=Border
     tracks [{
       type=MovieSceneColorTrack
       display_name=Brush Color
       sections [{
         type=MovieSceneColorSection
         end_frame=50
         start_frame=0
         channels [
          Color.R {type=MovieSceneScriptingFloatChannel}
          Color.G {type=MovieSceneScriptingFloatChannel}
          Color.B {type=MovieSceneScriptingFloatChannel}
          Color.A {type=MovieSceneScriptingFloatChannel, keys [[0,1],[50,0]]}
         ]
        }]
      }]
    }
    CollapsedArrow {
     id=#2
     possessed_class=Image
     tracks [
      {
       type=MovieSceneFloatTrack
       display_name=Render Opacity
       sections [{type=MovieSceneFloatSection, end_frame=50, start_frame=0, channels [None {type=MovieSceneScriptingFloatChannel, keys [[0,0],[50,1]]}]}]
      }
      {
       type=MovieScene2DTransformTrack
       sections [{
         type=MovieScene2DTransformSection
         end_frame=50
         start_frame=0
         channels [
          Translation.X {type=MovieSceneScriptingFloatChannel, const [0,1]}
          Translation.Y {type=MovieSceneScriptingFloatChannel, const [0,1]}
          Angle {type=MovieSceneScriptingFloatChannel, keys [[0,180],[50,0]]}
          Scale.X {type=MovieSceneScriptingFloatChannel, const [1,1]}
          Scale.Y {type=MovieSceneScriptingFloatChannel, const [1,1]}
          Shear.X {type=MovieSceneScriptingFloatChannel, const [0,1]}
          Shear.Y {type=MovieSceneScriptingFloatChannel, const [0,1]}
         ]
        }]
      }
     ]
    }
    Border_Content {
     id=#3
     possessed_class=Border
     tracks [{
       type=MovieSceneDoubleVectorTrack
       display_name=Desired Size Scale
       sections [{
         type=MovieSceneDoubleVectorSection
         end_frame=50
         start_frame=0
         channels [
          Vector.X {type=MovieSceneScriptingDoubleChannel}
          Vector.Y {type=MovieSceneScriptingDoubleChannel, keys [[0,1],[50,0]]}
         ]
        }]
      }]
    }
   ]
   display_rate {numerator=100}
  }
 }
 Collapse_Horizontal {
  binding_count=3
  length_seconds=0.5
  object=&self:Collapse_Horizontal
  sequence {
   bindings [
    Border_Background {
     id=#4
     possessed_class=Border
     tracks [{
       type=MovieSceneColorTrack
       display_name=Brush Color
       sections [{
         type=MovieSceneColorSection
         end_frame=50
         start_frame=0
         channels [
          Color.R {type=MovieSceneScriptingFloatChannel}
          Color.G {type=MovieSceneScriptingFloatChannel}
          Color.B {type=MovieSceneScriptingFloatChannel}
          Color.A {type=MovieSceneScriptingFloatChannel, keys [[0,1],[50,0]]}
         ]
        }]
      }]
    }
    CollapsedArrow {
     id=#5
     possessed_class=Image
     tracks [
      {
       type=MovieSceneFloatTrack
       display_name=Render Opacity
       sections [{type=MovieSceneFloatSection, end_frame=50, start_frame=0, channels [None {type=MovieSceneScriptingFloatChannel, keys [[0,0],[50,1]]}]}]
      }
      {
       type=MovieScene2DTransformTrack
       sections [{
         type=MovieScene2DTransformSection
         end_frame=50
         start_frame=0
         channels [
          Translation.X {type=MovieSceneScriptingFloatChannel, const [0,1]}
          Translation.Y {type=MovieSceneScriptingFloatChannel, const [0,1]}
          Angle {type=MovieSceneScriptingFloatChannel, keys [[0,180],[50,0]]}
          Scale.X {type=MovieSceneScriptingFloatChannel, const [1,1]}
          Scale.Y {type=MovieSceneScriptingFloatChannel, const [1,1]}
          Shear.X {type=MovieSceneScriptingFloatChannel, const [0,1]}
          Shear.Y {type=MovieSceneScriptingFloatChannel, const [0,1]}
         ]
        }]
      }
     ]
    }
    Border_Content {
     id=#3
     possessed_class=Border
     tracks [{
       type=MovieSceneDoubleVectorTrack
       display_name=Desired Size Scale
       sections [{
         type=MovieSceneDoubleVectorSection
         end_frame=50
         start_frame=0
         channels [
          Vector.X {type=MovieSceneScriptingDoubleChannel, keys [[0,1],[50,0]]}
          Vector.Y {type=MovieSceneScriptingDoubleChannel}
         ]
        }]
      }]
    }
   ]
   display_rate {numerator=100}
  }
 }
]
graphs [
 ClearCategorySelection {nodes [
   ClearCategorySelection() {is=title K2Node_FunctionEntry}
   n1 = Get CategoryTabs() -> (CategoryTabs:WBP UIKit Tabs Base Object Reference)
   > Is Valid(InputObject=@n1)
   > DeselectAll(self:UIKit Tab Group Object Reference=@n1)
   > RefreshHeaderVisuals()
  ]}
 ConstructTabGroup {nodes [
   ConstructTabGroup() {is=title K2Node_FunctionEntry}
   > n1 = Sequence!(execute) -> (then_0, then_1, then_2:Exec)
   > Is Valid(InputObject=Get CategoryTabs())
   > Unbind all Events from On Tab Selected!(execute:Exec, self:Common Tab List Widget Base Object Reference=Get CategoryTabs()) -> (then:Exec) {is=title K2Node_ClearDelegate}
   > RemoveFromParent(self=Get CategoryTabs())
   n2 = Create Widget!(
    execute:Exec=@n1.then_1
    Class:User Widget Class Reference=Get CategoryTabsClass() -> (CategoryTabsClass:WBP UIKit Tabs Base Class Reference)
    OwningPlayer:Player Controller Object Reference=GetOwningPlayer()
    bAutoListenForInput:bool=False
    bShouldWrapNavigation:bool=True
   ) -> (then:Exec, ReturnValue:WBP UIKit Tabs Base Object Reference)
   > n3 = Set CategoryTabs(CategoryTabs:WBP UIKit Tabs Base Object Reference=@n2) -> (Output_Get:WBP UIKit Tabs Base Object Reference)
   > Bind Event to On Tab Selected(self:Common Tab List Widget Base Object Reference=@n3, Delegate=Create Event()) {is=title K2Node_AddDelegate}
   > Bind Event to On Tab Clicked(self:UIKit Tab Group Object Reference=@n3, Delegate=Create Event()) {is=title K2Node_AddDelegate}
   > ClearChildren(self=Get Border_Content() -> (Border_Content:Border Object Reference))
   > n4 = AddChild(self=Get Border_Content(), Content=Get CategoryTabs())
   > n5 = Cast To BorderSlot!(execute, Object=@n4) -> (then, CastFailed, AsBorder Slot:Border Slot Object Reference) {is=title K2Node_DynamicCast}
   > Set HorizontalAlignment!(execute, HorizontalAlignment:EHorizontalAlignment Enum=HAlign_Fill, self:Border Slot Object Reference=@n5) -> (then, Output_Get:EHorizontalAlignment Enum=HAlign_Fill)
   > Set VerticalAlignment!(execute, VerticalAlignment:EVerticalAlignment Enum=VAlign_Fill, self:Border Slot Object Reference=@n5) -> (then, Output_Get:EVerticalAlignment Enum=VAlign_Fill)
   Call On Category Tabs Creation!(execute:Exec=@n1.then_2, self:WBP UIKit Tab Category Base Object Reference, Tabs:WBP UIKit Tabs Base Object Reference=Get CategoryTabs()) -> (then:Exec) {is=title K2Node_CallDelegate}
  ]}
 EventGraph {nodes [
   n1 = Event OnCategoryTabClicked!() -> (OutputDelegate, then, TabId:Name) {is=title K2Node_CustomEvent}
   > Call On Category Tab Clicked!(
    execute:Exec
    self:WBP UIKit Tab Category Base Object Reference
    CategoryID:Name=Get CategoryID() -> (CategoryID:Name=None)
    TabID:Name=@n1.TabId
    TabVM:UI Kit Tab Group Button View Model Object Reference=GetTabViewModelByID!(self:UIKit Tab Group Object Reference=Get CategoryTabs(), TabNameID:Name=@n1.TabId) -> (ReturnValue:UI Kit Tab Group Button View Model Object Reference)
   ) -> (then:Exec) {is=title K2Node_CallDelegate}
   TryFocusSelectedCategoryTab() {is=title K2Node_CustomEvent}
   > n2 = Switch on ECommonInputType!(execute, Selection:ECommonInputType Enum=GetCurrentInputType(self=Get CommonInputSubsystem())) -> (MouseAndKeyboard:Exec, Gamepad:Exec, Touch:Exec, Count:Exec) {is=title K2Node_SwitchEnum}
   n3 = Branch(execute=@n2.Gamepad, Condition=IsAnyAnimationPlaying!(self:User Widget Object Reference) -> (ReturnValue:bool))
   > SetAllNavigationRules!(execute, self:Widget Object Reference, Rule:EUINavigationRule Enum=Stop, WidgetToFocus:Name=None) -> (then)
   > Delay(Duration=Get ExpandDuration() -> (ExpandDuration:float))
   > n4 = SetAllNavigationRules(Rule=Escape, WidgetToFocus=None)
   FocusSelectedCategoryTab(execute=[@n4.then, @n3.else])
   n5 = Event OnCategoryTabSelected!() -> (OutputDelegate, then, TabId:Name) {is=title K2Node_CustomEvent}
   > Call On Category Tab Selected!(
    execute:Exec
    self:WBP UIKit Tab Category Base Object Reference
    CategoryID:Name=Get CategoryID() -> (CategoryID=None)
    TabID:Name=@n5.TabId
    TabVM:UI Kit Tab Group Button View Model Object Reference=GetTabViewModelByID(self=Get CategoryTabs(), TabNameID=@n5.TabId)
   ) -> (then:Exec) {is=title K2Node_CallDelegate}
   n6 = Event PreConstruct()
   > Parent: PreConstruct(IsDesignTime=@n6.IsDesignTime)
   > RefreshLayout()
   > SetIsExpanded!(execute, self, InIsExpanded:bool=Get Is Expanded() -> (Is Expanded:bool), SkipAnim:bool=true) -> (then)
   Event Construct()
   > Parent: Construct()
   Event Destruct()
   > Parent: Destruct()
   > Unbind all Events from On Category Tab Selected!(execute:Exec, self:WBP UIKit Tab Category Base Object Reference) -> (then:Exec) {is=title K2Node_ClearDelegate}
   > Unbind all Events from On Category Tabs Creation!(execute:Exec, self:WBP UIKit Tab Category Base Object Reference) -> (then:Exec) {is=title K2Node_ClearDelegate}
   > Set CategoryTabs()
  ]}
 ExpandToSelectedItem {nodes [
   ExpandToSelectedItem() {is=title K2Node_FunctionEntry}
   > SetIsExpanded(InIsExpanded=true, SkipAnim=Get Skip Expand Animation() -> (Skip Expand Animation:bool))
   > TryFocusSelectedCategoryTab() {is=title K2Node_CallFunction}
  ]}
 FocusSelectedCategoryTab {nodes [
   FocusSelectedCategoryTab() {is=title K2Node_FunctionEntry}
   n1 = Get Slot_Category() -> (Slot_Category:Named Slot Object Reference)
   > Is Valid(InputObject=GetChildAt!(self:Panel Widget Object Reference=@n1, Index:int=0) -> (ReturnValue:Widget Object Reference))
   > Branch(Condition=HasAnyUserFocusedDescendants!(self:Widget Object Reference=@n1) -> (ReturnValue:bool))
   n2 = Get CategoryTabs()
   > n3 = Set SelectedTab(SelectedTab:Common Button Base Object Reference=GetTabButtonBaseByID!(
     self:Common Tab List Widget Base Object Reference=@n2
     TabNameID:Name=GetSelectedTabId!(self:Common Tab List Widget Base Object Reference=@n2) -> (ReturnValue:Name=None)
    ) -> (ReturnValue:Common Button Base Object Reference)) -> (Output_Get:Common Button Base Object Reference)
   > Is Valid(InputObject=@n3)
   > SetFocus(self=@n3)
  ]}
 GetCategoryTabs {nodes [
   GetCategoryTabs() {is=title K2Node_FunctionEntry}
   > Return Node!(execute, InnerTabs:WBP UIKit Tabs Base Object Reference=Get CategoryTabs())
  ]}
 InitContainer {nodes [
   InitContainer!() -> (then, HeaderTabButton:Common Button Base Object Reference, CategoryVM:UI Kit Tab Group Category View Model Object Reference) {is=title K2Node_FunctionEntry}
   n1 = Get CategoryVM() -> (CategoryVM:UI Kit Tab Group Category View Model Object Reference)
   > n2 = Is Valid(InputObject=@n1)
   n3 = Get Category!(self:UI Kit Tab Group Category View Model Object Reference=@n1) -> (Category:UI Kit Tab Group Button View Model Object Reference)
   > n4 = Is Valid(InputObject=@n3)
   n5 = Get TabID!(self:UI Kit Tab Group Button View Model Object Reference=@n3) -> (TabID:Name=None)
   > n6 = Branch(Condition="Not Equal (Name)"(A=@n5) -> (ReturnValue=~))
   > Set CategoryID(CategoryID:Name=@n5) -> (Output_Get:Name=None)
   > AddChild(self=Get Slot_Category(), Content=Get HeaderTabButton() -> (HeaderTabButton:Common Button Base Object Reference))
   > ConstructTabGroup() {is=title K2Node_CallFunction}
   > Set Tabs Data!(
    execute
    self:WBP UIKit Tabs Base Object Reference=Get CategoryTabs()
    InTabs Data:Array of UI Kit Tab Group Button View Model Object References=Get Tabs!(self:UI Kit Tab Group Category View Model Object Reference=Get CategoryVM()) -> (Tabs:Array of UI Kit Tab Group Button View Model Object References)
   ) -> (then) {is=title K2Node_CallFunction}
   n7 = Self-Reference()
   PrintString!(
    execute=[@n6.else, @n4.Is Not Valid, @n2.Is Not Valid]
    InString=Append!(A=GetDisplayName(Object=@n7), B=.InitContainer:, C:String=" - invalid CategoryVM") -> (ReturnValue)
    bPrintToScreen
    bPrintToLog
    TextColor
    Duration
    Key
   ) -> (then)
  ]}
 OnCategoryExpandedStateChanged {nodes [OnCategoryExpandedStateChanged!() -> (then, CategoryID:Name, IsExpanded:bool=~) {is=title K2Node_FunctionEntry}]}
 OnCategoryTabClicked {nodes [OnCategoryTabClicked!() -> (then, CategoryID:Name, TabID:Name, TabVM:UI Kit Tab Group Button View Model Object Reference) {is=title K2Node_FunctionEntry}]}
 OnCategoryTabSelected {nodes [OnCategoryTabSelected!() -> (then, CategoryID:Name, TabID:Name, TabVM:UI Kit Tab Group Button View Model Object Reference) {is=title K2Node_FunctionEntry}]}
 OnCategoryTabsCreation {nodes [OnCategoryTabsCreation!() -> (then, Tabs:WBP UIKit Tabs Base Object Reference) {is=title K2Node_FunctionEntry}]}
 RefreshGridFill {nodes [
   RefreshGridFill() {is=title K2Node_FunctionEntry}
   > n1 = Branch(Condition=Get Is Vertical() -> (Is Vertical:bool))
   n2 = Get GridLayout() -> (GridLayout:Grid Panel Object Reference)
   > SetRowFill!(execute, self:Grid Panel Object Reference=@n2, RowIndex:int=0, Coefficient:float=0) -> (then)
   > SetRowFill(self=@n2, RowIndex=1, Coefficient=0)
   > SetColumnFill!(execute, self:Grid Panel Object Reference=@n2, ColumnIndex:int=0, Coefficient:float=0) -> (then)
   > SetColumnFill(self=@n2, ColumnIndex=1, Coefficient=1)
   SetRowFill(execute=@n1.else, self=@n2, RowIndex=0, Coefficient=0)
   > SetRowFill(self=@n2, RowIndex=1, Coefficient=1)
   > SetColumnFill(self=@n2, ColumnIndex=0, Coefficient=0)
   > SetColumnFill(self=@n2, ColumnIndex=1, Coefficient=0)
  ]}
 RefreshHeaderVisuals {nodes [
   RefreshHeaderVisuals() {is=title K2Node_FunctionEntry}
   > n1 = Cast To WBP_UIKit_Button_Tab_Collapsible_Base!(execute, Object=GetChildAt(self=Get Slot_Category(), Index=0)) -> (then, CastFailed, AsWBP UIKit Button Tab Collapsible Base:WBP UIKit Button Tab Collapsible Base Object Reference) {is=title K2Node_DynamicCast}
   n2 = GetSelected!(self:Common Button Base Object Reference=@n1) -> (ReturnValue:bool)
   n3 = NOT Boolean(A=Get Is Expanded())
   > SetBrushTintColor!(
    execute
    self:Image Object Reference=Get CollapsedArrow() -> (CollapsedArrow:Image Object Reference)
    TintColor:Slate Color Structure=Select(
     Option 0:Slate Color Structure=Make Slate Color(SpecifiedColor=(R=1,G=1,B=1,A=1))
     Option 1:Slate Color Structure=Make Slate Color(SpecifiedColor=(R=0,G=0,B=0,A=1))
     Index=AND Boolean(A=@n2, B=@n3)
    ) -> (ReturnValue:Slate Color Structure)
   ) -> (then)
   > OnCategoryExpandedStateChanged!(execute, self:WBP UIKit Button Tab Collapsible Base Object Reference=@n1, IsCategoryExpanded:bool=Get Is Expanded()) -> (then) {is=title K2Node_CallFunction}
  ]}
 RefreshLayout {nodes [
   RefreshLayout() {is=title K2Node_FunctionEntry}
   > ClearChildren(self=Get GridLayout())
   > AddChildToGrid!(
    execute
    self:Grid Panel Object Reference=Get GridLayout()
    Content:Widget Object Reference=Get Overlay_Category() -> (Overlay_Category:Overlay Object Reference)
    InRow:int=SelectInt(A=0, B=1, bPickA=Get Is Vertical())
    InColumn:int=SelectInt(A=1, B=0, bPickA=Get Is Vertical())
   ) -> (then, ReturnValue:Grid Slot Object Reference)
   > AddChildToGrid(self=Get GridLayout(), Content=Get Border_Content(), InRow=1, InColumn=1)
   > RefreshGridFill() {is=title K2Node_CallFunction}
  ]}
 SetArrowVisibilty {nodes [
   n1 = SetArrowVisibilty!() -> (then, IsVisible:bool=~) {is=title K2Node_FunctionEntry}
   > Set Visibility(
    Visibility=Select(Option 0:ESlateVisibility Enum=Collapsed, Option 1:ESlateVisibility Enum=SelfHitTestInvisible, Index=@n1) -> (ReturnValue:ESlateVisibility Enum)
    self=Get CollapsedArrow()
   )
  ]}
 SetBackground {nodes [
   n1 = SetBackground!() -> (then, InBrush:"Slate Brush Structure (by ref)") {is=title K2Node_FunctionEntry}
   > SetBrush(self:Border Object Reference=Get Border_Content(), InBrush=@n1)
  ]}
 SetIsExpanded {nodes [
   n1 = SetIsExpanded!() -> (then, InIsExpanded:bool=~, SkipAnim:bool=true) {is=title K2Node_FunctionEntry}
   > n2 = Sequence()
   > n3 = Branch(Condition=NOT Boolean(A=Get SkipAnim() -> (SkipAnim:bool)))
   > Branch(Condition="Not Equal (Boolean)"(A=Get InIsExpanded() -> (InIsExpanded:bool), B=Get Is Expanded()) -> (ReturnValue=~))
   > PlayAnimationDuringTime!(
    execute
    Widget:User Widget Object Reference=Self-Reference()
    Animation:Widget Animation Object Reference=Select(
     Option 0:Widget Animation Object Reference=Get Collapse_Horizontal() -> (Collapse_Horizontal:Widget Animation Object Reference)
     Option 1:Widget Animation Object Reference=Get Collapse_Vertical() -> (Collapse_Vertical:Widget Animation Object Reference)
     Index=Get Is Vertical()
    ) -> (ReturnValue:Widget Animation Object Reference)
    PlayTime:float=Get ExpandDuration()
    bPlayForward:bool=NOT Boolean(A=Get InIsExpanded())
   ) -> (then)
   SetDesiredSizeScale!(
    execute=@n3.else
    self:Border Object Reference=Get Border_Content()
    InScale:Vector 2D Structure=Select(Option 0:Vector 2D Structure, Option 1:Vector 2D Structure=(X=1,Y=1), Index=Get InIsExpanded()) -> (ReturnValue:Vector 2D Structure)
   ) -> (then)
   > SetBrushColor(
    self=Get Border_Background() -> (Border_Background:Border Object Reference)
    InBrushColor=SelectColor(A=(R=1,G=1,B=1,A=1), B=(R=1,G=1,B=1,A=0), bPickA=Get InIsExpanded())
   )
   n4 = Get CollapsedArrow()
   > Set RenderOpacity!(execute, RenderOpacity:float=SelectFloat(A=0, B=1, bPickA=Get InIsExpanded()), self:Widget Object Reference=@n4) -> (then, Output_Get:float)
   > SetRenderTransformAngle!(execute, self:Widget Object Reference=@n4, Angle:float=SelectFloat(A=180, B=0, bPickA=Get InIsExpanded())) -> (then)
   Branch(execute=@n2.then_1, Condition="Not Equal (Boolean)"(A=Get Is Expanded(), B=@n1) -> (ReturnValue=~))
   > n5 = Set Is Expanded(Is Expanded:bool=@n1) -> (Output_Get:bool)
   > Call On Category Expanded State Changed!(execute:Exec, self:WBP UIKit Tab Category Base Object Reference, CategoryID:Name=Get CategoryID() -> (CategoryID=None), IsExpanded:bool=@n5) -> (then:Exec) {is=title K2Node_CallDelegate}
   RefreshHeaderVisuals(execute=@n2.then_2) {is=title K2Node_CallFunction}
   > n6 = Switch on ECommonInputType(Selection=GetCurrentInputType(self=Get CommonInputSubsystem()))
   n7 = Get Is Expanded()
   n8 = Get Slot_Category()
   Set bIsEnabled!(execute=@n6.Gamepad, bIsEnabled:bool=NOT Boolean(A=@n7), self:Widget Object Reference=@n8) -> (then, Output_Get:bool)
   > Set Visibility(
    Visibility=Select(Option 0:ESlateVisibility Enum=SelfHitTestInvisible, Option 1:ESlateVisibility Enum=HitTestInvisible, Index=@n7) -> (ReturnValue:ESlateVisibility Enum)
    self=@n8
   )
  ]}
]
variables [
 CategoryID {type=string, default=None}
 CategoryTabs {type=WBP_UIKit_Tabs_Base_C, default=null}
 CategoryTabsClass {type=Class, default=/FortUILibrary/UIKit/Tabs/WBP_UIKit_Tabs.WBP_UIKit_Tabs_C}
 ExpandDuration {type=number, default=0.25}
 Is Expanded {type=boolean}
 Is Vertical {type=boolean, default=true}
 OnCategoryExpandedStateChanged {}
 OnCategoryTabClicked {}
 OnCategoryTabsCreation {}
 OnCategoryTabSelected {}
 Skip Expand Animation {type=boolean, default=true}
]
widgets {
 info {inheritedWidgetCount=0, parentClass=&UserWidget, rootWidgetClass=&Border, widgetCount=6}
 widgets [
  Border_Background {
   widgetClassPath=&Border
   bIsVariable=true
   widget=Border_Background
   properties {
    background {
     bIsDynamicallyLoaded=false
     drawAs=Image
     imageType=FullColor
     mirroring=NoMirror
     resourceName=None
     resourceObject=/FortUILibrary/Systems/Backplate/MaterialInstances/MI_UIKit_Backplate
     tiling=NoTile
     imageSize=<32 32>
     margin=<ltrb 0 0 0 0>
     outlineSettings {
      bUseBrushTransparency=false
      roundingType=HalfHeightRadius
      width=0
      color {colorUseRule=UseColor_Specified, specifiedColor=<rgba 0 0 0 0>}
      cornerRadii=<0 0 0 0>
     }
     tintColor {colorUseRule=UseColor_Specified, specifiedColor=<rgba 1 1 1 0.2>}
     uVRegion {bIsValid=false, max=<0 0>, min=<0 0>}
    }
    brushColor=<rgba 1 1 1 0>
    padding=<ltrb 0 0 0 0>
   }
  }
  GridLayout {
   widgetClassPath=&GridPanel
   bIsVariable=true
   parent=Border_Background
   widget=GridLayout
   properties {columnFill [0,1], rowFill [0,0]}
   slot {class=BorderSlot, properties {padding=<ltrb 0 0 0 0>}}
  }
  Overlay_Category {
   widgetClassPath=&Overlay
   bIsVariable=true
   parent=GridLayout
   widget=Overlay_Category
   slot {class=GridSlot, properties {column=1}}
  }
  Slot_Category {
   widgetClassPath=&NamedSlot
   bIsVariable=true
   parent=Overlay_Category
   widget=Slot_Category
   slot {class=OverlaySlot, properties {horizontalAlignment=HAlign_Fill, verticalAlignment=VAlign_Fill}}
  }
  CollapsedArrow {
   widgetClassPath=&Image
   bIsVariable=true
   parent=Overlay_Category
   widget=CollapsedArrow
   properties {
    visibility=SelfHitTestInvisible
    brush {
     bIsDynamicallyLoaded=false
     drawAs=Image
     imageType=NoImage
     mirroring=NoMirror
     resourceName=None
     resourceObject=/FortUILibrary/UIKit/Buttons/Materials/M_UIKit_CollapsableArrow
     tiling=NoTile
     imageSize=<21 21>
     margin=<ltrb 0 0 0 0>
     outlineSettings {
      bUseBrushTransparency=false
      roundingType=HalfHeightRadius
      width=0
      color {colorUseRule=UseColor_Specified, specifiedColor=<rgba 0 0 0 0>}
      cornerRadii=<0 0 0 0>
     }
     tintColor {colorUseRule=UseColor_Specified, specifiedColor=<rgba 1 1 1 1>}
     uVRegion {bIsValid=false, max=<0 0>, min=<0 0>}
    }
    colorAndOpacity=<rgba 0.521 0.521 0.521 1>
   }
   slot {class=OverlaySlot, properties {horizontalAlignment=HAlign_Right, verticalAlignment=VAlign_Center, padding=<ltrb 4 4 12 4>}}
  }
  Border_Content {
   widgetClassPath=&Border
   bIsVariable=true
   parent=GridLayout
   widget=Border_Content
   properties {
    clipping=ClipToBounds
    visibility=SelfHitTestInvisible
    background {
     bIsDynamicallyLoaded=false
     drawAs=Image
     imageType=NoImage
     mirroring=NoMirror
     resourceName=None
     resourceObject=null
     tiling=NoTile
     imageSize=<200 200>
     margin=<ltrb 0 0 0 0>
     outlineSettings {
      bUseBrushTransparency=false
      roundingType=HalfHeightRadius
      width=0
      color {colorUseRule=UseColor_Specified, specifiedColor=<rgba 0 0 0 0>}
      cornerRadii=<0 0 0 0>
     }
     tintColor {colorUseRule=UseColor_Specified, specifiedColor=<rgba 1 1 1 1>}
     uVRegion {bIsValid=false, max=<0 0>, min=<0 0>}
    }
    padding=<ltrb 0 0 0 0>
   }
   slot {class=GridSlot, properties {padding=<ltrb 50 50 50 50>}}
  }
 ]
}
