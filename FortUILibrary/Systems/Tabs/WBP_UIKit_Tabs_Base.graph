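# Base UIKit tab bar widget on UIKitTabGroup that builds horizontal or vertical tab buttons from view models, with optional backplates and subgroup dividers.
# Set Tabs Data registers tabs, using view-model button overrides, then updates layout and restores selection; HandleTabCreation styles and grids each button.
# Previous/next prompts follow input-action type (bumper, trigger, custom), input method and tab count.
# An event-track animation resizes spacers as prompts show or hide.
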WidgetBlueprint /FortUILibrary/Systems/Tabs/WBP_UIKit_Tabs_Base
animations [Anim_PromptSpaceChange {
  binding_count=0
  length_seconds=1
  object=&self:Anim_PromptSpaceChange
  sequence {
   bindings []
   display_rate {numerator=100}
   tracks [{
     type=MovieSceneEventTrack
     sections [
      {type=MovieSceneEventRepeaterSection, end_frame=100, start_frame=0}
      {
       type=MovieSceneEventTriggerSection
       channels [None {
         type=MovieSceneScriptingEventChannel
         keys [[0, "<Struct 'MovieSceneEvent' (0x467360fb720) {}>"], [100, "<Struct 'MovieSceneEvent' (0x467360faf00) {}>"]]
        }]
      }
     ]
    }]
  }
 }]
graphs [
 AddBackplate {nodes [
   AddBackplate() {is=title K2Node_FunctionEntry}
   > n1 = Sequence()
   > Cleanup!(execute, self, Widget:Widget Object Reference=Get ContentBackplate() -> (ContentBackplate:WBP UIKit Backplate Base Object Reference)) -> (then)
   > Branch(Condition=Get UseBackplate() -> (UseBackplate:bool))
   > n2 = Construct Object from Class(Class=Get BackplateClass() -> (BackplateClass:WBP UIKit Backplate Base Class Reference)) -> (ReturnValue:WBP UIKit Backplate Base Object Reference)
   > Branch(Condition=IsValid(Object=@n2))
   > n3 = Set ContentBackplate(ContentBackplate:WBP UIKit Backplate Base Object Reference=@n2) -> (Output_Get:WBP UIKit Backplate Base Object Reference)
   > n4 = Branch(Condition=Get PromptsInsideTabsBackplate() -> (PromptsInsideTabsBackplate:bool))
   > n5 = AddChildToGrid!(
    execute
    self:Grid Panel Object Reference=Get Root() -> (Root:Grid Panel Object Reference)
    Content:Widget Object Reference=@n3
    InRow:int=0
    InColumn:int=0
   ) -> (then, ReturnValue:Grid Slot Object Reference)
   SetBackplateProperties!(
    execute=[@n6.then, @n5.then]
    self
    Backplate:WBP UIKit Backplate Base Object Reference=@n3
    Brightness:E_UI_BackplateBrightness Enum=Get BackplateBrightness() -> (BackplateBrightness:E_UI_BackplateBrightness Enum=NewEnumerator0)
    Corner Radius:E_UI_BackplateCornerRadius Enum=Get BackplateCornerRadius() -> (BackplateCornerRadius:E_UI_BackplateCornerRadius Enum=NewEnumerator0)
   ) -> (then)
   n6 = AddChildToGrid(execute=@n4.else, self=Get Main() -> (Main:Grid Panel Object Reference), Content=@n3, InRow=1, InColumn=1)
   Cleanup(execute=@n1.then_1, Widget=Get PrevPromptBackplate() -> (PrevPromptBackplate:WBP UIKit Backplate Base Object Reference))
   > Cleanup(Widget=Get NextPromptBackplate() -> (NextPromptBackplate:WBP UIKit Backplate Base Object Reference))
   > Branch(Condition=Get UsePromptBackplate() -> (UsePromptBackplate:bool))
   > n7 = Sequence()
   > n8 = Construct Object from Class(Class=Get PromptsBackplateClass() -> (PromptsBackplateClass:WBP UIKit Backplate Base Class Reference))
   > Branch(Condition=IsValid(Object=@n8))
   > n9 = Set PrevPromptBackplate(PrevPromptBackplate:WBP UIKit Backplate Base Object Reference=@n8) -> (Output_Get:WBP UIKit Backplate Base Object Reference)
   > AddChild(self=Get Prev_SizeBox() -> (Prev_SizeBox:Size Box Object Reference), Content=@n9)
   n10 = Get PromptBackplateCornerRadius() -> (PromptBackplateCornerRadius:E_UI_BackplateCornerRadius Enum=NewEnumerator0)
   n11 = Get PromptBackplateBrightness() -> (PromptBackplateBrightness:E_UI_BackplateBrightness Enum=NewEnumerator0)
   > SetBackplateProperties(Backplate=@n9, Brightness=@n11, Corner Radius=@n10)
   n12 = Get Prev_SizeBox()
   n13 = Get PromptBackplateMinHeight() -> (PromptBackplateMinHeight:double)
   > SetMinDesiredHeight!(execute, self:Size Box Object Reference=@n12, InMinDesiredHeight:float=@n13) -> (then)
   n14 = Get PromptBackplateMinWidth() -> (PromptBackplateMinWidth:double)
   > SetMinDesiredWidth!(execute, self:Size Box Object Reference=@n12, InMinDesiredWidth:float=@n14) -> (then)
   n15 = Construct Object from Class(execute=@n7.then_1, Class=Get PromptsBackplateClass())
   > Branch(Condition=IsValid(Object=@n15))
   > n16 = Set NextPromptBackplate(NextPromptBackplate:WBP UIKit Backplate Base Object Reference=@n15) -> (Output_Get:WBP UIKit Backplate Base Object Reference)
   > AddChild(self=Get Next_SizeBox() -> (Next_SizeBox:Size Box Object Reference), Content=@n16)
   > SetBackplateProperties(Backplate=@n16, Brightness=@n11, Corner Radius=@n10)
   n17 = Get Next_SizeBox()
   > SetMinDesiredHeight(self=@n17, InMinDesiredHeight=@n13)
   > SetMinDesiredWidth(self=@n17, InMinDesiredWidth=@n14)
  ]}
 AddButtonToGrid {nodes [
   n1 = AddButtonToGrid!() -> (then, Widget:Widget Object Reference, Index:int) {is=title K2Node_FunctionEntry}
   > n2 = Branch(Condition=Get IsVertical() -> (IsVertical:bool))
   n3 = int + int(A=Get DividersCount() -> (DividersCount:int=0), B=@n1.Index)
   > n4 = AddChildToGrid(self=Get TabsContainer() -> (TabsContainer:Grid Panel Object Reference), Content=@n1, InRow=@n3, InColumn=0)
   > SetHorizontalAlignment(
    self:Grid Slot Object Reference=@n4
    InHorizontalAlignment=Get ButtonHorizontalAlignment() -> (ButtonHorizontalAlignment:EHorizontalAlignment Enum=HAlign_Fill)
   )
   n5 = SelectFloat(
    A=0
    B=Get TabSpacing() -> (TabSpacing:double)
    bPickA=OR Boolean(A=Get NoSpacingAfterDivider() -> (NoSpacingAfterDivider:bool), B="Equal (Integer)"(A=@n1.Index))
   )
   > n6 = Set Padding!(execute, Padding:Margin Structure=Make Margin(Left=0, Top=@n5, Right=0, Bottom=0), self:Grid Slot Object Reference=@n4) -> (then, Output_Get:Margin Structure)
   Set NoSpacingAfterDivider(execute=[@n8.then, @n6.then], NoSpacingAfterDivider:bool=false) -> (Output_Get:bool)
   n7 = AddChildToGrid(execute=@n2.else, self=Get TabsContainer(), Content=@n1, InRow=0, InColumn=@n3)
   > SetVerticalAlignment(
    self:Grid Slot Object Reference=@n7
    InVerticalAlignment=Get ButtonVerticalAlignment() -> (ButtonVerticalAlignment:EVerticalAlignment Enum=VAlign_Fill)
   )
   > n8 = Set Padding(Padding=Make Margin(Left=@n5, Top=0, Right=0, Bottom=0), self=@n7)
  ]}
 AddDivider {nodes [
   n1 = AddDivider!() -> (then, Index:int, SubGroup:Name) {is=title K2Node_FunctionEntry}
   > n2 = Sequence()
   Set CurrentSubGroupID(execute=@n2.then_1, CurrentSubGroupID:Name=@n1.SubGroup) -> (Output_Get:Name=None)
   Branch(execute=@n2.then_0, Condition=AND Boolean(A=Get UseDividers() -> (UseDividers:bool), B="Not Equal (Integer)"(A=@n1)))
   > Branch(Condition="Not Equal (Name)"(A=Get CurrentSubGroupID() -> (CurrentSubGroupID:Name=None), B=@n1.SubGroup) -> (ReturnValue=~))
   > n3 = Construct Object from Class(Class=Get DividerClass() -> (DividerClass:WBP UIKit Divider Base Class Reference)) -> (ReturnValue:WBP UIKit Divider Base Object Reference)
   > Branch(Condition=IsValid(Object=@n3))
   > AddDividerToGrid!(execute, self, Divider:WBP UIKit Divider Base Object Reference=@n3, Index:int=@n1) -> (then)
   > SetDividerOrientation!(execute, self:WBP UIKit Divider Base Object Reference=@n3, IsVertical:bool=Get IsVertical()) -> (then)
   > Increment Int(Value=Get DividersCount() -> (DividersCount=0))
   > Set NoSpacingAfterDivider(NoSpacingAfterDivider=true)
  ]}
 AddDividerToGrid {nodes [
   n1 = AddDividerToGrid!() -> (then, Divider:WBP UIKit Divider Base Object Reference, Index:int) {is=title K2Node_FunctionEntry}
   > n2 = Branch(Condition=Get IsVertical())
   n3 = int + int(A=Get DividersCount() -> (DividersCount=0), B=@n1.Index)
   > n4 = AddChildToGrid(self=Get TabsContainer(), Content=@n1, InRow=@n3, InColumn=0)
   SetDividerSpacing!(
    execute=[@n4.then, @n5.then]
    self:WBP UIKit Divider Base Object Reference=@n1
    Spacing:double=Get DividerSpacing() -> (DividerSpacing:double)
    IsVertical:bool=Get IsVertical()
   ) -> (then)
   > SetPadding!(execute, self:User Widget Object Reference=@n1, InPadding:Margin Structure=Get DividerPadding() -> (DividerPadding:Margin Structure)) -> (then)
   n5 = AddChildToGrid(execute=@n2.else, self=Get TabsContainer(), Content=@n1, InRow=0, InColumn=@n3)
  ]}
 Cleanup {nodes [
   n1 = Cleanup!() -> (then, Widget:Widget Object Reference) {is=title K2Node_FunctionEntry}
   > Branch(Condition=IsValid(Object=@n1))
   > RemoveChild(self=GetParent!(self:Widget Object Reference=@n1) -> (ReturnValue:Panel Widget Object Reference), Content=@n1)
  ]}
 ClearTabGroup {nodes [
   ClearTabGroup() {is=title K2Node_FunctionEntry}
   > RemoveAllTabs(self:Common Tab List Widget Base Object Reference)
   > ClearChildren(self=Get TabsContainer())
  ]}
 EventGraph {nodes [
   n1 = Event On Input Method Changed!() -> (OutputDelegate, then, bNewInputType:ECommonInputType Enum) {is=title K2Node_CustomEvent}
   > Set CurrentInputType(CurrentInputType:ECommonInputType Enum=@n1.bNewInputType) -> (Output_Get:ECommonInputType Enum)
   > UpdatePromptsVisibility()
   n2 = ResetButton!() -> (OutputDelegate, then, Button:Common Button Base Object Reference) {is=title K2Node_CustomEvent}
   > n3 = Cast To WBP_UIKit_ButtonGeneric_Base!(execute, Object=@n2.Button) -> (then, CastFailed, AsWBP UIKit Button Generic Base:WBP UIKit Button Generic Base Object Reference) {is=title K2Node_DynamicCast}
   > n4 = For Each Loop(Array:Array of UIKit Block Object References=Get ModularBlocks!(self:UIKit Modular Button Object Reference=@n3) -> (ModularBlocks:Array of UIKit Block Object References)) -> (Array Element:UIKit Block Object Reference)
   > StopAllAnimations(self:User Widget Object Reference=@n4)
   > FlushAnimations(self:User Widget Object Reference=@n4)
   > RequestInstantTransition!(
    execute
    self:UIKit Block Object Reference=@n4
    State:EUIKitBlockInstantTransitionState Enum=Select(
     Option 0:EUIKitBlockInstantTransitionState Enum=InstantEnabled
     Option 1:EUIKitBlockInstantTransitionState Enum=InstantSelected
     Index=GetSelected!(self:Common Button Base Object Reference=@n3) -> (ReturnValue:bool)
    ) -> (ReturnValue:EUIKitBlockInstantTransitionState Enum)
   ) -> (then)
   n5 = Event PreConstruct()
   > n6 = Branch(Condition=@n5.IsDesignTime)
   > n7 = PresetForDesignTime()
   n8 = Get LocalTabsViewModels() -> (LocalTabsViewModels:Array of UI Kit Tab Group Button View Model Object References)
   n9 = Branch(
    execute=[@n7.then, @n6.else]
    Condition=Is Not Empty!(TargetArray:Array of UI Kit Tab Group Button View Model Object References=@n8) -> (ReturnValue:bool)
   )
   > n10 = Set Tabs Data!(execute, self, InTabs Data:Array of UI Kit Tab Group Button View Model Object References=@n8) -> (then) {is=title K2Node_CallFunction}
   AddBackplate(execute=[@n10.then, @n9.else]) {is=title K2Node_CallFunction}
   > n11 = SelectTabByID!(
    execute
    self:Common Tab List Widget Base Object Reference
    TabNameID:Name=Get InitialSelectedTabID() -> (InitialSelectedTabID:Name=None)
    bSuppressClickFeedback:bool=true
   ) -> (then, ReturnValue:bool)
   > n12 = Branch(Condition=@n11)
   SelectTabByIndex!(
    execute=@n12.else
    self:UIKit Tab Group Object Reference
    Index:int=Get InitialSelectedTabIndex() -> (InitialSelectedTabIndex:int=0)
    bSuppressClickFeedback:bool=true
   ) -> (then, ReturnValue:bool)
   Event Construct()
   n13 = Get CommonInputSubsystem()
   > Branch(Condition=IsValid(Object=@n13))
   > Bind Event to On Input Method Changed(self=@n13, Delegate=@n1)
   > Set CurrentInputType(CurrentInputType=GetCurrentInputType(self=@n13))
   > UpdatePromptsVisibility()
   Event Destruct()
   n14 = Get CommonInputSubsystem()
   > Branch(Condition=IsValid(Object=@n14))
   > Unbind Event from On Input Method Changed(self=@n14, Delegate=@n1)
   n15 = Event HandleTabCreation!() -> (OutputDelegate, then, TabNameID:Name=None, TabButton:Common Button Base Object Reference) {is=title K2Node_Event}
   > ResetButton!(execute, self, Button:Common Button Base Object Reference=@n15.TabButton) -> (then) {is=title K2Node_CallFunction}
   > SetupTabButton!(execute, self, Tab ID:Name=@n15.TabNameID, CreatedButtonReference:Common Button Base Object Reference=@n15.TabButton) -> (then)
   > AddButtonToGrid!(
    execute
    self
    Widget:Widget Object Reference=@n15.TabButton
    Index:int=GetTabIndexByID!(self:UIKit Tab Group Object Reference, TabNameID:Name=@n15.TabNameID) -> (ReturnValue:int=0)
   ) -> (then) {is=title K2Node_CallFunction}
  ]}
 GetButtonClass {nodes [
   n1 = GetButtonClass!() -> (then, TabVM:UI Kit Tab Group Button View Model Object Reference) {is=title K2Node_FunctionEntry}
   > n2 = Cast To FortTabViewModel!(execute, Object=@n1) -> (then, CastFailed, AsFort Tab Viewmodel:Fort Tab Viewmodel Object Reference) {is=title K2Node_DynamicCast}
   n3 = GetOverrideTabWidget!(self:Fort Tab Viewmodel Object Reference=@n2) -> (ReturnValue:Common Button Base Class Reference)
   > n4 = Branch(Condition=IsValidClass!(Class:Object Class Reference=@n3) -> (ReturnValue:bool))
   > Return Node!(execute, ButtonClass:Common Button Base Class Reference=@n3)
   Return Node(execute=[@n4.else, @n2.CastFailed], ButtonClass=Get ButtonClass() -> (ButtonClass:WBP UIKit Button Generic Base Class Reference))
  ]}
 OverrideInputActions {nodes [
   n1 = OverrideInputActions!() -> (then, Next Tab:Data Table Row Handle Structure, PreviousTab:Data Table Row Handle Structure) {is=title K2Node_FunctionEntry}
   > SetInputAction!(
    execute
    self:Common Action Widget Object Reference=Get NextTabAction() -> (NextTabAction:Common Action Widget Object Reference)
    InputActionRow:Data Table Row Handle Structure=@n1
   ) -> (then)
   > SetInputAction(self=Get PrevTabAction() -> (PrevTabAction:Common Action Widget Object Reference), InputActionRow=@n1.PreviousTab)
   > SetInputActions!(
    execute
    self:UIKit Tab Group Object Reference
    NextTab:Data Table Row Handle Structure=@n1
    PreviousTab:Data Table Row Handle Structure=@n1.PreviousTab
   ) -> (then)
  ]}
 PresetForDesignTime {nodes [
   PresetForDesignTime() {is=title K2Node_FunctionEntry}
   n1 = Get Slot_Root() -> (Slot_Root:Named Slot Object Reference)
   > SetVisibility(
    self=@n1
    InVisibility=Select(
     Option 0:ESlateVisibility Enum=Collapsed
     Option 1:ESlateVisibility Enum=SelfHitTestInvisible
     Index=HasAnyChildren!(self:Panel Widget Object Reference=@n1) -> (ReturnValue:bool)
    ) -> (ReturnValue:ESlateVisibility Enum)
   )
   > n2 = Branch(Condition=Get HidePrompts() -> (HidePrompts:bool))
   Set Visibility(execute=[@n3.then, @n2.then], Visibility=Collapsed, self=Get Prev_Prompt() -> (Prev_Prompt:Grid Panel Object Reference))
   > Set Visibility(Visibility=Collapsed, self=Get Next_Prompt() -> (Next_Prompt:Grid Panel Object Reference))
   > SetSize!(
    execute
    self:Spacer Object Reference=[Get Next_Spacer() -> (Next_Spacer:Spacer Object Reference), Get Prev_Spacer() -> (Prev_Spacer:Spacer Object Reference)]
    InSize:Vector 2D Structure
   ) -> (then)
   n3 = Branch(
    execute=@n2.else
    Condition="Equal (Enum)"(
     A:E_UI_TabsInputActionType Enum=Get InputActions() -> (InputActions:E_UI_TabsInputActionType Enum=NewEnumerator2)
     B:E_UI_TabsInputActionType Enum=Literal enum E_UI_TabsInputActionType!(Enum:E_UI_TabsInputActionType Enum=NewEnumerator2) -> (ReturnValue:E_UI_TabsInputActionType Enum) {is=title K2Node_EnumLiteral}
    )
   )
   Set Visibility(execute=@n3.else, Visibility=Visible, self=Get Prev_Prompt())
   > Set Visibility(Visibility=Visible, self=Get Next_Prompt())
   n4 = Get ReservedPromptSpace() -> (ReservedPromptSpace:double)
   > SetSize(
    self=[Get Next_Spacer(), Get Prev_Spacer()]
    InSize=Select(Option 0:Vector 2D Structure=MakeVector2D(X=@n4, Y=0), Option 1:Vector 2D Structure=MakeVector2D(X=0, Y=@n4), Index=Get IsVertical()) -> (ReturnValue:Vector 2D Structure)
   )
  ]}
 RestoreTabSelection {nodes [
   n1 = RestoreTabSelection!() -> (then, TabNameID:Name) {is=title K2Node_FunctionEntry}
   > n2 = Cast To WBP_UIKit_ButtonGeneric_Base(Object=GetTabButtonBaseByID!(self:Common Tab List Widget Base Object Reference, TabNameID:Name=@n1) -> (ReturnValue:Common Button Base Object Reference))
   > RequestInstantSelection(self:WBP UIKit Button Generic Base Object Reference=@n2)
   > SelectTabByID(TabNameID=@n1, bSuppressClickFeedback=true)
  ]}
 Sequencer Events {nodes [
   Repeater_PromptSpaceChange() {is=title K2Node_CustomEvent}
   > n1 = Branch(Condition=Get IsVertical())
   n2 = float * float(
    A=GetAnimationCurrentTime!(
     self:User Widget Object Reference
     InAnimation:Widget Animation Object Reference=Get Anim_PromptSpaceChange() -> (Anim_PromptSpaceChange:Widget Animation Object Reference)
    ) -> (ReturnValue:float)
    B=Get ReservedPromptSpace()
   ) {is=title K2Node_PromotableOperator}
   > SetSize(self=[Get Next_Spacer(), Get Prev_Spacer()], InSize=MakeVector2D(X=0, Y=@n2))
   SetSize(execute=@n1.else, self=[Get Next_Spacer(), Get Prev_Spacer()], InSize=MakeVector2D(X=@n2, Y=0))
   Trigger_Start() {is=title K2Node_CustomEvent}
   > SetSize(self=[Get Next_Spacer(), Get Prev_Spacer()])
   Trigger_End() {is=title K2Node_CustomEvent}
   > n3 = Branch(Condition=Get IsPromptsVisible() -> (IsPromptsVisible:bool))
   > Set Visibility(Visibility=Visible, self=Get Next_Prompt())
   > n4 = Set Visibility(Visibility=Visible, self=Get Prev_Prompt())
   n5 = Branch(execute=[@n4.then, @n7.then], Condition=Get IsVertical())
   n6 = Get ReservedPromptSpace()
   > SetSize(self=[Get Prev_Spacer(), Get Next_Spacer()], InSize=MakeVector2D(X=0, Y=@n6))
   SetSize(execute=@n5.else, self=[Get Next_Spacer(), Get Prev_Spacer()], InSize=MakeVector2D(X=@n6, Y=0))
   Set Visibility(execute=@n3.else, Visibility=Collapsed, self=Get Next_Prompt())
   > n7 = Set Visibility(Visibility=Collapsed, self=Get Prev_Prompt())
  ]}
 Set Tabs Data {nodes [
   n1 = Set Tabs Data!() -> (then, InTabs Data:Array of UI Kit Tab Group Button View Model Object References) {is=title K2Node_FunctionEntry}
   > Set LocalTabsViewModels(LocalTabsViewModels:Array of UI Kit Tab Group Button View Model Object References=@n1) -> (Output_Get:Array of UI Kit Tab Group Button View Model Object References)
   > Set CurrentSelection(CurrentSelection:Name=GetSelectedTabId!(self:Common Tab List Widget Base Object Reference) -> (ReturnValue:Name=None)) -> (Output_Get:Name=None)
   > ClearTabGroup() {is=title K2Node_CallFunction}
   > n2 = For Each Loop(Array:Array of UI Kit Tab Group Button View Model Object References=Get LocalTabsViewModels()) -> (Array Element:UI Kit Tab Group Button View Model Object Reference)
   UpdateTabLayout(execute=@n2.Completed)
   > UpdateInputActions()
   > UpdatePromptsVisibility()
   > Set DividersCount(DividersCount:int=0) -> (Output_Get:int=0)
   > Branch(Condition="integer > integer"(A=GetTabCount!(self:Common Tab List Widget Base Object Reference) -> (ReturnValue:int=0)))
   > RestoreTabSelection!(execute, self, TabNameID:Name=Get CurrentSelection() -> (CurrentSelection:Name=None)) -> (then) {is=title K2Node_CallFunction}
   Branch(
    execute=@n2.LoopBody
    Condition=IsTabViewModelValid!(self:UIKit Tab Group Object Reference, TabVM:UI Kit Tab Group Button View Model Object Reference=@n2) -> (ReturnValue:bool)
   )
   > AddDivider!(
    execute
    self
    Index:int=@n2.Array Index
    SubGroup:Name=Select(
     Option 0:Name=None
     Option 1:Name=Get SubGroupID!(self:UI Kit Tab Group Button View Model Object Reference=@n2) -> (SubGroupID:Name=None)
     Index=Get bEnableSubGroupID!(self:UI Kit Tab Group Button View Model Object Reference=@n2) -> (bEnableSubGroupID:bool)
    ) -> (ReturnValue:Name)
   ) -> (then) {is=title K2Node_CallFunction}
   n3 = "To String (Integer)"(InInt=@n2.Array Index)
   > RegisterTab!(
    execute
    self:Common Tab List Widget Base Object Reference
    TabNameID:Name=Select(
     Option 0=String To Name(InString=@n3)
     Option 1=Get TabID!(self:UI Kit Tab Group Button View Model Object Reference=@n2) -> (TabID:Name=None)
     Index=Get bEnableTabID!(self:UI Kit Tab Group Button View Model Object Reference=@n2) -> (bEnableTabID:bool)
    )
    ButtonWidgetType:Common Button Base Class Reference=GetButtonClass!(TabVM:UI Kit Tab Group Button View Model Object Reference=@n2) -> (ButtonClass:Common Button Base Class Reference) {is=title K2Node_CallFunction}
    ContentWidget:Widget Object Reference
    TabIndex:int=@n2.Array Index
   ) -> (then, ReturnValue:bool)
  ]}
 SetBackplateProperties {nodes [
   n1 = SetBackplateProperties!() -> (
    then
    Backplate:WBP UIKit Backplate Base Object Reference
    Brightness:E_UI_BackplateBrightness Enum=NewEnumerator0
    Corner Radius:E_UI_BackplateCornerRadius Enum=NewEnumerator0
   ) {is=title K2Node_FunctionEntry}
   > SetBrightness!(execute, self:WBP UIKit Backplate Base Object Reference=@n1, Brightness:E_UI_BackplateBrightness Enum=@n1.Brightness) -> (then)
   > SetCornerRadius!(execute, self:WBP UIKit Backplate Base Object Reference=@n1, Corner Radius:E_UI_BackplateCornerRadius Enum=@n1.Corner Radius) -> (then)
  ]}
 SetGridAlignment {nodes [
   n1 = SetGridAlignment!() -> (then, WIdget:Widget Object Reference, Horizontal:EHorizontalAlignment Enum=HAlign_Fill, Vertical:EVerticalAlignment Enum=VAlign_Fill) {is=title K2Node_FunctionEntry}
   > n2 = Cast To GridSlot!(execute, Object=Get Slot!(self:Widget Object Reference=@n1) -> (Slot:Panel Slot Object Reference)) -> (then, CastFailed, AsGrid Slot:Grid Slot Object Reference) {is=title K2Node_DynamicCast}
   > SetHorizontalAlignment(self=@n2, InHorizontalAlignment=@n1.Horizontal)
   > SetVerticalAlignment(self=@n2, InVerticalAlignment=@n1.Vertical)
  ]}
 SetGridColumnAndRow {nodes [
   n1 = SetGridColumnAndRow!() -> (then, WIdget:Widget Object Reference, Row:int, Row Span:int, Column:int, Column Span:int) {is=title K2Node_FunctionEntry}
   > n2 = Cast To GridSlot(Object=Get Slot(self=@n1))
   > SetColumn!(execute, self:Grid Slot Object Reference=@n2, InColumn:int=@n1.Column) -> (then)
   > SetColumnSpan!(execute, self:Grid Slot Object Reference=@n2, InColumnSpan:int=@n1.Column Span) -> (then)
   > SetRow!(execute, self:Grid Slot Object Reference=@n2, InRow:int=@n1.Row) -> (then)
   > SetRowSpan!(execute, self:Grid Slot Object Reference=@n2, InRowSpan:int=@n1.Row Span) -> (then)
  ]}
 SetGridLayer {nodes [
   n1 = SetGridLayer!() -> (then, WIdget:Widget Object Reference, Layer:int) {is=title K2Node_FunctionEntry}
   > n2 = Cast To GridSlot(Object=Get Slot(self=@n1))
   > SetLayer!(execute, self:Grid Slot Object Reference=@n2, InLayer:int=@n1.Layer) -> (then)
  ]}
 SetGridNudge {nodes [
   n1 = SetGridNudge!() -> (then, WIdget:Widget Object Reference, Nudge:Vector 2D Structure) {is=title K2Node_FunctionEntry}
   > n2 = Cast To GridSlot(Object=Get Slot(self=@n1))
   > SetNudge!(execute, self:Grid Slot Object Reference=@n2, InNudge:Vector 2D Structure=@n1.Nudge) -> (then)
  ]}
 SetGridPadding {nodes [
   n1 = SetGridPadding!() -> (then, WIdget:Widget Object Reference, Padding:Margin Structure) {is=title K2Node_FunctionEntry}
   > n2 = Cast To GridSlot(Object=Get Slot(self=@n1))
   > SetPadding(self:Grid Slot Object Reference=@n2, InPadding=@n1.Padding)
  ]}
 SetInputActionsType {nodes [
   n1 = SetInputActionsType!() -> (then, InputActions:E_UI_TabsInputActionType Enum=NewEnumerator0, InputActionsWithMouseScroll:bool=~) {is=title K2Node_FunctionEntry}
   > Set InputActions(InputActions:E_UI_TabsInputActionType Enum=@n1) -> (Output_Get:E_UI_TabsInputActionType Enum=NewEnumerator2)
   > Set InputActionsWithMouseScroll(InputActionsWithMouseScroll:bool=@n1.InputActionsWithMouseScroll) -> (Output_Get:bool)
   > UpdateInputActions()
  ]}
 SetPlacement {nodes [
   n1 = SetPlacement!() -> (then, Widget:Widget Object Reference, Placement:S UI Placement Structure) {is=title K2Node_FunctionEntry}
   n2 = Break S UI Placement!(S_UI_Placement:"S UI Placement Structure (by ref)"=@n1.Placement) -> (
    HorizontalAlignment_4_#1:EHorizontalAlignment Enum
    VerticalAlignment_5_#2:EVerticalAlignment Enum
    Padding_8_#3:Margin Structure
    Grid_20_#4:S UI Grid Cell Structure
   ) {is=title K2Node_BreakStruct}
   > SetGridAlignment!(
    execute
    self
    WIdget:Widget Object Reference=@n1
    Horizontal:EHorizontalAlignment Enum=@n2
    Vertical:EVerticalAlignment Enum=@n2.VerticalAlignment_5_#2
   ) -> (then) {is=title K2Node_CallFunction}
   > SetGridPadding!(execute, self, WIdget:Widget Object Reference=@n1, Padding:Margin Structure=@n2.Padding_8_#3) -> (then) {is=title K2Node_CallFunction}
   n3 = Break S UI Grid Cell!(S_UI_GridCell:"S UI Grid Cell Structure (by ref)"=@n2.Grid_20_#4) -> (
    Row_15_#1:int
    RowSpan_17_#5:int
    Column_19_#6:int
    ColumnSpan_21_#7:int
    Layer_23_#8:int
    Nudge_26_#9:Vector 2D Structure
   ) {is=title K2Node_BreakStruct}
   > SetGridColumnAndRow!(
    execute
    self
    WIdget:Widget Object Reference=@n1
    Row:int=@n3
    Row Span:int=@n3.RowSpan_17_#5
    Column:int=@n3.Column_19_#6
    Column Span:int=@n3.ColumnSpan_21_#7
   ) -> (then) {is=title K2Node_CallFunction}
   > SetGridLayer!(execute, self, WIdget:Widget Object Reference=@n1, Layer:int=@n3.Layer_23_#8) -> (then) {is=title K2Node_CallFunction}
   > SetGridNudge!(execute, self, WIdget:Widget Object Reference=@n1, Nudge:Vector 2D Structure=@n3.Nudge_26_#9) -> (then) {is=title K2Node_CallFunction}
  ]}
 SetPromptsVisibility {nodes [
   n1 = SetPromptsVisibility!() -> (then, ShowPrompts:bool=~) {is=title K2Node_FunctionEntry}
   > Branch(Condition="Not Equal (Boolean)"(A=Get IsPromptsVisible(), B=@n1) -> (ReturnValue=~))
   > n2 = Set IsPromptsVisible(IsPromptsVisible:bool=@n1) -> (Output_Get:bool)
   > n3 = Branch(Condition=@n2)
   > n4 = Branch(Condition=Get AnimateReservedPromptSpace() -> (AnimateReservedPromptSpace:bool))
   > PlayAnimationForward(InAnimation=Get Anim_PromptSpaceChange(), PlaybackSpeed=SafeDivide(A=1, B=Get TimeToShowPrompts() -> (TimeToShowPrompts:double)))
   Set Visibility(execute=@n4.else, Visibility=Visible, self=Get Prev_Prompt())
   > Set Visibility(Visibility=Visible, self=Get Next_Prompt())
   n5 = Get ReservedPromptSpace()
   > SetSize(
    self=[Get Next_Spacer(), Get Prev_Spacer()]
    InSize=Select(Option 0:Vector 2D Structure=MakeVector2D(X=@n5, Y=0), Option 1:Vector 2D Structure=MakeVector2D(X=0, Y=@n5), Index=Get IsVertical()) -> (ReturnValue:Vector 2D Structure)
   )
   n6 = Branch(execute=@n3.else, Condition=Get AnimateReservedPromptSpace())
   > PlayAnimationReverse(InAnimation=Get Anim_PromptSpaceChange(), PlaybackSpeed=SafeDivide(A=1, B=Get TimeToHidePrompts() -> (TimeToHidePrompts:double)))
   Set Visibility(execute=@n6.else, Visibility=Collapsed, self=Get Prev_Prompt())
   > Set Visibility(Visibility=Collapsed, self=Get Next_Prompt())
   > SetSize(self=[Get Next_Spacer(), Get Prev_Spacer()])
  ]}
 SetupTabButton {nodes [
   n1 = SetupTabButton!() -> (then, Tab ID:Name, CreatedButtonReference:Common Button Base Object Reference) {is=title K2Node_FunctionEntry}
   > n2 = Cast To WBP_UIKit_ButtonGeneric_Base(Object=@n1.CreatedButtonReference)
   > SetIconSize!(execute, self:WBP UIKit Button Generic Base Object Reference=@n2, Icon:Vector 2D Structure=Get IconSize() -> (IconSize:Vector 2D Structure)) -> (then)
   > SetButtonMinSize!(
    execute
    self:WBP UIKit Button Generic Base Object Reference=@n2
    Min Width:int=Get MinButtonWidth() -> (MinButtonWidth:int=0)
    Min Height:int=Get MinButtonHeight() -> (MinButtonHeight:int=0)
   ) -> (then)
   > n3 = SetData!(
    execute
    self:WBP UIKit Button Generic Base Object Reference=@n2
    VM Button:UI Kit Selectable Button View Model Object Reference=GetTabViewModelByID!(self:UIKit Tab Group Object Reference, TabNameID:Name=Get Tab ID() -> (Tab ID:Name=None)) -> (ReturnValue:UI Kit Tab Group Button View Model Object Reference)
   ) -> (then)
   n4 = Get CreatedButtonReference() -> (CreatedButtonReference:Common Button Base Object Reference)
   SetStyle!(
    execute=[@n3.then, @n2.CastFailed]
    self:Common Button Base Object Reference=@n4
    InStyle:Common Button Style Class Reference=/FortUILibrary/UIKit/Buttons/Blocks/ButtonStyle_UIKit_Empty.ButtonStyle_UIKit_Empty_C
   ) -> (then)
   > SetIsFocusable!(execute, self:Common Button Base Object Reference=@n4, bInIsFocusable:bool=Get IsNaturalNavigation() -> (IsNaturalNavigation:bool)) -> (then)
  ]}
 UpdateInputActions {nodes [
   UpdateInputActions() {is=title K2Node_FunctionEntry}
   > n1 = Switch on E_UI_TabsInputActionType!(execute, Selection:E_UI_TabsInputActionType Enum=Get InputActions() -> (InputActions=NewEnumerator2)) -> (NewEnumerator2:Exec, NewEnumerator0:Exec, NewEnumerator1:Exec, NewEnumerator3:Exec) {is=title K2Node_SwitchEnum}
   n2 = Make Data Table Row Handle!(DataTable:Data Table Object Reference, RowName:Name=None) -> (DataTableRowHandle:Data Table Row Handle Structure) {is=title K2Node_MakeStruct}
   > OverrideInputActions!(execute, self, Next Tab:Data Table Row Handle Structure=@n2, PreviousTab:Data Table Row Handle Structure=@n2) -> (then) {is=title K2Node_CallFunction}
   n3 = Get InputActionsWithMouseScroll() -> (InputActionsWithMouseScroll:bool)
   OverrideInputActions(
    execute=@n1.NewEnumerator0
    Next Tab=Select(
     Option 0:Data Table Row Handle Structure=Get Bumper_NextTab() -> (Bumper_NextTab:Data Table Row Handle Structure)
     Option 1:Data Table Row Handle Structure=Get BumperAndMouseScroll_NextTab() -> (BumperAndMouseScroll_NextTab:Data Table Row Handle Structure)
     Index=@n3
    ) -> (ReturnValue:Data Table Row Handle Structure)
    PreviousTab=Select(
     Option 0=Get Bumper_PrevTab() -> (Bumper_PrevTab:Data Table Row Handle Structure)
     Option 1=Get BumperAndMouseScroll_PrevTab() -> (BumperAndMouseScroll_PrevTab:Data Table Row Handle Structure)
     Index=@n3
    )
   )
   n4 = Get InputActionsWithMouseScroll()
   OverrideInputActions(
    execute=@n1.NewEnumerator1
    Next Tab=Select(
     Option 0=Get Trigger_NextTab() -> (Trigger_NextTab:Data Table Row Handle Structure)
     Option 1=Get TriggerAndMouseScroll_NextTab() -> (TriggerAndMouseScroll_NextTab:Data Table Row Handle Structure)
     Index=@n4
    )
    PreviousTab=Select(
     Option 0=Get Trigger_PrevTab() -> (Trigger_PrevTab:Data Table Row Handle Structure)
     Option 1=Get TriggerAndMouseScroll_PrevTab() -> (TriggerAndMouseScroll_PrevTab:Data Table Row Handle Structure)
     Index=@n4
    )
   )
   SetInputAction(
    execute=@n1.NewEnumerator3
    self=Get NextTabAction()
    InputActionRow=Get NextTabInputActionData() -> (NextTabInputActionData:Data Table Row Handle Structure)
   )
   > SetInputAction(self=Get PrevTabAction(), InputActionRow=Get PreviousTabInputActionData() -> (PreviousTabInputActionData:Data Table Row Handle Structure))
  ]}
 UpdatePromptsVisibility {nodes [
   UpdatePromptsVisibility() {is=title K2Node_FunctionEntry}
   n1 = GetTabCount() -> (ReturnValue=0)
   > n2 = Branch(Condition=OR Boolean(A=Get HidePrompts(), B="Equal (Integer)"(A=@n1, B=1)))
   SetPromptsVisibility!(execute=[@n9.else, @n6.then, @n2.then, @n5.then], self, ShowPrompts:bool=false) -> (then) {is=title K2Node_CallFunction}
   n3 = Literal enum ECommonInputType!(Enum:ECommonInputType Enum=MouseAndKeyboard) -> (ReturnValue:ECommonInputType Enum=~) {is=title K2Node_EnumLiteral}
   n4 = Get CurrentInputType() -> (CurrentInputType:ECommonInputType Enum)
   n5 = Branch(
    execute=@n2.else
    Condition=AND Boolean(
     A=Get HideMouseAndKeyboardPrompts() -> (HideMouseAndKeyboardPrompts:bool)
     B="Equal (Enum)"(A:ECommonInputType Enum=@n4, B:ECommonInputType Enum=@n3)
    )
   )
   n6 = Branch(
    execute=@n5.else
    Condition="Equal (Enum)"(
     A:E_UI_TabsInputActionType Enum=Get InputActions() -> (InputActions=NewEnumerator2)
     B:E_UI_TabsInputActionType Enum=Literal enum E_UI_TabsInputActionType(Enum=NewEnumerator2)
    )
   )
   n7 = Break Slate Brush!(SlateBrush:"Slate Brush Structure (by ref)"=GetIcon!(self:Common Action Widget Object Reference=Get NextTabAction()) -> (ReturnValue:Slate Brush Structure)) -> (
    TintColor:Slate Color Structure
    DrawAs:ESlateBrushDrawType Enum
    Tiling:ESlateBrushTileType Enum
    Mirroring:ESlateBrushMirrorType Enum
    ImageSize:Vector 2D Structure
    Margin:Margin Structure
    ResourceObject:Object Reference
    OutlineSettings:Slate Brush Outline Settings Structure
   ) {is=title K2Node_BreakStruct}
   n8 = Break Slate Brush(SlateBrush=GetIcon(self=Get PrevTabAction()))
   n9 = Branch(execute=@n6.else, Condition=AND Boolean(A=IsValid(Object=@n7.ResourceObject), B=IsValid(Object=@n8.ResourceObject)))
   > SetPromptsVisibility(ShowPrompts=true)
  ]}
 UpdateTabLayout {nodes [
   UpdateTabLayout() {is=title K2Node_FunctionEntry}
   > n1 = Branch(Condition=Get IsVertical())
   n2 = Get Prev_Prompt()
   > SetGridColumnAndRow(WIdget=@n2, Row=0, Row Span=0, Column=1, Column Span=0)
   n3 = Get Prev_Spacer()
   > SetGridColumnAndRow(WIdget=@n3, Row=0, Row Span=0, Column=1, Column Span=0)
   n4 = Get Next_Prompt()
   > SetGridColumnAndRow(WIdget=@n4, Row=2, Row Span=0, Column=1, Column Span=0)
   n5 = Get Next_Spacer()
   > SetGridColumnAndRow(WIdget=@n5, Row=2, Row Span=0, Column=1, Column Span=0)
   n6 = Get TabsContainer()
   > SetColumnFill!(execute, self:Grid Panel Object Reference=@n6, ColumnIndex:int=0, Coefficient:float=1) -> (then)
   > n7 = SetRowFill!(execute, self:Grid Panel Object Reference=@n6, RowIndex:int=0, Coefficient:float=0) -> (then)
   SetGridAlignment(
    execute=[@n13.then, @n7.then]
    WIdget=Get Slot_Tabs() -> (Slot_Tabs:Named Slot Object Reference)
    Horizontal=Get TabGroupHorizontalAlignment() -> (TabGroupHorizontalAlignment:EHorizontalAlignment Enum=HAlign_Fill)
    Vertical=Get TabGroupVerticalAlignment() -> (TabGroupVerticalAlignment:EVerticalAlignment Enum=VAlign_Fill)
   )
   > n8 = Branch(Condition=Get PromptsInsideTabsBackplate())
   n9 = Make Margin(Left=0, Top=0, Right=0, Bottom=0)
   > SetGridPadding(WIdget=Get Slot_Tabs(), Padding=@n9)
   n10 = Get TabGroupPadding() -> (TabGroupPadding:Margin Structure)
   > n11 = SetGridPadding(WIdget=Get Main(), Padding=@n10)
   SetGridAlignment(
    execute=[@n12.then, @n11.then]
    WIdget=Get Prev_Prompt()
    Horizontal=Get PrevPromptHorizontalAlignment() -> (PrevPromptHorizontalAlignment:EHorizontalAlignment Enum=HAlign_Fill)
    Vertical=Get PrevPromptVerticalAlignment() -> (PrevPromptVerticalAlignment:EVerticalAlignment Enum=VAlign_Fill)
   )
   > SetGridAlignment(
    WIdget=Get Next_Prompt()
    Horizontal=Get NextPromptHorizontalAlignment() -> (NextPromptHorizontalAlignment:EHorizontalAlignment Enum=HAlign_Fill)
    Vertical=Get NextPromptVerticalAlignment() -> (NextPromptVerticalAlignment:EVerticalAlignment Enum=VAlign_Fill)
   )
   > SetGridPadding(WIdget=Get Prev_Prompt(), Padding=Get PrevPromptPadding() -> (PrevPromptPadding:Margin Structure))
   > SetGridPadding(WIdget=Get Next_Prompt(), Padding=Get NextPromptPadding() -> (NextPromptPadding:Margin Structure))
   SetGridPadding(execute=@n8.else, WIdget=Get Main(), Padding=@n9)
   > n12 = SetGridPadding(WIdget=Get Slot_Tabs(), Padding=@n10)
   SetGridColumnAndRow(execute=@n1.else, WIdget=@n2, Row=1, Row Span=0, Column=0, Column Span=0)
   > SetGridColumnAndRow(WIdget=@n3, Row=1, Row Span=0, Column=0, Column Span=0)
   > SetGridColumnAndRow(WIdget=@n4, Row=1, Row Span=0, Column=2, Column Span=0)
   > SetGridColumnAndRow(WIdget=@n5, Row=1, Row Span=0, Column=2, Column Span=0)
   > SetColumnFill(self=@n6, ColumnIndex=0, Coefficient=0)
   > n13 = SetRowFill(self=@n6, RowIndex=0, Coefficient=1)
  ]}
]
variables [
 AnimateReservedPromptSpace {type=boolean, default=true}
 BackplateBrightness {type=E_UI_BackplateBrightness, default=Dark}
 BackplateClass {type=Class, default=/FortUILibrary/UIKit/Backplates/WBP_UIKit_Backplate.WBP_UIKit_Backplate_C}
 BackplateCornerRadius {type=E_UI_BackplateCornerRadius, default=Backplate}
 Bumper_NextTab {type=DataTableRowHandle, default {dataTable=&InputActionDataTable, rowName=MainTabRight}}
 Bumper_PrevTab {type=DataTableRowHandle, default {dataTable=&InputActionDataTable, rowName=MainTabLeft}}
 BumperAndMouseScroll_NextTab {type=DataTableRowHandle, default {dataTable=&InputActionDataTable, rowName=MainTabRight_WithMouse}}
 BumperAndMouseScroll_PrevTab {type=DataTableRowHandle, default {dataTable=&InputActionDataTable, rowName=MainTabLeft_WithMouse}}
 ButtonClass {type=Class, default=/FortUILibrary/UIKit/Buttons/WBP_UIKit_Button_Selectable.WBP_UIKit_Button_Selectable_C}
 ButtonHorizontalAlignment {type=EHorizontalAlignment, default=HAlign_Fill}
 ButtonVerticalAlignment {type=EVerticalAlignment, default=VAlign_Top}
 ContentBackplate {type=WBP_UIKit_Backplate_Base_C, default=null}
 ContentDivider {type=WBP_UIKit_Divider_Base_C, default=null}
 CurrentInputType {type=ECommonInputType, default=MouseAndKeyboard}
 CurrentSubGroupID {type=string, default=None}
 DividerClass {type=Class, default=&/WBP_UIKit_Divider_Base.WBP_UIKit_Divider_Base_C}
 DividerPadding {type=Margin, default=<ltrb 0 0 0 0>}
 DividersCount {type=integer}
 DividerSpacing {type=number, default=20}
 HideMouseAndKeyboardPrompts {type=boolean}
 HidePrompts {type=boolean}
 IconSize {type=Vector2D, default=<32 32>}
 InitialSelectedTabID {type=string, default=None}
 InitialSelectedTabIndex {type=integer, default=-1}
 InputActions {type=E_UI_TabsInputActionType, default=Custom}
 InputActionsWithMouseScroll {type=boolean}
 IsNaturalNavigation {type=boolean}
 IsPromptsVisible {type=boolean, default=true}
 IsVertical {type=boolean}
 MinButtonHeight {type=integer}
 MinButtonWidth {type=integer}
 NextPromptBackplate {type=WBP_UIKit_Backplate_Base_C, default=null}
 NextPromptHorizontalAlignment {type=EHorizontalAlignment, default=HAlign_Right}
 NextPromptPadding {type=Margin, default=<ltrb 0 0 0 0>}
 NextPromptVerticalAlignment {type=EVerticalAlignment, default=VAlign_Center}
 NoSpacingAfterDivider {type=boolean}
 PrevPromptBackplate {type=WBP_UIKit_Backplate_Base_C, default=null}
 PrevPromptHorizontalAlignment {type=EHorizontalAlignment, default=HAlign_Left}
 PrevPromptPadding {type=Margin, default=<ltrb 0 0 0 0>}
 PrevPromptVerticalAlignment {type=EVerticalAlignment, default=VAlign_Center}
 PromptBackplateBrightness {type=E_UI_BackplateBrightness, default=Dark}
 PromptBackplateCornerRadius {type=E_UI_BackplateCornerRadius, default=Circle}
 PromptBackplateMinHeight {type=number, default=48}
 PromptBackplateMinWidth {type=number, default=62}
 PromptsBackplateClass {type=Class, default=&WBP_UIKit_Backplate_C}
 PromptsInsideTabsBackplate {type=boolean}
 ReservedPromptSpace {type=number, default=70}
 SelectionAlignment {type=E_UI_Alignment, default=Top}
 TabGroupHorizontalAlignment {type=EHorizontalAlignment, default=HAlign_Fill}
 TabGroupPadding {type=Margin, default=<ltrb 8 8 8 8>}
 TabGroupVerticalAlignment {type=EVerticalAlignment, default=VAlign_Top}
 TabSpacing {type=number, default=20}
 TimeToHidePrompts {type=number, default=0.2}
 TimeToShowPrompts {type=number, default=0.2}
 Trigger_NextTab {type=DataTableRowHandle, default {dataTable=&InputActionDataTable, rowName=SecondaryTabRight}}
 Trigger_PrevTab {type=DataTableRowHandle, default {dataTable=&InputActionDataTable, rowName=SecondaryTabLeft}}
 TriggerAndMouseScroll_NextTab {type=DataTableRowHandle, default {dataTable=&InputActionDataTable, rowName=SecondaryTabRight_WithMouse}}
 TriggerAndMouseScroll_PrevTab {type=DataTableRowHandle, default {dataTable=&InputActionDataTable, rowName=SecondaryTabLeft_WithMouse}}
 UseBackplate {type=boolean, default=true}
 UseDividers {type=boolean}
 UsePromptBackplate {type=boolean}
]
widgets {
 info {inheritedWidgetCount=0, parentClass=/Script/FortniteUI.UIKitTabGroup, rootWidgetClass=&GridPanel, widgetCount=13}
 widgets [
  Root {widgetClassPath=&GridPanel, bIsVariable=true, widget=Root, properties {columnFill [1], rowFill [1]}}
  Slot_Root {widgetClassPath=&NamedSlot, bIsVariable=true, parent=Root, widget=Slot_Root, slot {class=GridSlot, properties {layer=1}}}
  Main {
   widgetClassPath=&GridPanel
   bIsVariable=true
   parent=Root
   widget=Main
   properties {columnFill [0,1,0], rowFill [0,1,0]}
   slot {class=GridSlot, properties {layer=2}}
  }
  Prev_Prompt {
   widgetClassPath=&GridPanel
   bIsVariable=true
   parent=Main
   widget=Prev_Prompt
   properties {columnFill [1], rowFill [1]}
   slot {class=GridSlot, properties {row=1}}
  }
  Prev_SizeBox {widgetClassPath=&SizeBox, bIsVariable=true, parent=Prev_Prompt, widget=Prev_SizeBox, slot {class=GridSlot}}
  PrevTabAction {
   widgetClassPath=&CommonActionWidget
   bIsVariable=true
   parent=Prev_Prompt
   widget=PrevTabAction
   properties {visibility=Collapsed}
   slot {class=GridSlot, properties {horizontalAlignment=HAlign_Center, layer=10, verticalAlignment=VAlign_Center}}
  }
  Prev_Spacer {
   widgetClassPath=&Spacer
   bIsVariable=true
   parent=Main
   widget=Prev_Spacer
   properties {visibility=Visible, size=<48 32>}
   slot {class=GridSlot, properties {horizontalAlignment=HAlign_Center, row=1, verticalAlignment=VAlign_Center}}
  }
  Slot_Tabs {
   widgetClassPath=&NamedSlot
   bIsVariable=true
   parent=Main
   widget=Slot_Tabs
   slot {class=GridSlot, properties {column=1, layer=10, row=1, padding=<ltrb 8 8 8 8>}}
  }
  TabsContainer {
   widgetClassPath=&GridPanel
   bIsVariable=true
   parent=Slot_Tabs
   widget=TabsContainer
   properties {columnFill [1], rowFill [0]}
   slot {class=PanelSlot}
  }
  Next_Spacer {
   widgetClassPath=&Spacer
   bIsVariable=true
   parent=Main
   widget=Next_Spacer
   properties {visibility=Visible, size=<48 32>}
   slot {class=GridSlot, properties {column=2, horizontalAlignment=HAlign_Center, row=1, verticalAlignment=VAlign_Center}}
  }
  Next_Prompt {
   widgetClassPath=&GridPanel
   bIsVariable=true
   parent=Main
   widget=Next_Prompt
   properties {columnFill [1], rowFill [1]}
   slot {class=GridSlot, properties {column=2, row=1}}
  }
  Next_SizeBox {widgetClassPath=&SizeBox, bIsVariable=true, parent=Next_Prompt, widget=Next_SizeBox, slot {class=GridSlot}}
  NextTabAction {
   widgetClassPath=&CommonActionWidget
   bIsVariable=true
   parent=Next_Prompt
   widget=NextTabAction
   properties {visibility=Collapsed}
   slot {class=GridSlot, properties {horizontalAlignment=HAlign_Center, layer=10, verticalAlignment=VAlign_Center}}
  }
 ]
}
